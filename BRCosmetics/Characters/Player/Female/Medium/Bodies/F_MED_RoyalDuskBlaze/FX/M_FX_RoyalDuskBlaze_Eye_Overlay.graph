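# Translucent, unlit, two-sided overlay material for the RoyalDuskBlaze character's eye effect, enabled for Niagara mesh particles.
# Particle color RGB passes through the shared MF_FX_EmissiveAdjustments function, so the emitter drives the glow color.
# The eye mask texture's red channel is multiplied by particle alpha, giving a shaped value that fades with the particle.
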Material /BRCosmetics/Characters/Player/Female/Medium/Bodies/F_MED_RoyalDuskBlaze/FX/M_FX_RoyalDuskBlaze_Eye_Overlay
graph {
 properties {
  bUsedWithNiagaraMeshParticles=true
  blendMode=BLEND_Translucent
  pixelDepthOffsetMode=PDOM_Legacy
  shadingModel=MSM_Unlit
  twoSided=true
  usedShadingModels=Unlit
 }
 nodes [
  Constant3Vector() {constant=<rgba 1 1 1 0>}
  n1 = ParticleColor()
  MaterialFunctionCall(Emissive=@n1.RGB) {materialFunction=&MF_FX_EmissiveAdjustments}
  Multiply(
   A=TextureSample() {texture=&/T_FX_RoyalDuskBlaze_Eye_Mask}.R
   B=@n1.A
  )
 ]
}
