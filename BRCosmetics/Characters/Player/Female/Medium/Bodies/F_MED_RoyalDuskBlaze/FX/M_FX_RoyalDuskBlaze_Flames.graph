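# Translucent unlit flame material for the RoyalDuskBlaze female character's cosmetic FX, usable on Niagara ribbons and mesh particles.
# A shaped UV mask and panning wispy noise, varied by vertex red, drive magenta emissive and particle-alpha opacity, with depth fade on default and deferred paths.
# World position offset adds SplineThicken to a named reroute; a Waves chain builds particle-randomised motion from two panning distortion textures and a sine.
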Material /BRCosmetics/Characters/Player/Female/Medium/Bodies/F_MED_RoyalDuskBlaze/FX/M_FX_RoyalDuskBlaze_Flames
graph {
 properties {
  bUsedWithNiagaraMeshParticles=true
  bUsedWithNiagaraRibbons=true
  blendMode=BLEND_Translucent
  pixelDepthOffsetMode=PDOM_Legacy
  shadingModel=MSM_Unlit
  usedShadingModels=Unlit
 }
 nodes [
  Add(
   A=MaterialFunctionCall(
    WidthBase=ScalarParameter() {defaultValue=10, parameterName=Width Base}
    WidthTip=ScalarParameter() {defaultValue=10, parameterName=Width Tip}
    Expand U or V UV Channel=StaticBool() {value=true}
   ) {materialFunction=&SplineThicken}.WorldPositionOffset
   B=NamedRerouteUsage()
  )
  n1 = VectorParameter() {parameterName=MainColor, defaultValue=<rgba 0.8714 0 0.6724 1>}
  n2 = Multiply(A=@n1.RGB) {constB=0.2}
  n3 = ComponentMask(None=TextureCoordinate()) {g=true}
  n4 = ScalarParameter() {defaultValue=0.232, group=VMask, parameterName=Amount}
  n5 = SmoothStep(Min=@n4, Max=Add(A=@n4, B=ScalarParameter() {defaultValue=1, group=VMask, parameterName=Softness}), Value=@n3)
  n6 = ScalarParameter() {group=VMask, parameterName=UseVMask?}
  n7 = SphereMask(A=ComponentMask(None=TextureCoordinate()) {r=true}, B=Constant() {r=0.5}) {attenuationRadius=0.5, hardnessPercent=0}
  n8 = ScalarParameter() {defaultValue=1, parameterName=Concave, sliderMax=10}
  n9 = SmoothStep(Value=Multiply(A=LinearInterpolate(A=@n3, B=@n5, Alpha=@n6), B=Power(Base=@n7, Exp=@n8))) {constMin=0.1}
  n10 = AppendVector(
   A=ScalarParameter() {group=Wispy Flame, parameterName=Speed X, sliderMax=5, sliderMin=-5}
   B=ScalarParameter() {defaultValue=1, group=Wispy Flame, parameterName=Speed Y, sliderMax=5, sliderMin=-5}
  )
  n11 = VertexColor()
  n12 = Multiply(
   A=Panner(Coordinate=TextureCoordinate(), Speed=Multiply(A=@n10, B=@n11.R)) {speedX=1}
   B=AppendVector(
    A=ScalarParameter() {defaultValue=1, group=Wispy Flame, parameterName=Scale X, sliderMax=5}
    B=ScalarParameter() {defaultValue=1, group=Wispy Flame, parameterName=Scale Y, sliderMax=5}
   )
  )
  n13 = AppendVector(A=@n11.R, B=Constant())
  n14 = TextureSampleParameter2D(UVs=MaterialFunctionCall(XY=Add(A=@n12, B=@n13)) {materialFunction=&Swizzle}.YX) {
   group=Wispy Flame
   parameterName=Flame Texture
   texture=/BRCosmetics/Athena/Effects/Particles/Skydiving/NS_Contrail_Chainmail/T_WispyFireNoise03
  }
  n15 = LinearInterpolate(A=Multiply(A=Multiply(A=@n9) {constB=5}, B=@n14.R), B=@n9, Alpha=@n9)
  n16 = ScalarParameter() {defaultValue=2, parameterName=Base Mult}
  n17 = Constant() {r=0.6945}
  n18 = SmoothStep(
   Min=@n17
   Max=Add(A=@n17, B=Constant() {r=0.104})
   Value=MaterialFunctionCall(Float2=TextureCoordinate()) {materialFunction=&BreakOutFloat2Components}.G
  )
  MaterialFunctionCall(Emissive=Multiply(A=LinearInterpolate(A=@n2, B=@n1.RGB, Alpha=@n15) {constB=0}, B=LinearInterpolate(B=@n16, Alpha=@n18) {constA=1, constB=1.5})) {materialFunction=&MF_FX_EmissiveAdjustments}
  n19 = NamedRerouteUsage()
  n20 = SmoothStep(
   Max=OneMinus(None=ScalarParameter() {defaultValue=0.9, parameterName=Edge Hardness, sliderMax=1})
   Value=Multiply(A=ParticleColor().A, B=Multiply(A=@n15, B=@n19))
  ) {constMax=0.1}
  n21 = DepthFade(Opacity=@n20, FadeDistance=ScalarParameter() {defaultValue=35, parameterName=Depth Fade Distance, sliderMax=100}) {fadeDistanceDefault=25}
  Multiply(
   A=ShadingPathSwitch(Default=@n21, Deferred=@n21, Forward=@n20, Mobile=@n20)
   B=ScalarParameter() {defaultValue=1, parameterName=Opacity, sliderMax=1}
  )
  n22 = ScalarParameter() {group=WPO, parameterName=WPO Intensity}
  n23 = DynamicParameter() {defaultValue=<rgba 1 1 50 1>, paramNames [Smoke Pan Speed, WPO Speed, Sphere Mask Radius, WPO Intensity]}
  Multiply(A=@n22, B=@n23.WPO Intensity)
  n24 = LinearInterpolate(
   A=ScalarParameter() {defaultValue=1, group=WPO, parameterName=WPO UV Scale Mult}
   B=DynamicParameter() {parameterIndex=1, defaultValue=<rgba 1 1 50 1>, paramNames [WPO UV Scale, "", "", ""]}.WPO UV Scale
   Alpha=ScalarParameter() {defaultValue=1, group=WPO, parameterName=Use Dynamic WPO UV Scale}
  )
  n25 = ComponentMask(None=VectorParameter() {group=WPO, parameterName=WPO_A_UVScale, defaultValue=<rgba 0 1 0 0>}.RGB) {g=true, r=true}
  Multiply(A=@n24, B=@n25)
  n26 = ComponentMask(None=VectorParameter() {group=WPO, parameterName=WPO_B_UVScale, defaultValue=<rgba 0 1 0 0>}.RGB) {g=true, r=true}
  Multiply(A=@n26, B=@n24)
  n27 = Constant3Vector() {constant=<rgba 0.056 1.2668 0 0>}
  n28 = OneMinus(None=SphereMask(
    A=TextureCoordinate()
    B=MaterialFunctionCall(X=@n27.R, Y=@n27.R) {materialFunction=&MakeFloat2}.Result
   ) {attenuationRadius=0.5, hardnessPercent=0})
  NamedRerouteDeclaration(None=Saturate(None=Power(Base=@n28) {constExponent=10})) {name=Mask, nodeColor=<rgba 0 1 0.0156 1>}
  n29 = Add(A=TextureCoordinate(), B=ParticleRandom())
  n30 = AppendVector(A=Constant(), B=ScalarParameter() {defaultValue=0.5, group=WPO, parameterName=WPO_B_Speed})
  n31 = ScalarParameter() {defaultValue=1, parameterName=WPO Speed}
  n32 = TextureSample(UVs=Panner(Coordinate=Multiply(A=@n29, B=@n26), Speed=Multiply(A=@n30, B=@n31)) {speedY=0.5}) {samplerType=SAMPLERTYPE_Normal, texture=&T_Weap_Distort_N_02}
  n33 = Add(A=TextureCoordinate(), B=ParticleRandom())
  n34 = AppendVector(A=Constant(), B=ScalarParameter() {defaultValue=0.5, group=WPO, parameterName=WPO_A_Speed})
  n35 = TextureSample(UVs=Panner(Coordinate=Multiply(A=@n33, B=@n25), Speed=Multiply(A=@n34, B=@n31)) {speedY=0.5}) {samplerType=SAMPLERTYPE_Normal, texture=&T_Weap_Distort_N_02}
  n36 = Add(
   A=Multiply(A=ScalarParameter() {defaultValue=1, group=WPO, parameterName=WPO_B_Intensity}, B=LinearInterpolate(Alpha=@n32.R) {constB=1.5})
   B=Multiply(A=ScalarParameter() {defaultValue=1, group=WPO, parameterName=WPO_A_Intensity}, B=LinearInterpolate(Alpha=@n35.R) {constB=1.5})
  )
  n37 = OneMinus(None=ComponentMask(None=TextureCoordinate()) {g=true})
  n38 = Saturate(None=Multiply(A=Power(Base=@n37) {constExponent=1}))
  n39 = OneMinus(None=ComponentMask(None=TextureCoordinate()) {g=true})
  n40 = Clamp(None=Multiply(A=Power(Base=@n39) {constExponent=1})) {maxDefault=0.5}
  n41 = Time()
  n42 = Add(A=ComponentMask(None=TextureCoordinate()) {g=true}, B=Add(A=Multiply(A=@n31, B=@n41), B=ParticleRandom()))
  n43 = Multiply(A=Saturate(None=Sine(None=@n42))) {constB=0.3}
  n44 = Add(
   A=Multiply(A=Multiply(A=@n36, B=@n38), B=Constant3Vector() {constant=<rgba 1.5 0 0 1>})
   B=Multiply(A=Multiply(A=@n40, B=@n43), B=Constant3Vector() {constant=<rgba 0 0 1 1>})
  ) {constB=0}
  NamedRerouteDeclaration(None=Multiply(A=Transform(None=@n44), B=@n22) {constB=12}) {name=Waves, nodeColor=<rgba 1 0.1641 0 1>}
 ]
}
