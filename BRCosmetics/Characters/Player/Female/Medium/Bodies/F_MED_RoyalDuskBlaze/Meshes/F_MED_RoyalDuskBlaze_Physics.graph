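# Physics asset for the RoyalDuskBlaze medium female player body cosmetic, covering its dangling belt straps and right-side cloth volumes.
# Capsules on the pelvis and lower spine form the body core, with two two-bone cloth chains constrained to the pelvis.
# Front and back strap chains attach to spine_01 through master bones, each running three links.
# Strap links have narrow swing and twist limits, while core, cloth and master joints rotate freely.
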PhysicsAsset /BRCosmetics/Characters/Player/Female/Medium/Bodies/F_MED_RoyalDuskBlaze/Meshes/F_MED_RoyalDuskBlaze_Physics
bodies {
 dyn_Cloth_volume_r_A [{shapeType=Capsule, length=10.3452, radius=1.9515, shapeName=dyn_Cloth_volume_r_B_capsule, center=<-1.8977 0.1176 0>, rotation=<rot 0 -90 0>}]
 dyn_Cloth_volume_r_A01 [{shapeType=Capsule, length=10.9452, radius=1.9515, shapeName=dyn_Cloth_volume_r_B_capsule, center=<-1.1771 0.1176 0>, rotation=<rot 0 -90 0>}]
 dyn_Cloth_volume_r_B [{shapeType=Capsule, length=10.9962, radius=1.9005, shapeName=dyn_Cloth_volume_r_B_capsule, center=<-2.6487 -0.0615 0>, rotation=<rot 0 -90 0>}]
 dyn_Cloth_volume_r_B01 [{shapeType=Capsule, length=10.9962, radius=1.9005, shapeName=dyn_Cloth_volume_r_B_capsule, center=<-1.1771 0.4382 0>, rotation=<rot 0 -90 0>}]
 dyn_belt_strap_bk_01 [{shapeType=Capsule, length=0.4206, radius=1.7168, shapeName=dyn_belt_strap_bk_01_capsule, center=<1.9012 0 3>, rotation=<rot 90 0 0>}]
 dyn_belt_strap_bk_02 [{shapeType=Capsule, length=0, radius=2.2066, shapeName=dyn_belt_strap_bk_02_capsule, center=<2.0455 0 3>, rotation=<rot 90 0 0>}]
 dyn_belt_strap_bk_03 [{shapeType=Capsule, length=4.1511, radius=1.4306, shapeName=dyn_belt_strap_bk_03_capsule, center=<3.3631 0 3>, rotation=<rot 90 0 0>}]
 dyn_belt_strap_bk_mstr [{shapeType=Capsule, length=2, radius=2, shapeName=dyn_belt_strap_bk_mstr_capsule}]
 dyn_belt_strap_fr_01 [{shapeType=Capsule, length=0.4206, radius=1.7168, shapeName=dyn_belt_strap_fr_01_capsule, center=<1.9012 0 3>, rotation=<rot 90 0 0>}]
 dyn_belt_strap_fr_02 [{shapeType=Capsule, length=0, radius=2.2066, shapeName=dyn_belt_strap_fr_02_capsule, center=<1.9012 0 3>, rotation=<rot 90 0 0>}]
 dyn_belt_strap_fr_03 [{shapeType=Capsule, length=4.1511, radius=1.4306, shapeName=dyn_belt_strap_fr_03_capsule, center=<3.3631 0 3>, rotation=<rot 90 0 0>}]
 dyn_belt_strap_fr_mstr [{shapeType=Capsule, length=2, radius=2, shapeName=dyn_belt_strap_fr_mstr_capsule}]
 pelvis [
  {shapeType=Capsule, length=17.4865, radius=7.7669}
  {shapeType=Capsule, length=8.1918, radius=6.6685, shapeName=pelvis_capsule, center=<2.7257 0 8.3385>, rotation=<rot 0 0 90>}
 ]
 spine_01 [{shapeType=Capsule, length=15.859, radius=6.2166, shapeName=spine_01_capsule}]
 spine_02 [{shapeType=Capsule, length=8.176, radius=6.2166, shapeName=spine_02_capsule}]
}
constraints [
 {
  bone1Name=spine_01
  bone2Name=pelvis
  swing1LimitDegrees=45
  swing1Motion=Free
  swing2LimitDegrees=45
  swing2Motion=Free
  twistLimitDegrees=45
  twistMotion=Free
 }
 {
  bone1Name=spine_02
  bone2Name=spine_01
  swing1LimitDegrees=45
  swing1Motion=Free
  swing2LimitDegrees=45
  swing2Motion=Free
  twistLimitDegrees=45
  twistMotion=Free
 }
 {
  bone1Name=dyn_Cloth_volume_r_B
  bone2Name=pelvis
  swing1LimitDegrees=45
  swing1Motion=Free
  swing2LimitDegrees=45
  swing2Motion=Free
  twistLimitDegrees=45
  twistMotion=Free
 }
 {
  bone1Name=dyn_belt_strap_bk_mstr
  bone2Name=spine_01
  swing1LimitDegrees=45
  swing1Motion=Free
  swing2LimitDegrees=45
  swing2Motion=Free
  twistLimitDegrees=45
  twistMotion=Free
 }
 {bone1Name=dyn_belt_strap_bk_01, bone2Name=dyn_belt_strap_bk_mstr, swing1LimitDegrees=20, swing2LimitDegrees=20, twistLimitDegrees=7}
 {bone1Name=dyn_belt_strap_bk_02, bone2Name=dyn_belt_strap_bk_01, swing1LimitDegrees=25, swing2LimitDegrees=35, twistLimitDegrees=7}
 {bone1Name=dyn_belt_strap_bk_03, bone2Name=dyn_belt_strap_bk_02, swing1LimitDegrees=25, swing2LimitDegrees=35, twistLimitDegrees=7}
 {
  bone1Name=dyn_belt_strap_fr_mstr
  bone2Name=spine_01
  swing1LimitDegrees=45
  swing1Motion=Free
  swing2LimitDegrees=45
  swing2Motion=Free
  twistLimitDegrees=45
  twistMotion=Free
 }
 {bone1Name=dyn_belt_strap_fr_01, bone2Name=dyn_belt_strap_fr_mstr, swing1LimitDegrees=20, swing2LimitDegrees=20, twistLimitDegrees=7}
 {bone1Name=dyn_belt_strap_fr_02, bone2Name=dyn_belt_strap_fr_01, swing1LimitDegrees=25, swing2LimitDegrees=35, twistLimitDegrees=7}
 {bone1Name=dyn_belt_strap_fr_03, bone2Name=dyn_belt_strap_fr_02, swing1LimitDegrees=25, swing2LimitDegrees=35, twistLimitDegrees=7}
 {
  bone1Name=dyn_Cloth_volume_r_B01
  bone2Name=dyn_Cloth_volume_r_B
  swing1LimitDegrees=45
  swing1Motion=Free
  swing2LimitDegrees=45
  swing2Motion=Free
  twistLimitDegrees=45
  twistMotion=Free
 }
 {
  bone1Name=dyn_Cloth_volume_r_A01
  bone2Name=dyn_Cloth_volume_r_A
  swing1LimitDegrees=45
  swing1Motion=Free
  swing2LimitDegrees=45
  swing2Motion=Free
  twistLimitDegrees=45
  twistMotion=Free
 }
 {
  bone1Name=dyn_Cloth_volume_r_A
  bone2Name=pelvis
  swing1LimitDegrees=45
  swing1Motion=Free
  swing2LimitDegrees=45
  swing2Motion=Free
  twistLimitDegrees=45
  twistMotion=Free
 }
]
