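# Control Rig for the RoyalDuskBlaze female skirt that procedurally corrects cloth volume and armor bones from body pose.
# Spherical pose readers on thighs, pelvis, spine, a calf and upper arms weight parent constraints and transform offsets.
# Left and right skirt bones are posed first, then blended into front, back and in-between float bones, with extra thigh-driven front correctives.
# A further branch adjusts wrist guard bones from hand pose.
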ControlRigBlueprint /BRCosmetics/Characters/Player/Female/Medium/Bodies/F_MED_RoyalDuskBlaze/Meshes/F_MED_RoyalDuskBlaze_Skirt_CtrlRig
graphs [
 ContainedGraph {nodes [
   n1 = RigVMFunctionEntryNode()
   SequenceAggregate = &RigVMFunction_Sequence(ExecuteContext=@n1.ExecuteContext)
   SequenceAggregate_1 = &RigVMFunction_Sequence(ExecuteContext=@SequenceAggregate.B)
   SequenceAggregate_2 = &RigVMFunction_Sequence(ExecuteContext=@SequenceAggregate_1.B)
   RigVMFunctionReturnNode(A=@SequenceAggregate.A, B=@SequenceAggregate_1.A, C=@SequenceAggregate_2.A, D=@SequenceAggregate_2.B)
  ]}
 RigVMFunctionLibrary {links [], nodes []}
 RigVMModel {nodes [
   &RigUnit_ParentConstraint(
    AdvancedSettings=(RotationOrderForFilter=XZY)
    Child=(Type=Bone,Name="dyn_diamond_l_fr_big_mstr")
    ExecutePin=()
    Filter=(TranslationFilter=(bX=true,bY=true,bZ=true),RotationFilter=(bX=True,bY=True,bZ=True),ScaleFilter=(bX=True,bY=True,bZ=True))
    Parents=((Item=(Type=Bone,Name="dyn_skirt_l_fr_02"),Weight=0.5),(Item=(Type=Bone,Name="dyn_skirt_l_fr_mid_02"),Weight=0.5))
    Weight=1
    bMaintainOffset=true
   )
   > &RigUnit_ParentConstraint(
    AdvancedSettings=(RotationOrderForFilter=XZY)
    Child=(Type=Bone,Name="dyn_diamond_l_fr_sml_mstr")
    Filter=(TranslationFilter=(bX=true,bY=true,bZ=true),RotationFilter=(bX=True,bY=True,bZ=True),ScaleFilter=(bX=True,bY=True,bZ=True))
    Parents=((Item=(Type=Bone,Name="dyn_skirt_l_mid_02"),Weight=0.5),(Item=(Type=Bone,Name="dyn_skirt_l_fr_mid_02"),Weight=0.5))
    Weight=1
   )
   RigUnit_BeginExecution = &RigUnit_BeginExecution()
   SequenceAggregate_2 = &RigVMFunction_Sequence(ExecuteContext=@RigUnit_BeginExecution.ExecutePin)
   SequenceAggregate_1 = RigVMAggregateNode(ExecuteContext=@SequenceAggregate_2.A)
   n1 = &RigUnit_SphericalPoseReader(
    ActiveRegionScaleFactors=(PositiveWidth=2.5,NegativeWidth=3,PositiveHeight=3,NegativeHeight=3)
    ActiveRegionSize=0.1
    Debug=(bDrawDebug=false,DebugScale=25,DebugSegments=20,DebugThickness=0.25)
    DriverAxis=(X=1,Y=0,Z=0)
    DriverItem=(Type=Bone,Name="upperarm_twist_01_l")
    FalloffRegionScaleFactors=(PositiveWidth=1,NegativeWidth=1,PositiveHeight=1,NegativeHeight=1)
    FalloffSize=0.7
    OptionalParentItem=(Type=Bone)
    RotationOffset=(X=0,Y=10,Z=0)
    ExecutePin=@SequenceAggregate_1.A
   )
   > &RigUnit_ModifyTransforms(
    ItemToModify="((Item=(Type=Bone,Name=\"upperarm_l\"),Transform=(Rotation=(X=0,Y=0,Z=0,W=1),Translation=(X=0,Y=0,Z=0),Scale3D=(X=1,Y=1,Z=1))),(Item=(Type=Bone,Name=\"upperarm_twist_01_l\"),Transform=(Rotation=(X=0,Y=0,Z=0,W=1),Translation=(X=0,Y=0,Z=0),Scale3D=(X=1,Y=1,Z=1)) ...(truncated)"
    ItemToModify.1.Transform.Rotation=&RigVMFunction_MathQuaternionFromEuler(Euler=(X=20,Y=0,Z=0)).Result
    Weight=@n1.OutputParam
   )
   > n2 = &RigUnit_SphericalPoseReader(
    ActiveRegionScaleFactors=(PositiveWidth=2,NegativeWidth=3,PositiveHeight=1,NegativeHeight=1)
    ActiveRegionSize=0.1
    Debug=(bDrawDebug=false,DebugScale=25,DebugSegments=20,DebugThickness=0.25)
    DriverAxis=(X=1,Y=0,Z=0)
    DriverItem=(Type=Bone,Name="upperarm_l")
    FalloffRegionScaleFactors=(PositiveWidth=1,NegativeWidth=1,PositiveHeight=2,NegativeHeight=2)
    FalloffSize=0.7
    OptionalParentItem=(Type=Bone)
    RotationOffset=(X=0,Y=10,Z=0)
   )
   > &RigUnit_ModifyTransforms(
    ItemToModify="((Item=(Type=Bone,Name=\"upperarm_armor_01_l\"),Transform=(Rotation=(X=0,Y=0,Z=0,W=1),Translation=(X=3.5,Y=-1,Z=0),Scale3D=(X=0.6,Y=1.2,Z=1))),(Item=(Type=Bone,Name=\"None\"),Transform=(Rotation=(X=0,Y=0,Z=0,W=1),Translation=(X=0,Y=0,Z=0),Scale3D=(X=0,Y=0,Z=0) ...(truncated)"
    ItemToModify.0.Transform.Rotation=&RigVMFunction_MathQuaternionFromEuler(Euler=(X=80,Y=20,Z=20)).Result
    Weight=@n2.OutputParam
   )
   > n3 = &RigUnit_SphericalPoseReader(
    ActiveRegionScaleFactors=(PositiveWidth=2.5,NegativeWidth=3,PositiveHeight=3,NegativeHeight=3)
    ActiveRegionSize=0.1
    Debug=(bDrawDebug=false,DebugScale=25,DebugSegments=20,DebugThickness=0.25)
    DriverAxis=(X=-1,Y=0,Z=0)
    DriverItem=(Type=Bone,Name="upperarm_r")
    FalloffRegionScaleFactors=(PositiveWidth=1,NegativeWidth=1,PositiveHeight=1,NegativeHeight=1)
    FalloffSize=0.5
    FlipHeightScaling=true
    OptionalParentItem=(Type=Bone)
    RotationOffset=(X=180,Y=0,Z=180)
   )
   > &RigUnit_ModifyTransforms(
    ItemToModify=((Item=(Type=Bone,Name="upperarm_twist_01_r"),Transform=(Rotation=(X=0,Y=0,Z=0,W=1),Translation=(X=0,Y=0,Z=0),Scale3D=(X=1,Y=1,Z=1))))
    ItemToModify.0.Transform.Rotation=&RigVMFunction_MathQuaternionFromEuler(Euler=(X=-20,Y=0,Z=0)).Result
    Weight=@n3.OutputParam
   )
   > n4 = &RigUnit_SphericalPoseReader(
    ActiveRegionScaleFactors=(PositiveWidth=2,NegativeWidth=3,PositiveHeight=1,NegativeHeight=1)
    ActiveRegionSize=0.1
    Debug=(bDrawDebug=false,DebugScale=25,DebugSegments=20,DebugThickness=0.25)
    DriverAxis=(X=-1,Y=0,Z=0)
    DriverItem=(Type=Bone,Name="upperarm_r")
    FalloffRegionScaleFactors=(PositiveWidth=1,NegativeWidth=1,PositiveHeight=0.9,NegativeHeight=0.9)
    FalloffSize=0.7
    FlipWidthScaling=false
    OptionalParentItem=(Type=Bone)
    RotationOffset=(X=180,Y=-10,Z=180)
   )
   > &RigUnit_ModifyTransforms(
    ItemToModify="((Item=(Type=Bone,Name=\"upperarm_armor_01_r\"),Transform=(Rotation=(X=0,Y=0,Z=0,W=1),Translation=(X=-1.5,Y=2,Z=-1.5),Scale3D=(X=0.6,Y=1,Z=1))),(Item=(Type=Bone,Name=\"None\"),Transform=(Rotation=(X=0,Y=0,Z=0,W=1),Translation=(X=0,Y=0,Z=0),Scale3D=(X=0,Y=0,Z=0 ...(truncated)"
    ItemToModify.0.Transform.Rotation=&RigVMFunction_MathQuaternionFromEuler(Euler=(X=100,Y=0,Z=20)).Result
    Weight=@n4.OutputParam
   )
   n5 = &RigUnit_SphericalPoseReader(
    ActiveRegionScaleFactors=(PositiveWidth=3,NegativeWidth=0.5,PositiveHeight=3,NegativeHeight=0.1)
    ActiveRegionSize=0.1
    Debug=(bDrawDebug=false,DebugScale=25,DebugSegments=20,DebugThickness=0.25)
    DriverAxis=(X=-1,Y=0,Z=0)
    DriverItem=(Type=Bone,Name="thigh_twist_01_l")
    FalloffRegionScaleFactors=(PositiveWidth=1,NegativeWidth=2,PositiveHeight=5,NegativeHeight=1)
    FalloffSize=0.38
    OptionalParentItem=(Type=Bone,Name="pelvis")
    RotationOffset=(X=90,Y=0,Z=-35)
    ExecutePin=@SequenceAggregate_1.B
   )
   > &RigUnit_ParentConstraint(
    AdvancedSettings=(RotationOrderForFilter=XZY)
    Child=(Type=Bone,Name="dyn_Cloth_volume_l_A")
    Filter=(TranslationFilter=(bX=false,bY=false,bZ=false),RotationFilter=(bX=True,bY=True,bZ=True),ScaleFilter=(bX=True,bY=True,bZ=True))
    Parents=((Item=(Type=Bone,Name="thigh_twist_01_l"),Weight=0.5),(Item=(Type=Bone,Name="pelvis"),Weight=0.5))
    Weight=@n5.OutputParam
   )
   > &RigUnit_ParentConstraint(
    AdvancedSettings=(RotationOrderForFilter=XZY)
    Child=(Type=Bone,Name="dyn_Cloth_volume_l_B")
    Parents=((Item=(Type=Bone,Name="thigh_twist_01_l"),Weight=0.25),(Item=(Type=Bone,Name="pelvis"),Weight=0.75))
    Weight=@n5.OutputParam
   )
   > &RigUnit_ModifyTransforms(
    ItemToModify="((Item=(Type=Bone,Name=\"dyn_Cloth_volume_l_A\"),Transform=(Rotation=(X=0,Y=0,Z=0,W=1),Translation=(X=5,Y=-1,Z=-2),Scale3D=(X=1,Y=1,Z=1))),(Item=(Type=Bone,Name=\"dyn_Cloth_volume_l_B\"),Transform=(Rotation=(X=0,Y=0,Z=0,W=1),Translation=(X=0,Y=0,Z=0),Scale3D=( ...(truncated)"
    ItemToModify.2.Transform.Rotation=&RigVMFunction_MathQuaternionFromEuler(Euler=(X=0,Y=0,Z=-50)).Result
    ItemToModify.1.Transform.Rotation=&RigVMFunction_MathQuaternionFromEuler(Euler=(X=-40,Y=-40,Z=-20)).Result
    ItemToModify.0.Transform.Rotation=&RigVMFunction_MathQuaternionFromEuler(Euler=(X=12.5,Y=20,Z=60)).Result
    Weight=@n5.OutputParam
   )
   > n6 = &RigUnit_SphericalPoseReader(
    ActiveRegionScaleFactors=(PositiveWidth=5.5,NegativeWidth=4.5,PositiveHeight=2,NegativeHeight=2)
    ActiveRegionSize=0.1
    Debug=(bDrawDebug=false,DebugScale=25,DebugSegments=20,DebugThickness=0.25)
    DriverAxis=(X=-1,Y=0,Z=0)
    DriverItem=(Type=Bone,Name="thigh_l")
    FalloffRegionScaleFactors=(PositiveWidth=1,NegativeWidth=1.5,PositiveHeight=1,NegativeHeight=1)
    FalloffSize=0.7
    FlipWidthScaling=false
    OptionalParentItem=(Type=Bone)
    RotationOffset=(X=180,Y=50,Z=0)
   )
   > &RigUnit_ParentConstraint(
    AdvancedSettings=(RotationOrderForFilter=XZY)
    Child=(Type=Bone,Name="dyn_Cloth_volume_l_B")
    Filter=(TranslationFilter=(bX=false,bY=false,bZ=false),RotationFilter=(bX=True,bY=True,bZ=True),ScaleFilter=(bX=True,bY=True,bZ=True))
    Parents=((Item=(Type=Bone,Name="thigh_twist_01_l"),Weight=0.8),(Item=(Type=Bone,Name="pelvis"),Weight=0.2))
    Weight=@n6.OutputParam
   )
   > &RigUnit_ParentConstraint(
    AdvancedSettings=(RotationOrderForFilter=XZY)
    Child=(Type=Bone,Name="dyn_Cloth_volume_l_A")
    Filter=(TranslationFilter=(bX=false,bY=false,bZ=false),RotationFilter=(bX=True,bY=True,bZ=True),ScaleFilter=(bX=True,bY=True,bZ=True))
    Parents=((Item=(Type=Bone,Name="thigh_twist_01_l"),Weight=0.8),(Item=(Type=Bone,Name="pelvis"),Weight=0.2))
    Weight=@n6.OutputParam
   )
   > &RigUnit_ModifyTransforms(
    ItemToModify="((Item=(Type=Bone,Name=\"dyn_Cloth_volume_l_B\"),Transform=(Rotation=(X=0,Y=0,Z=0,W=1),Translation=(X=0,Y=0,Z=0),Scale3D=(X=1,Y=1,Z=1))),(Item=(Type=Bone,Name=\"dyn_Cloth_volume_l_A\"),Transform=(Rotation=(X=0,Y=0,Z=0,W=1),Translation=(X=0,Y=0,Z=0),Scale3D=(X= ...(truncated)"
    ItemToModify.2.Transform.Rotation=&RigVMFunction_MathQuaternionFromEuler(Euler=(X=0,Y=20,Z=0)).Result
    ItemToModify.1.Transform.Rotation=&RigVMFunction_MathQuaternionFromEuler(Euler=(X=0,Y=5,Z=5)).Result
    ItemToModify.0.Transform.Rotation=&RigVMFunction_MathQuaternionFromEuler(Euler=(X=0,Y=5,Z=5)).Result
    Weight=@n6.OutputParam
   )
   > n7 = &RigUnit_SphericalPoseReader(
    ActiveRegionScaleFactors=(PositiveWidth=3.5,NegativeWidth=0.1,PositiveHeight=0.5,NegativeHeight=2.3)
    ActiveRegionSize=0.1
    Debug=(bDrawDebug=false,DebugScale=25,DebugSegments=20,DebugThickness=0.25)
    DriverAxis=(X=-1,Y=0,Z=0)
    DriverItem=(Type=Bone,Name="thigh_l")
    FalloffRegionScaleFactors=(PositiveWidth=1,NegativeWidth=1.6,PositiveHeight=1,NegativeHeight=3)
    FalloffSize=0.3
    OptionalParentItem=(Type=Bone)
    RotationOffset=(X=-90,Y=0,Z=0)
   )
   > &RigUnit_ParentConstraint(
    AdvancedSettings=(RotationOrderForFilter=XZY)
    Child=(Type=Bone,Name="dyn_Cloth_volume_l_C")
    Filter=(TranslationFilter=(bX=false,bY=false,bZ=false),RotationFilter=(bX=True,bY=True,bZ=True),ScaleFilter=(bX=True,bY=True,bZ=True))
    Parents=((Item=(Type=Bone,Name="thigh_l"),Weight=0.5),(Item=(Type=Bone,Name="pelvis"),Weight=0.5))
    Weight=@n7.OutputParam
   )
   > &RigUnit_ModifyTransforms(
    ItemToModify="((Item=(Type=Bone,Name=\"dyn_Cloth_volume_l_C\"),Transform=(Rotation=(X=0,Y=0,Z=0,W=1),Translation=(X=0,Y=0,Z=0),Scale3D=(X=1,Y=1,Z=1))),(Item=(Type=Bone,Name=\"dyn_Cloth_volume_l_B\"),Transform=(Rotation=(X=0,Y=0,Z=0,W=1),Translation=(X=0,Y=0,Z=0),Scale3D=(X= ...(truncated)"
    ItemToModify.2.Transform.Rotation=&RigVMFunction_MathQuaternionFromEuler(Euler=(X=0,Y=20,Z=60)).Result
    ItemToModify.1.Transform.Rotation=&RigVMFunction_MathQuaternionFromEuler(Euler=(X=0,Y=0,Z=0)).Result
    ItemToModify.0.Transform.Rotation=&RigVMFunction_MathQuaternionFromEuler(Euler=(X=20,Y=0,Z=50)).Result
    Weight=@n7.OutputParam
   )
   > n8 = &RigUnit_SphericalPoseReader(
    ActiveRegionScaleFactors=(PositiveWidth=3.5,NegativeWidth=0.1,PositiveHeight=0.5,NegativeHeight=2.3)
    ActiveRegionSize=0.1
    Debug=(bDrawDebug=false,DebugScale=25,DebugSegments=20,DebugThickness=0.25)
    DriverAxis=(X=-1,Y=0,Z=0)
    DriverItem=(Type=Bone,Name="thigh_l")
    FalloffRegionScaleFactors=(PositiveWidth=1,NegativeWidth=1.5,PositiveHeight=1,NegativeHeight=3)
    FalloffSize=0.3
    OptionalParentItem=(Type=Bone)
    RotationOffset=(X=90,Y=0,Z=0)
   )
   > &RigUnit_ParentConstraint(
    AdvancedSettings=(RotationOrderForFilter=XZY)
    Child=(Type=Bone,Name="dyn_Cloth_volume_l_B")
    Filter=(TranslationFilter=(bX=false,bY=false,bZ=false),RotationFilter=(bX=True,bY=True,bZ=True),ScaleFilter=(bX=True,bY=True,bZ=True))
    Parents=((Item=(Type=Bone,Name="thigh_twist_01_l"),Weight=0.5),(Item=(Type=Bone,Name="pelvis"),Weight=0.5))
    Weight=@n8.OutputParam
   )
   > n9 = &RigUnit_SphericalPoseReader(
    ActiveRegionScaleFactors=(PositiveWidth=3,NegativeWidth=0.5,PositiveHeight=0.3,NegativeHeight=0.3)
    ActiveRegionSize=0.1
    Debug=(bDrawDebug=false,DebugScale=25,DebugSegments=20,DebugThickness=0.25)
    DriverAxis=(X=1,Y=0,Z=0)
    DriverItem=(Type=Bone,Name="spine_01")
    FalloffRegionScaleFactors=(PositiveWidth=1,NegativeWidth=1,PositiveHeight=1,NegativeHeight=1)
    FalloffSize=0.45
    FlipWidthScaling=false
    OptionalParentItem=(Type=Bone,Name="None")
    RotationOffset=(X=90,Y=0,Z=-10)
   )
   > &RigUnit_ParentConstraint(
    AdvancedSettings=(RotationOrderForFilter=XZY)
    Child=(Type=Bone,Name="dyn_Cloth_volume_l_C")
    Filter=(TranslationFilter=(bX=false,bY=false,bZ=false),RotationFilter=(bX=True,bY=True,bZ=True),ScaleFilter=(bX=True,bY=True,bZ=True))
    Parents=((Item=(Type=Bone,Name="thigh_twist_01_l"),Weight=0.1),(Item=(Type=Bone,Name="pelvis"),Weight=0.9))
    Weight=@n9.OutputParam
   )
   > &RigUnit_ParentConstraint(
    AdvancedSettings=(RotationOrderForFilter=XZY)
    Child=(Type=Bone,Name="dyn_Cloth_volume_r_C")
    Filter=(TranslationFilter=(bX=false,bY=false,bZ=false),RotationFilter=(bX=True,bY=True,bZ=True),ScaleFilter=(bX=True,bY=True,bZ=True))
    Parents=((Item=(Type=Bone,Name="thigh_r"),Weight=0.1),(Item=(Type=Bone,Name="pelvis"),Weight=0.9))
    Weight=@n9.OutputParam
   )
   > &RigUnit_ParentConstraint(
    AdvancedSettings=(RotationOrderForFilter=XZY)
    Child=(Type=Bone,Name="dyn_Cloth_volume_l_C01")
    Filter=(TranslationFilter=(bX=false,bY=false,bZ=false),RotationFilter=(bX=True,bY=True,bZ=True),ScaleFilter=(bX=True,bY=True,bZ=True))
    Parents=((Item=(Type=Bone,Name="root"),Weight=0),(Item=(Type=Bone,Name="pelvis"),Weight=0.1))
    bMaintainOffset=true
    Weight=@n9.OutputParam
   )
   > &RigUnit_ParentConstraint(
    AdvancedSettings=(RotationOrderForFilter=XZY)
    Child=(Type=Bone,Name="dyn_Cloth_volume_r_C01")
    Filter=(TranslationFilter=(bX=false,bY=false,bZ=false),RotationFilter=(bX=True,bY=True,bZ=True),ScaleFilter=(bX=True,bY=True,bZ=True))
    Parents=((Item=(Type=Bone,Name="root"),Weight=0),(Item=(Type=Bone,Name="pelvis"),Weight=0.1))
    Weight=@n9.OutputParam
   )
   n10 = &RigVMFunction_MathQuaternionFromEuler(Euler=(X=0,Y=0,Z=-10))
   n11 = &RigVMFunction_MathQuaternionFromEuler(Euler=(X=0,Y=0,Z=-25))
   n12 = &RigVMFunction_MathQuaternionFromEuler(Euler=(X=0,Y=0,Z=-10))
   > &RigUnit_ModifyTransforms(
    ItemToModify="((Item=(Type=Bone,Name=\"dyn_Cloth_volume_l_C\"),Transform=(Rotation=(X=0,Y=0,Z=0,W=1),Translation=(X=0,Y=0,Z=0),Scale3D=(X=1,Y=1,Z=1))),(Item=(Type=Bone,Name=\"dyn_Cloth_volume_r_C\"),Transform=(Rotation=(X=0,Y=0,Z=0,W=1),Translation=(X=0,Y=-1.2,Z=0),Scale3D= ...(truncated)"
    ItemToModify.2.Transform.Rotation=@n10.Result
    ItemToModify.3.Transform.Rotation=@n10.Result
    ItemToModify.4.Transform.Rotation=@n11.Result
    ItemToModify.5.Transform.Rotation=@n11.Result
    ItemToModify.0.Transform.Rotation=@n12.Result
    ItemToModify.1.Transform.Rotation=@n12.Result
    Weight=@n9.OutputParam
   )
   n13 = &RigUnit_SphericalPoseReader(
    ActiveRegionScaleFactors=(PositiveWidth=3,NegativeWidth=0.5,PositiveHeight=3,NegativeHeight=0.1)
    ActiveRegionSize=0.1
    Debug=(bDrawDebug=false,DebugScale=25,DebugSegments=20,DebugThickness=0.25)
    DriverAxis=(X=1,Y=0,Z=0)
    DriverItem=(Type=Bone,Name="thigh_twist_01_r")
    FalloffRegionScaleFactors=(PositiveWidth=1,NegativeWidth=2,PositiveHeight=5,NegativeHeight=1)
    FalloffSize=0.38
    FlipHeightScaling=false
    FlipWidthScaling=true
    OptionalParentItem=(Type=Bone,Name="pelvis")
    RotationOffset=(X=-90,Y=0,Z=-35)
    ExecutePin=@SequenceAggregate_1.C
   )
   > &RigUnit_ParentConstraint(
    AdvancedSettings=(RotationOrderForFilter=XZY)
    Child=(Type=Bone,Name="dyn_Cloth_volume_r_A")
    Filter=(TranslationFilter=(bX=false,bY=false,bZ=false),RotationFilter=(bX=True,bY=True,bZ=True),ScaleFilter=(bX=True,bY=True,bZ=True))
    Parents=((Item=(Type=Bone,Name="thigh_twist_01_r"),Weight=0.6),(Item=(Type=Bone,Name="pelvis"),Weight=0.4))
    Weight=@n13.OutputParam
   )
   > &RigUnit_ParentConstraint(
    AdvancedSettings=(RotationOrderForFilter=XZY)
    Child=(Type=Bone,Name="dyn_Cloth_volume_r_B")
    Parents=((Item=(Type=Bone,Name="thigh_twist_01_r"),Weight=0.25),(Item=(Type=Bone,Name="pelvis"),Weight=0.75))
    Weight=@n13.OutputParam
   )
   > &RigUnit_ModifyTransforms(
    ItemToModify="((Item=(Type=Bone,Name=\"dyn_Cloth_volume_r_A\"),Transform=(Rotation=(X=0,Y=0,Z=0,W=1),Translation=(X=4,Y=0.5,Z=-2),Scale3D=(X=1,Y=1,Z=1))),(Item=(Type=Bone,Name=\"dyn_Cloth_volume_r_B\"),Transform=(Rotation=(X=0,Y=0,Z=0,W=1),Translation=(X=0,Y=0,Z=0),Scale3D= ...(truncated)"
    ItemToModify.1.Transform.Rotation=&RigVMFunction_MathQuaternionFromEuler(Euler=(X=-20,Y=20,Z=20)).Result
    ItemToModify.2.Transform.Rotation=&RigVMFunction_MathQuaternionFromEuler(Euler=(X=0,Y=0,Z=50)).Result
    ItemToModify.0.Transform.Rotation=&RigVMFunction_MathQuaternionFromEuler(Euler=(X=12,Y=-20,Z=80)).Result
    Weight=@n13.OutputParam
   )
   > n14 = &RigUnit_SphericalPoseReader(
    ActiveRegionScaleFactors=(PositiveWidth=5.5,NegativeWidth=4.5,PositiveHeight=2,NegativeHeight=2)
    ActiveRegionSize=0.1
    Debug=(bDrawDebug=false,bDraw2D=false,DebugScale=25,DebugSegments=20,DebugThickness=0.25)
    DriverAxis=(X=1,Y=0,Z=0)
    DriverItem=(Type=Bone,Name="thigh_r")
    FalloffRegionScaleFactors=(PositiveWidth=1,NegativeWidth=1,PositiveHeight=1,NegativeHeight=1)
    FalloffSize=0.7
    FlipHeightScaling=false
    FlipWidthScaling=true
    OptionalParentItem=(Type=Bone,Name="None")
    RotationOffset=(X=0,Y=50,Z=0)
   )
   > &RigUnit_ParentConstraint(
    AdvancedSettings=(RotationOrderForFilter=XZY)
    Child=(Type=Bone,Name="dyn_Cloth_volume_r_B")
    Filter=(TranslationFilter=(bX=false,bY=false,bZ=false),RotationFilter=(bX=True,bY=True,bZ=True),ScaleFilter=(bX=True,bY=True,bZ=True))
    Parents=((Item=(Type=Bone,Name="thigh_twist_01_r"),Weight=0.8),(Item=(Type=Bone,Name="pelvis"),Weight=0.2))
    Weight=@n14.OutputParam
   )
   > &RigUnit_ParentConstraint(
    AdvancedSettings=(RotationOrderForFilter=XZY)
    Child=(Type=Bone,Name="dyn_Cloth_volume_r_A")
    Filter=(TranslationFilter=(bX=false,bY=false,bZ=false),RotationFilter=(bX=True,bY=True,bZ=True),ScaleFilter=(bX=True,bY=True,bZ=True))
    Parents=((Item=(Type=Bone,Name="thigh_twist_01_r"),Weight=0.8),(Item=(Type=Bone,Name="pelvis"),Weight=0.2))
    Weight=@n14.OutputParam
   )
   > &RigUnit_ModifyTransforms(
    ItemToModify="((Item=(Type=Bone,Name=\"dyn_Cloth_volume_r_B\"),Transform=(Rotation=(X=0,Y=0,Z=0,W=1),Translation=(X=5,Y=-5,Z=0),Scale3D=(X=1,Y=1,Z=1))),(Item=(Type=Bone,Name=\"dyn_Cloth_volume_r_A\"),Transform=(Rotation=(X=0,Y=0,Z=0,W=1),Translation=(X=0,Y=0,Z=0),Scale3D=(X ...(truncated)"
    ItemToModify.2.Transform.Rotation=&RigVMFunction_MathQuaternionFromEuler(Euler=(X=0,Y=20,Z=10)).Result
    ItemToModify.1.Transform.Rotation=&RigVMFunction_MathQuaternionFromEuler(Euler=(X=10,Y=-10,Z=15)).Result
    ItemToModify.0.Transform.Rotation=&RigVMFunction_MathQuaternionFromEuler(Euler=(X=0,Y=-10,Z=30)).Result
    Weight=@n14.OutputParam
   )
   > n15 = &RigUnit_SphericalPoseReader(
    ActiveRegionScaleFactors=(PositiveWidth=3.5,NegativeWidth=0.1,PositiveHeight=0.5,NegativeHeight=2.3)
    ActiveRegionSize=0.1
    Debug=(bDrawDebug=false,DebugScale=25,DebugSegments=20,DebugThickness=0.25)
    DriverAxis=(X=1,Y=0,Z=0)
    DriverItem=(Type=Bone,Name="thigh_r")
    FalloffRegionScaleFactors=(PositiveWidth=1,NegativeWidth=1.6,PositiveHeight=1,NegativeHeight=3)
    FalloffSize=0.3
    FlipHeightScaling=false
    FlipWidthScaling=true
    OptionalParentItem=(Type=Bone,Name="pelvis")
    RotationOffset=(X=90,Y=0,Z=-5)
   )
   > &RigUnit_ParentConstraint(
    AdvancedSettings=(RotationOrderForFilter=XZY)
    Child=(Type=Bone,Name="dyn_Cloth_volume_r_C")
    Filter=(TranslationFilter=(bX=false,bY=false,bZ=false),RotationFilter=(bX=True,bY=True,bZ=True),ScaleFilter=(bX=True,bY=True,bZ=True))
    Parents=((Item=(Type=Bone,Name="thigh_twist_01_r"),Weight=0.5),(Item=(Type=Bone,Name="pelvis"),Weight=0.5))
    Weight=@n15.OutputParam
   )
   > &RigUnit_ModifyTransforms(
    ItemToModify="((Item=(Type=Bone,Name=\"dyn_Cloth_volume_r_C\"),Transform=(Rotation=(X=0,Y=0,Z=0,W=1),Translation=(X=0,Y=0,Z=0),Scale3D=(X=1,Y=1,Z=1))),(Item=(Type=Bone,Name=\"dyn_Cloth_volume_r_B\"),Transform=(Rotation=(X=0,Y=0,Z=0,W=1),Translation=(X=0,Y=0,Z=0),Scale3D=(X= ...(truncated)"
    ItemToModify.1.Transform.Rotation=&RigVMFunction_MathQuaternionFromEuler(Euler=(X=30,Y=20,Z=0)).Result
    ItemToModify.2.Transform.Rotation=&RigVMFunction_MathQuaternionFromEuler(Euler=(X=0,Y=-20,Z=60)).Result
    ItemToModify.0.Transform.Rotation=&RigVMFunction_MathQuaternionFromEuler(Euler=(X=-20,Y=0,Z=50)).Result
    Weight=@n15.OutputParam
   )
   > n16 = &RigUnit_SphericalPoseReader(
    ActiveRegionScaleFactors=(PositiveWidth=3.5,NegativeWidth=0.1,PositiveHeight=0.4,NegativeHeight=0.4)
    ActiveRegionSize=0.1
    Debug=(bDrawDebug=false,DebugScale=25,DebugSegments=20,DebugThickness=0.25)
    DriverAxis=(X=1,Y=0,Z=0)
    DriverItem=(Type=Bone,Name="thigh_r")
    FalloffRegionScaleFactors=(PositiveWidth=1,NegativeWidth=1,PositiveHeight=3,NegativeHeight=1)
    FalloffSize=0.65
    FlipHeightScaling=false
    FlipWidthScaling=true
    OptionalParentItem=(Type=Bone,Name="pelvis")
    RotationOffset=(X=-150,Y=0,Z=-10)
   )
   > &RigUnit_ParentConstraint(
    AdvancedSettings=(RotationOrderForFilter=XZY)
    Child=(Type=Bone,Name="dyn_Cloth_volume_r_B")
    Filter=(TranslationFilter=(bX=false,bY=false,bZ=false),RotationFilter=(bX=True,bY=True,bZ=True),ScaleFilter=(bX=True,bY=True,bZ=True))
    Parents=((Item=(Type=Bone,Name="thigh_twist_01_r"),Weight=0.5),(Item=(Type=Bone,Name="pelvis"),Weight=0.5))
    Weight=@n16.OutputParam
   )
   > &RigUnit_ParentConstraint(
    AdvancedSettings=(RotationOrderForFilter=XZY)
    Child=(Type=Bone,Name="dyn_Cloth_volume_r_A")
    Filter=(TranslationFilter=(bX=false,bY=false,bZ=false),RotationFilter=(bX=True,bY=True,bZ=True),ScaleFilter=(bX=True,bY=True,bZ=True))
    Parents=((Item=(Type=Bone,Name="thigh_twist_01_r"),Weight=0.8),(Item=(Type=Bone,Name="pelvis"),Weight=0.2))
    Weight=0
   )
   > n17 = &RigUnit_SphericalPoseReader(
    ActiveRegionScaleFactors=(PositiveWidth=1,NegativeWidth=1,PositiveHeight=1,NegativeHeight=1)
    ActiveRegionSize=0.1
    Debug=(bDrawDebug=false,bDraw2D=false,DebugScale=25,DebugSegments=20,DebugThickness=0.25)
    DriverAxis=(X=1,Y=0,Z=0)
    DriverItem=(Type=Bone,Name="thigh_r")
    FalloffRegionScaleFactors=(PositiveWidth=1,NegativeWidth=1,PositiveHeight=1,NegativeHeight=1)
    FalloffSize=0.28
    FlipHeightScaling=false
    FlipWidthScaling=true
    OptionalParentItem=(Type=Bone,Name="None")
    RotationOffset=(X=-90,Y=0,Z=-80)
   )
   > &RigUnit_ParentConstraint(
    AdvancedSettings=(RotationOrderForFilter=XZY)
    Child=(Type=Bone,Name="dyn_Cloth_volume_r_B")
    Filter=(TranslationFilter=(bX=true,bY=true,bZ=true),RotationFilter=(bX=True,bY=True,bZ=True),ScaleFilter=(bX=True,bY=True,bZ=True))
    Parents=((Item=(Type=Bone,Name="thigh_twist_01_r"),Weight=0.5),(Item=(Type=Bone,Name="pelvis"),Weight=0.5))
    Weight=@n17.OutputParam
   )
   > &RigUnit_ParentConstraint(
    AdvancedSettings=(RotationOrderForFilter=XZY)
    Child=(Type=Bone,Name="dyn_Cloth_volume_r_A")
    Filter=(TranslationFilter=(bX=true,bY=true,bZ=true),RotationFilter=(bX=True,bY=True,bZ=True),ScaleFilter=(bX=True,bY=True,bZ=True))
    Parents=((Item=(Type=Bone,Name="thigh_twist_01_r"),Weight=0.8),(Item=(Type=Bone,Name="pelvis"),Weight=0.2))
    Weight=@n17.OutputParam
   )
   n18 = &RigVMFunction_MathQuaternionFromEuler(Euler=(X=0,Y=0,Z=50))
   > &RigUnit_ModifyTransforms(
    ItemToModify="((Item=(Type=Bone,Name=\"dyn_Cloth_volume_r_B\"),Transform=(Rotation=(X=0,Y=0,Z=0,W=1),Translation=(X=0,Y=0,Z=0),Scale3D=(X=1,Y=1,Z=1))),(Item=(Type=Bone,Name=\"dyn_Cloth_volume_r_A\"),Transform=(Rotation=(X=0,Y=0,Z=0,W=1),Translation=(X=0,Y=0,Z=0),Scale3D=(X= ...(truncated)"
    ItemToModify.0.Transform.Rotation=@n18.Result
    ItemToModify.1.Transform.Rotation=@n18.Result
    Weight=@n17.OutputParam
   )
   > n19 = &RigUnit_SphericalPoseReader(
    ActiveRegionScaleFactors=(PositiveWidth=3,NegativeWidth=0.1,PositiveHeight=0.3,NegativeHeight=0.3)
    ActiveRegionSize=0.1
    Debug=(bDrawDebug=false,DebugScale=25,DebugSegments=20,DebugThickness=0.25)
    DriverAxis=(X=1,Y=0,Z=0)
    DriverItem=(Type=Bone,Name="pelvis")
    FalloffRegionScaleFactors=(PositiveWidth=1,NegativeWidth=1,PositiveHeight=1,NegativeHeight=1)
    FalloffSize=0.45
    FlipWidthScaling=false
    OptionalParentItem=(Type=Bone,Name="root")
    RotationOffset=(X=-90,Y=0,Z=-10)
   )
   n20 = &RigVMFunction_MathQuaternionFromEuler(Euler=(X=0,Y=0,Z=90))
   > &RigUnit_ModifyTransforms(
    ItemToModify="((Item=(Type=Bone,Name=\"dyn_Cloth_volume_r_C\"),Transform=(Rotation=(X=0,Y=0,Z=0,W=1),Translation=(X=0,Y=0,Z=0),Scale3D=(X=1,Y=1,Z=1))),(Item=(Type=Bone,Name=\"dyn_Cloth_volume_l_C\"),Transform=(Rotation=(X=0,Y=0,Z=0,W=1),Translation=(X=0,Y=0,Z=0),Scale3D=(X= ...(truncated)"
    ItemToModify.0.Transform.Rotation=@n20.Result
    ItemToModify.1.Transform.Rotation=@n20.Result
    Weight=@n19.OutputParam
   )
   > n21 = &RigUnit_SphericalPoseReader(
    ActiveRegionScaleFactors=(PositiveWidth=3.5,NegativeWidth=0.1,PositiveHeight=3,NegativeHeight=3)
    ActiveRegionSize=0.1
    Debug=(bDrawDebug=false,DebugScale=25,DebugSegments=20,DebugThickness=0.25)
    DriverAxis=(X=1,Y=0.05,Z=0)
    DriverItem=(Type=Bone,Name="calf_r")
    FalloffRegionScaleFactors=(PositiveWidth=1,NegativeWidth=1,PositiveHeight=2,NegativeHeight=2)
    FalloffSize=0.35
    FlipHeightScaling=false
    FlipWidthScaling=true
    OptionalParentItem=(Type=Bone,Name="thigh_r")
    RotationOffset=(X=90,Y=0,Z=70)
   )
   n22 = &RigVMFunction_MathQuaternionFromEuler(Euler=(X=0,Y=10,Z=30))
   > &RigUnit_ModifyTransforms(
    ItemToModify="((Item=(Type=Bone,Name=\"dyn_Cloth_volume_r_C\"),Transform=(Rotation=(X=0,Y=0,Z=0,W=1),Translation=(X=1,Y=-1,Z=0),Scale3D=(X=1,Y=1,Z=1))),(Item=(Type=Bone,Name=\"dyn_Cloth_volume_l_C\"),Transform=(Rotation=(X=0,Y=0,Z=0,W=1),Translation=(X=1,Y=-1,Z=0),Scale3D=( ...(truncated)"
    ItemToModify.0.Transform.Rotation=@n22.Result
    ItemToModify.1.Transform.Rotation=@n22.Result
    Weight=@n21.OutputParam
   )
   &RigUnit_ParentConstraint(
    AdvancedSettings=(RotationOrderForFilter=XZY)
    Child=(Type=Bone,Name="dyn_Cloth_volume_fn")
    Parents=((Item=(Type=Bone,Name="dyn_Cloth_volume_l_A"),Weight=0.5),(Item=(Type=Bone,Name="dyn_Cloth_volume_r_A"),Weight=0.5))
    Weight=1
    ExecutePin=@SequenceAggregate_1.D
   )
   > &RigUnit_ParentConstraint(
    AdvancedSettings=(RotationOrderForFilter=XZY)
    Child=(Type=Bone,Name="dyn_Cloth_volume_fn01")
    Parents=((Item=(Type=Bone,Name="dyn_Cloth_volume_l_A01"),Weight=0.5),(Item=(Type=Bone,Name="dyn_Cloth_volume_r_A01"),Weight=0.5))
    Weight=1
   )
   > &RigUnit_ParentConstraint(
    AdvancedSettings=(RotationOrderForFilter=XZY)
    Child=(Type=Bone,Name="dyn_Cloth_volume_fn02")
    Parents=((Item=(Type=Bone,Name="dyn_Cloth_volume_l_A02"),Weight=0.5),(Item=(Type=Bone,Name="dyn_Cloth_volume_r_A02"),Weight=0.5))
    Weight=1
   )
   > &RigUnit_ParentConstraint(
    AdvancedSettings=(RotationOrderForFilter=XZY)
    Child=(Type=Bone,Name="dyn_Cloth_volume_bk")
    Parents=((Item=(Type=Bone,Name="dyn_Cloth_volume_l_C"),Weight=0.5),(Item=(Type=Bone,Name="dyn_Cloth_volume_r_C"),Weight=0.5))
    Weight=1
   )
   > &RigUnit_ParentConstraint(
    AdvancedSettings=(RotationOrderForFilter=XZY)
    Child=(Type=Bone,Name="dyn_Cloth_volume_bk01")
    Parents=((Item=(Type=Bone,Name="dyn_Cloth_volume_l_C01"),Weight=0.5),(Item=(Type=Bone,Name="dyn_Cloth_volume_r_C01"),Weight=0.5))
    Weight=1
   )
   > &RigUnit_ParentConstraint(
    AdvancedSettings=(RotationOrderForFilter=XZY)
    Child=(Type=Bone,Name="dyn_Cloth_volume_bk02")
    Parents=((Item=(Type=Bone,Name="dyn_Cloth_volume_l_C02"),Weight=0.5),(Item=(Type=Bone,Name="dyn_Cloth_volume_r_C02"),Weight=0.5))
    Weight=1
   )
   > n23 = &RigUnit_SphericalPoseReader(
    ActiveRegionScaleFactors=(PositiveWidth=3.5,NegativeWidth=1,PositiveHeight=3,NegativeHeight=3)
    ActiveRegionSize=0.1
    Debug=(bDrawDebug=false,DebugScale=25,DebugSegments=20,DebugThickness=0.25)
    DriverAxis=(X=-1,Y=0,Z=0)
    DriverItem=(Type=Bone,Name="thigh_l")
    FalloffRegionScaleFactors=(PositiveWidth=1,NegativeWidth=2,PositiveHeight=3,NegativeHeight=1)
    FalloffSize=0.35
    OptionalParentItem=(Type=Bone)
    RotationOffset=(X=90,Y=0,Z=-35)
   )
   > &RigUnit_ParentConstraint(
    AdvancedSettings=(RotationOrderForFilter=XZY)
    Child=(Type=Bone,Name="dyn_Cloth_volume_fn")
    Filter=(TranslationFilter=(bX=false,bY=false,bZ=false),RotationFilter=(bX=True,bY=True,bZ=True),ScaleFilter=(bX=True,bY=True,bZ=True))
    Parents=((Item=(Type=Bone,Name="thigh_twist_01_l"),Weight=0.5),(Item=(Type=Bone,Name="pelvis"),Weight=0.5))
    Weight=@n23.OutputParam
   )
   > &RigUnit_ModifyTransforms(
    ItemToModify="((Item=(Type=Bone,Name=\"dyn_Cloth_volume_fn01\"),Transform=(Rotation=(X=0,Y=0,Z=0,W=1),Translation=(X=0,Y=0,Z=0),Scale3D=(X=1,Y=1,Z=1))),(Item=(Type=Bone,Name=\"dyn_Cloth_volume_fn_float\"),Transform=(Rotation=(X=0,Y=0,Z=0,W=1),Translation=(X=0,Y=0,Z=0),Scale ...(truncated)"
    ItemToModify.0.Transform.Rotation=&RigVMFunction_MathQuaternionFromEuler(Euler=(X=20,Y=10,Z=15)).Result
    ItemToModify.1.Transform.Rotation=&RigVMFunction_MathQuaternionFromEuler(Euler=(X=20,Y=0,Z=-10)).Result
    ItemToModify.2.Transform.Rotation=&RigVMFunction_MathQuaternionFromEuler(Euler=(X=0,Y=0,Z=10)).Result
    Weight=@n23.OutputParam
   )
   > n24 = &RigUnit_SphericalPoseReader(
    ActiveRegionScaleFactors=(PositiveWidth=3.5,NegativeWidth=0.1,PositiveHeight=3,NegativeHeight=3)
    ActiveRegionSize=0.1
    Debug=(bDrawDebug=false,DebugScale=25,DebugSegments=20,DebugThickness=0.25)
    DriverAxis=(X=1,Y=0,Z=0)
    DriverItem=(Type=Bone,Name="thigh_twist_01_r")
    FalloffRegionScaleFactors=(PositiveWidth=1,NegativeWidth=2.3,PositiveHeight=2,NegativeHeight=2)
    FalloffSize=0.35
    FlipHeightScaling=false
    FlipWidthScaling=true
    OptionalParentItem=(Type=Bone,Name="pelvis")
    RotationOffset=(X=-90,Y=0,Z=-25)
   )
   > &RigUnit_ParentConstraint(
    AdvancedSettings=(RotationOrderForFilter=XZY)
    Child=(Type=Bone,Name="dyn_Cloth_volume_fn")
    Filter=(TranslationFilter=(bX=false,bY=false,bZ=false),RotationFilter=(bX=True,bY=True,bZ=True),ScaleFilter=(bX=True,bY=True,bZ=True))
    Parents=((Item=(Type=Bone,Name="thigh_twist_01_r"),Weight=0.7),(Item=(Type=Bone,Name="pelvis"),Weight=0.3))
    Weight=@n24.OutputParam
   )
   > &RigUnit_ModifyTransforms(
    ItemToModify="((Item=(Type=Bone,Name=\"dyn_Cloth_volume_fn\"),Transform=(Rotation=(X=0,Y=0,Z=0,W=1),Translation=(X=0,Y=0,Z=0),Scale3D=(X=1,Y=1,Z=1))),(Item=(Type=Bone,Name=\"dyn_Cloth_volume_fn_float\"),Transform=(Rotation=(X=0,Y=0,Z=0,W=1),Translation=(X=0,Y=0,Z=0),Scale3D ...(truncated)"
    ItemToModify.0.Transform.Rotation=&RigVMFunction_MathQuaternionFromEuler(Euler=(X=-20,Y=-10,Z=15)).Result
    ItemToModify.1.Transform.Rotation=&RigVMFunction_MathQuaternionFromEuler(Euler=(X=-20,Y=0,Z=-10)).Result
    ItemToModify.2.Transform.Rotation=&RigVMFunction_MathQuaternionFromEuler(Euler=(X=0,Y=20,Z=-10)).Result
    Weight=@n24.OutputParam
   )
   > n25 = &RigUnit_SphericalPoseReader(
    ActiveRegionScaleFactors=(PositiveWidth=3.5,NegativeWidth=0.1,PositiveHeight=3,NegativeHeight=1)
    ActiveRegionSize=0.2
    Debug=(bDrawDebug=false,DebugScale=25,DebugSegments=20,DebugThickness=0.25)
    DriverAxis=(X=-1,Y=0,Z=0)
    DriverItem=(Type=Bone,Name="thigh_l")
    FalloffRegionScaleFactors=(PositiveWidth=1,NegativeWidth=1,PositiveHeight=1,NegativeHeight=5)
    FalloffSize=0.25
    FlipWidthScaling=false
    OptionalParentItem=(Type=Bone,Name="None")
    RotationOffset=(X=150,Y=100,Z=0)
   )
   > &RigUnit_ParentConstraint(
    AdvancedSettings=(RotationOrderForFilter=XZY)
    Child=(Type=Bone,Name="dyn_Cloth_volume_fn")
    Filter=(TranslationFilter=(bX=false,bY=false,bZ=false),RotationFilter=(bX=True,bY=True,bZ=True),ScaleFilter=(bX=True,bY=True,bZ=True))
    Parents=((Item=(Type=Bone,Name="thigh_twist_01_l"),Weight=0.6),(Item=(Type=Bone,Name="pelvis"),Weight=0.4))
    Weight=@n25.OutputParam
   )
   > &RigUnit_ModifyTransforms(
    ItemToModify="((Item=(Type=Bone,Name=\"dyn_Cloth_volume_fn01\"),Transform=(Rotation=(X=0,Y=0,Z=0,W=1),Translation=(X=0,Y=0,Z=0),Scale3D=(X=1,Y=1,Z=1))),(Item=(Type=Bone,Name=\"None\"),Transform=(Rotation=(X=0,Y=0,Z=0,W=1),Translation=(X=2,Y=-2,Z=0),Scale3D=(X=1,Y=1,Z=1))),( ...(truncated)"
    ItemToModify.0.Transform.Rotation=&RigVMFunction_MathQuaternionFromEuler(Euler=(X=20,Y=20,Z=20)).Result
    Weight=@n25.OutputParam
   )
   > n26 = &RigUnit_SphericalPoseReader(
    ActiveRegionScaleFactors=(PositiveWidth=3.5,NegativeWidth=0.1,PositiveHeight=3,NegativeHeight=1)
    ActiveRegionSize=0.2
    Debug=(bDrawDebug=false,DebugScale=25,DebugSegments=20,DebugThickness=0.25)
    DriverAxis=(X=1,Y=0,Z=0)
    DriverItem=(Type=Bone,Name="thigh_r")
    FalloffRegionScaleFactors=(PositiveWidth=1,NegativeWidth=1,PositiveHeight=1,NegativeHeight=5)
    FalloffSize=0.25
    FlipHeightScaling=false
    FlipWidthScaling=true
    OptionalParentItem=(Type=Bone,Name="None")
    RotationOffset=(X=-150,Y=85,Z=-120)
   )
   > &RigUnit_ParentConstraint(
    AdvancedSettings=(RotationOrderForFilter=XZY)
    Child=(Type=Bone,Name="dyn_Cloth_volume_fn")
    Filter=(TranslationFilter=(bX=false,bY=false,bZ=false),RotationFilter=(bX=True,bY=True,bZ=True),ScaleFilter=(bX=True,bY=True,bZ=True))
    Parents=((Item=(Type=Bone,Name="thigh_twist_01_r"),Weight=0.6),(Item=(Type=Bone,Name="pelvis"),Weight=0.4))
    Weight=@n26.OutputParam
   )
   > &RigUnit_ModifyTransforms(
    ItemToModify="((Item=(Type=Bone,Name=\"dyn_Cloth_volume_fn\"),Transform=(Rotation=(X=0,Y=0,Z=0,W=1),Translation=(X=0,Y=0,Z=0),Scale3D=(X=1,Y=1,Z=1))),(Item=(Type=Bone,Name=\"None\"),Transform=(Rotation=(X=0,Y=0,Z=0,W=1),Translation=(X=0,Y=0,Z=0),Scale3D=(X=1,Y=1,Z=1))),(Ite ...(truncated)"
    ItemToModify.0.Transform.Rotation=&RigVMFunction_MathQuaternionFromEuler(Euler=(X=-30,Y=-30,Z=30)).Result
    Weight=@n26.OutputParam
   )
   > &RigUnit_ParentConstraint(
    AdvancedSettings=(RotationOrderForFilter=XZY)
    Child=(Type=Bone,Name="dyn_Cloth_volume_bk01")
    Parents=((Item=(Type=Bone,Name="dyn_Cloth_volume_l_C01"),Weight=0.5),(Item=(Type=Bone,Name="dyn_Cloth_volume_r_C01"),Weight=0.5))
    Weight=1
   )
   > &RigUnit_ParentConstraint(
    AdvancedSettings=(RotationOrderForFilter=XZY)
    Child=(Type=Bone,Name="dyn_Cloth_volume_bk02")
    Parents=((Item=(Type=Bone,Name="dyn_Cloth_volume_l_C02"),Weight=0.5),(Item=(Type=Bone,Name="dyn_Cloth_volume_r_C02"),Weight=0.5))
    Weight=1
   )
   > &RigUnit_ParentConstraint(
    AdvancedSettings=(RotationOrderForFilter=XZY)
    Child=(Type=Bone,Name="dyn_Cloth_volume_l_A_float")
    Parents=((Item=(Type=Bone,Name="dyn_Cloth_volume_l_A02"),Weight=0.4),(Item=(Type=Bone,Name="dyn_Cloth_volume_l_B02"),Weight=0.3),(Item=(Type=Bone,Name="dyn_Cloth_volume_fn02"),Weight=0.3))
    Weight=0.5
   )
   > &RigUnit_ParentConstraint(
    AdvancedSettings=(RotationOrderForFilter=XZY)
    Child=(Type=Bone,Name="dyn_Cloth_volume_l_B_float")
    Parents=((Item=(Type=Bone,Name="dyn_Cloth_volume_l_B02"),Weight=0.4),(Item=(Type=Bone,Name="dyn_Cloth_volume_l_A02"),Weight=0.3),(Item=(Type=Bone,Name="dyn_Cloth_volume_l_C02"),Weight=0.3))
    Weight=0.5
   )
   > &RigUnit_ParentConstraint(
    AdvancedSettings=(RotationOrderForFilter=XZY)
    Child=(Type=Bone,Name="dyn_Cloth_volume_l_C_float01")
    Parents=((Item=(Type=Bone,Name="dyn_Cloth_volume_l_C02"),Weight=0.4),(Item=(Type=Bone,Name="dyn_Cloth_volume_l_B02"),Weight=0.3),(Item=(Type=Bone,Name="dyn_Cloth_volume_bk02"),Weight=0.3))
    Weight=0.5
   )
   > &RigUnit_ParentConstraint(
    AdvancedSettings=(RotationOrderForFilter=XZY)
    Child=(Type=Bone,Name="dyn_Croth_vorume_r_A_float")
    Parents=((Item=(Type=Bone,Name="dyn_Cloth_volume_r_A02"),Weight=0.4),(Item=(Type=Bone,Name="dyn_Cloth_volume_r_B02"),Weight=0.3),(Item=(Type=Bone,Name="dyn_Cloth_volume_fn02"),Weight=0.3))
    Weight=0.5
   )
   > &RigUnit_ParentConstraint(
    AdvancedSettings=(RotationOrderForFilter=XZY)
    Child=(Type=Bone,Name="dyn_Croth_vorume_r_B_float")
    Parents=((Item=(Type=Bone,Name="dyn_Cloth_volume_r_B02"),Weight=0.4),(Item=(Type=Bone,Name="dyn_Cloth_volume_r_A02"),Weight=0.3),(Item=(Type=Bone,Name="dyn_Cloth_volume_r_C02"),Weight=0.3))
    Weight=0.5
   )
   > &RigUnit_ParentConstraint(
    AdvancedSettings=(RotationOrderForFilter=XZY)
    Child=(Type=Bone,Name="dyn_Croth_vorume_r_C_float01")
    Parents=((Item=(Type=Bone,Name="dyn_Cloth_volume_r_C02"),Weight=0.4),(Item=(Type=Bone,Name="dyn_Cloth_volume_r_B02"),Weight=0.3),(Item=(Type=Bone,Name="dyn_Cloth_volume_bk02"),Weight=0.3))
    Weight=0.5
   )
   > &RigUnit_ParentConstraint(
    AdvancedSettings=(RotationOrderForFilter=XZY)
    Child=(Type=Bone,Name="dyn_Croth_vorume_r_C_float02")
    Filter=(TranslationFilter=(bX=true,bY=true,bZ=true),RotationFilter=(bX=True,bY=True,bZ=True),ScaleFilter=(bX=True,bY=True,bZ=True))
    Parents=((Item=(Type=Bone,Name="dyn_Cloth_volume_r_C02"),Weight=0.3),(Item=(Type=Bone,Name="dyn_Cloth_volume_l_C02"),Weight=0.1),(Item=(Type=Bone,Name="dyn_Cloth_volume_bk02"),Weight=0.6))
    Weight=0.5
   )
   > &RigUnit_ParentConstraint(
    AdvancedSettings=(RotationOrderForFilter=XZY)
    Child=(Type=Bone,Name="dyn_Cloth_volume_l_C_float02")
    Filter=(TranslationFilter=(bX=true,bY=true,bZ=true),RotationFilter=(bX=True,bY=True,bZ=True),ScaleFilter=(bX=True,bY=True,bZ=True))
    Parents=((Item=(Type=Bone,Name="dyn_Cloth_volume_r_C02"),Weight=0.1),(Item=(Type=Bone,Name="dyn_Cloth_volume_l_C02"),Weight=0.3),(Item=(Type=Bone,Name="dyn_Cloth_volume_bk02"),Weight=0.6))
    Weight=0.5
   )
   SequenceAggregate_2_1 = RigVMAggregateNode(ExecuteContext=@SequenceAggregate_2.B)
   n27 = &RigUnit_SphericalPoseReader(
    ActiveRegionScaleFactors=(PositiveWidth=1,NegativeWidth=1,PositiveHeight=1,NegativeHeight=1)
    ActiveRegionSize=0.75
    Debug=(bDrawDebug=true,DebugScale=25,DebugSegments=20,DebugThickness=0.25)
    DriverAxis=(X=1,Y=0,Z=0)
    DriverItem=(Type=Bone,Name="hand_l")
    FalloffRegionScaleFactors=(PositiveWidth=1,NegativeWidth=1,PositiveHeight=1,NegativeHeight=1)
    FalloffSize=1
    FlipHeightScaling=false
    FlipWidthScaling=false
    OptionalParentItem=(Type=Bone,Name="lowerarm_l")
    RotationOffset=(X=0,Y=90,Z=0)
    ExecutePin=@SequenceAggregate_2_1.A
   )
   Multiply = &RigVMFunction_MathFloatMul(B=0.5, A=@n27.OutputParam) {template="Multiply::Execute(in A,in B,out Result)"}
   > &RigUnit_ModifyTransforms(
    ItemToModify=((Item=(Type=Bone,Name="dyn_l_wrist_guard"),Transform=(Rotation=(X=0,Y=0,Z=0,W=1),Translation=(X=-6,Y=0,Z=1.8),Scale3D=(X=1,Y=1,Z=1))))
    Weight=@Multiply.Result
   )
   > &RigUnit_ModifyTransforms(
    ItemToModify=((Item=(Type=Bone,Name="dyn_l_wrist_guard"),Transform=(Rotation=(X=0,Y=0,Z=0,W=1),Translation=(X=0,Y=0,Z=0),Scale3D=(X=1,Y=1,Z=1))))
    ItemToModify.0.Transform.Rotation=&RigVMFunction_MathQuaternionFromEuler(Euler=(X=5,Y=0,Z=20)).Result
    Weight=@n27.OutputParam
   )
   > &RigUnit_ModifyTransforms(
    ItemToModify=((Item=(Type=Bone,Name="dyn_l_wrist_guard_01"),Transform=(Rotation=(X=0,Y=0,Z=0,W=1),Translation=(X=0,Y=0,Z=1.2),Scale3D=(X=1,Y=1,Z=1))))
    ItemToModify.0.Transform.Rotation=&RigVMFunction_MathQuaternionFromEuler(Euler=(X=0,Y=-20,Z=0)).Result
    Weight=@n27.OutputParam
   )
   > n28 = &RigUnit_SphericalPoseReader(
    ActiveRegionScaleFactors=(PositiveWidth=1,NegativeWidth=1,PositiveHeight=1,NegativeHeight=1)
    ActiveRegionSize=0.75
    Debug=(bDrawDebug=true,DebugScale=25,DebugSegments=20,DebugThickness=0.25)
    DriverAxis=(X=1,Y=0,Z=0)
    DriverItem=(Type=Bone,Name="hand_r")
    FalloffRegionScaleFactors=(PositiveWidth=1,NegativeWidth=1,PositiveHeight=1,NegativeHeight=1)
    FalloffSize=1
    FlipHeightScaling=false
    FlipWidthScaling=false
    OptionalParentItem=(Type=Bone,Name="lowerarm_r")
    RotationOffset=(X=0,Y=90,Z=0)
   )
   > &RigUnit_ModifyTransforms(
    ItemToModify=((Item=(Type=Bone,Name="dyn_r_wrist_guard"),Transform=(Rotation=(X=0,Y=0,Z=0,W=1),Translation=(X=0,Y=0,Z=0),Scale3D=(X=1,Y=1,Z=1))))
    ItemToModify.0.Transform.Rotation=&RigVMFunction_MathQuaternionFromEuler(Euler=(X=5,Y=-5,Z=20)).Result
    Weight=@n28.OutputParam
   )
   > &RigUnit_ModifyTransforms(
    ItemToModify=((Item=(Type=Bone,Name="dyn_r_wrist_guard"),Transform=(Rotation=(X=0,Y=0,Z=0,W=1),Translation=(X=-6,Y=0,Z=1.5),Scale3D=(X=1,Y=1,Z=1))))
    Weight=&RigVMFunction_MathFloatMul(B=0.45, A=@n28.OutputParam) {template="Multiply::Execute(in A,in B,out Result)"}.Result
   )
   > &RigUnit_ModifyTransforms(
    ItemToModify=((Item=(Type=Bone,Name="dyn_r_wrist_guard_01"),Transform=(Rotation=(X=0,Y=0,Z=0,W=1),Translation=(X=0,Y=0,Z=1.2),Scale3D=(X=1,Y=1,Z=1))))
    ItemToModify.0.Transform.Rotation=&RigVMFunction_MathQuaternionFromEuler(Euler=(X=0,Y=-20,Z=0)).Result
    Weight=@n28.OutputParam
   )
   n29 = &RigUnit_SphericalPoseReader(
    ActiveRegionScaleFactors=(PositiveWidth=1.5,NegativeWidth=1.5,PositiveHeight=2,NegativeHeight=2)
    ActiveRegionSize=0.1
    Debug=(bDrawDebug=true,bDraw2D=false,bDrawLocalAxes=false,DebugScale=25,DebugSegments=20,DebugThickness=0.25)
    DriverAxis=(X=-1,Y=0,Z=0)
    DriverItem=(Type=Bone,Name="upperarm_r")
    FalloffRegionScaleFactors=(PositiveWidth=1,NegativeWidth=1,PositiveHeight=2,NegativeHeight=2)
    FalloffSize=0.35
    OptionalParentItem=(Type=Bone,Name="spine_05")
    RotationOffset=(X=-60,Y=30,Z=40)
    ExecutePin=@SequenceAggregate_2_1.B
   )
   > &RigUnit_ModifyTransforms(
    ItemToModify=((Item=(Type=Bone,Name="dyn_r_shoulder_fr"),Transform=(Rotation=(X=0,Y=0,Z=0,W=1),Translation=(X=0,Y=-4,Z=5),Scale3D=(X=1,Y=1,Z=1))))
    ItemToModify.0.Transform.Rotation=&RigVMFunction_MathQuaternionFromEuler(Euler=(X=-30,Y=-20,Z=0)).Result
    Weight=@n29.OutputParam
   )
   n30 = RigVMVariableNode() {variable=Dual_Pistols_Twist}
   > &RigUnit_ModifyTransforms(
    ItemToModify=((Item=(Type=Bone,Name="dyn_r_shoulder_fr"),Transform=(Rotation=(X=0,Y=0,Z=0,W=1),Translation=(X=-5,Y=-1,Z=5),Scale3D=(X=1,Y=1,Z=1))))
    WeightMinimum=0
    ItemToModify.0.Transform.Rotation=&RigVMFunction_MathQuaternionFromEuler(Euler=(X=5,Y=-10,Z=-20)).Result
    Weight=@n30.Value
   )
   > &RigUnit_ModifyTransforms(
    ItemToModify=((Item=(Type=Bone,Name="dyn_r_shoulder"),Transform=(Rotation=(X=0,Y=0,Z=0,W=1),Translation=(X=1,Y=0,Z=0),Scale3D=(X=1,Y=1,Z=1))))
    WeightMinimum=0
    ItemToModify.0.Transform.Rotation=&RigVMFunction_MathQuaternionFromEuler(Euler=(X=20,Y=-10,Z=30)).Result
    Weight=@n30.Value
   )
   > n31 = &RigUnit_SphericalPoseReader(
    ActiveRegionScaleFactors=(PositiveWidth=1.5,NegativeWidth=1.5,PositiveHeight=2,NegativeHeight=2)
    ActiveRegionSize=0.1
    Debug=(bDrawDebug=false,bDraw2D=false,bDrawLocalAxes=false,DebugScale=25,DebugSegments=20,DebugThickness=0.25)
    DriverAxis=(X=-1,Y=0,Z=0)
    DriverItem=(Type=Bone,Name="upperarm_r")
    FalloffRegionScaleFactors=(PositiveWidth=1,NegativeWidth=1,PositiveHeight=2,NegativeHeight=2)
    FalloffSize=0.35
    OptionalParentItem=(Type=Bone,Name="spine_05")
    RotationOffset=(X=-90,Y=30,Z=40)
   )
   > &RigUnit_ModifyTransforms(
    ItemToModify=((Item=(Type=Bone,Name="dyn_r_shoulder"),Transform=(Rotation=(X=0,Y=0,Z=0,W=1),Translation=(X=-0.75,Y=0.5,Z=0),Scale3D=(X=1,Y=1,Z=1))))
    ItemToModify.0.Transform.Rotation=&RigVMFunction_MathQuaternionFromEuler(Euler=(X=30,Y=20,Z=0)).Result
    Weight=@n31.OutputParam
   )
   > n32 = &RigUnit_SphericalPoseReader(
    ActiveRegionScaleFactors=(PositiveWidth=1.5,NegativeWidth=1.5,PositiveHeight=2,NegativeHeight=2)
    ActiveRegionSize=0.1
    Debug=(bDrawDebug=false,bDraw2D=false,bDrawLocalAxes=false,DebugScale=25,DebugSegments=20,DebugThickness=0.25)
    DriverAxis=(X=-1,Y=0,Z=0)
    DriverItem=(Type=Bone,Name="upperarm_l")
    FalloffRegionScaleFactors=(PositiveWidth=1,NegativeWidth=1.6,PositiveHeight=2,NegativeHeight=2)
    FalloffSize=0.35
    FlipHeightScaling=false
    FlipWidthScaling=false
    OptionalParentItem=(Type=Bone,Name="spine_05")
    RotationOffset=(X=-60,Y=30,Z=40)
   )
   > &RigUnit_ModifyTransforms(
    ItemToModify=((Item=(Type=Bone,Name="dyn_l_shoulder_fr"),Transform=(Rotation=(X=0,Y=0,Z=0,W=1),Translation=(X=0,Y=4,Z=5),Scale3D=(X=1,Y=1,Z=1))))
    ItemToModify.0.Transform.Rotation=&RigVMFunction_MathQuaternionFromEuler(Euler=(X=0,Y=20,Z=0)).Result
    Weight=@n32.OutputParam
   )
   > n33 = &RigUnit_SphericalPoseReader(
    ActiveRegionScaleFactors=(PositiveWidth=1.5,NegativeWidth=1.5,PositiveHeight=2,NegativeHeight=2)
    ActiveRegionSize=0.1
    Debug=(bDrawDebug=false,bDraw2D=false,bDrawLocalAxes=false,DebugScale=25,DebugSegments=20,DebugThickness=0.25)
    DriverAxis=(X=-1,Y=0,Z=0)
    DriverItem=(Type=Bone,Name="upperarm_l")
    FalloffRegionScaleFactors=(PositiveWidth=1,NegativeWidth=1.6,PositiveHeight=2,NegativeHeight=2)
    FalloffSize=0.35
    FlipHeightScaling=false
    FlipWidthScaling=false
    OptionalParentItem=(Type=Bone,Name="spine_05")
    RotationOffset=(X=-90,Y=30,Z=40)
   )
   > &RigUnit_ModifyTransforms(
    ItemToModify=((Item=(Type=Bone,Name="dyn_l_shoulder"),Transform=(Rotation=(X=0,Y=0,Z=0,W=1),Translation=(X=-0.75,Y=-0.5,Z=0),Scale3D=(X=1,Y=1,Z=1))))
    ItemToModify.0.Transform.Rotation=&RigVMFunction_MathQuaternionFromEuler(Euler=(X=30,Y=20,Z=0)).Result
    Weight=@n33.OutputParam
   )
   n34 = &RigUnit_SphericalPoseReader(
    ActiveRegionScaleFactors=(PositiveWidth=1,NegativeWidth=1,PositiveHeight=1,NegativeHeight=1)
    ActiveRegionSize=0.15
    Debug=(bDrawDebug=false,bDraw2D=false,bDrawLocalAxes=false,DebugScale=25,DebugSegments=20,DebugThickness=0.25)
    DriverAxis=(X=0,Y=0,Z=1)
    DriverItem=(Type=Bone,Name="upperarm_l")
    FalloffRegionScaleFactors=(PositiveWidth=1,NegativeWidth=1,PositiveHeight=1,NegativeHeight=1)
    FalloffSize=0.3
    OptionalParentItem=(Type=Bone,Name="clavicle_l")
    RotationOffset=(X=-90,Y=-45,Z=-15)
    ExecutePin=@SequenceAggregate_2_1.C
   )
   > &RigUnit_ModifyTransforms(
    ItemToModify=((Item=(Type=Bone,Name="dyn_l_shoulder"),Transform=(Rotation=(X=0,Y=0,Z=0,W=1),Translation=(X=0,Y=0,Z=0),Scale3D=(X=1,Y=1,Z=1))))
    ItemToModify.0.Transform.Rotation=&RigVMFunction_MathQuaternionFromEuler(Euler=(X=0,Y=50,Z=0)).Result
    Weight=@n34.OutputParam
   )
   > &RigUnit_ModifyTransforms(
    ItemToModify=((Item=(Type=Bone,Name="dyn_l_shoulder_bk"),Transform=(Rotation=(X=0,Y=0,Z=0,W=1),Translation=(X=0,Y=0,Z=0),Scale3D=(X=1,Y=1,Z=1))))
    ItemToModify.0.Transform.Rotation=&RigVMFunction_MathQuaternionFromEuler(Euler=(X=0,Y=0,Z=25)).Result
    Weight=@n34.OutputParam
   )
   > n35 = &RigUnit_SphericalPoseReader(
    ActiveRegionScaleFactors=(PositiveWidth=1,NegativeWidth=1,PositiveHeight=1,NegativeHeight=1)
    ActiveRegionSize=0.15
    Debug=(bDrawDebug=false,bDraw2D=false,bDrawLocalAxes=false,DebugScale=25,DebugSegments=20,DebugThickness=0.25)
    DriverAxis=(X=0,Y=0,Z=1)
    DriverItem=(Type=Bone,Name="upperarm_r")
    FalloffRegionScaleFactors=(PositiveWidth=1,NegativeWidth=1,PositiveHeight=1,NegativeHeight=1)
    FalloffSize=0.3
    OptionalParentItem=(Type=Bone,Name="clavicle_r")
    RotationOffset=(X=-90,Y=-45,Z=-15)
   )
   > &RigUnit_ModifyTransforms(
    ItemToModify=((Item=(Type=Bone,Name="dyn_r_shoulder"),Transform=(Rotation=(X=0,Y=0,Z=0,W=1),Translation=(X=0,Y=0,Z=0),Scale3D=(X=1,Y=1,Z=1))))
    ItemToModify.0.Transform.Rotation=&RigVMFunction_MathQuaternionFromEuler(Euler=(X=0,Y=50,Z=0)).Result
    Weight=@n35.OutputParam
   )
   > &RigUnit_ModifyTransforms(
    ItemToModify=((Item=(Type=Bone,Name="dyn_r_shoulder_bk"),Transform=(Rotation=(X=0,Y=0,Z=0,W=1),Translation=(X=0,Y=0,Z=0),Scale3D=(X=1,Y=1,Z=1))))
    ItemToModify.0.Transform.Rotation=&RigVMFunction_MathQuaternionFromEuler(Euler=(X=0,Y=0,Z=-25)).Result
    Weight=@n35.OutputParam
   )
   RigVMRerouteNode(Value=())
   &RigUnit_SphericalPoseReader(
    ActiveRegionScaleFactors=(PositiveWidth=5.5,NegativeWidth=4.5,PositiveHeight=2,NegativeHeight=2)
    ActiveRegionSize=0.1
    Debug=(bDrawDebug=true,bDraw2D=false,DebugScale=25,DebugSegments=20,DebugThickness=0.25)
    DriverAxis=(X=-1,Y=0,Z=0)
    DriverItem=(Type=Bone,Name="thigh_l")
    ExecutePin=()
    FalloffRegionScaleFactors=(PositiveWidth=1,NegativeWidth=1.5,PositiveHeight=1,NegativeHeight=1)
    FalloffSize=0.7
    FlipWidthScaling=false
    OptionalParentItem=(Type=Bone)
    RotationOffset=(X=180,Y=50,Z=0)
   )
  ]}
 SequenceAggregate_2_1_1_ContainedGraph {nodes [
   n1 = RigVMFunctionEntryNode()
   SequenceAggregate_2_1 = &RigVMFunction_Sequence(ExecuteContext=@n1.ExecuteContext)
   SequenceAggregate_2_1_1 = &RigVMFunction_Sequence(ExecuteContext=@SequenceAggregate_2_1.B)
   RigVMFunctionReturnNode(A=@SequenceAggregate_2_1.A, B=@SequenceAggregate_2_1_1.A, C=@SequenceAggregate_2_1_1.B)
  ]}
]
hierarchy [
 root {loc=<0 0 0>, rot=<0 0 0 1>}
 pelvis {parent=root, loc=<0 -0.5301 98.0571>, rot=<0.0127 0.707 -0.0127 -0.707>}
 spine_01 {parent=pelvis, loc=<6.1208 -0.9328 0>, rot=<0 0 0.034 -0.9994>}
 spine_02 {parent=spine_01, loc=<6.5996 -0.2095 0>, rot=<0 0 0.0238 -0.9997>}
 spine_03 {parent=spine_02, loc=<6.4914 -0.1237 0>, rot=<0 0 0.0149 -0.9999>}
 spine_04 {parent=spine_03, loc=<6.9598 0.0794 0>, rot=<0 0 -0.0132 -0.9999>}
 spine_05 {parent=spine_04, loc=<6.9987 0 0>, rot=<0 0 -0.0015 -1>}
 attach_cape {parent=spine_05, loc=<9.6538 -10.3341 0>, rot=<0.0616 -0.7044 -0.0616 -0.7044>}
 neck_01 {parent=spine_05, loc=<11.3138 0.1529 0>, rot=<0 0 -0.2003 -0.9797>}
 neck_02 {parent=neck_01, loc=<4.5627 0 0>, rot=<0 0 0.0504 -0.9987>}
 head {parent=neck_02, loc=<5.3914 0 0>, rot=<0 0 0.115 -0.9934>}
 attach_backpack {parent=spine_05, loc=<4.3354 -14.5126 0>, rot=<0.0323 -0.7064 -0.0323 -0.7064>}
 clavicle_l {parent=spine_05, loc=<5.4831 -0.1053 -3.2946>, rot=<0.0465 -0.7289 -0.009 -0.6829>}
 upperarm_l {parent=clavicle_l, loc=<9.8658 -1.2136 -0.4105>, rot=<0.0164 -0.3245 -0.0269 -0.9454>}
 lowerarm_l {parent=upperarm_l, loc=<26.465 -0.8345 0.259>, rot=<0.0036 0.0161 -0.2298 -0.9731>}
 hand_l {parent=lowerarm_l, loc=<22.7355 -0.3203 -0.6105>, rot=<-0.6784 -0.0223 0.0735 -0.7307>}
 thumb_01_l {parent=hand_l, loc=<1.3118 -0.3191 -2.459>, rot=<0.6546 -0.4042 0.0664 -0.6354>}
 thumb_02_l {parent=thumb_01_l, loc=<5.9343 -0.0461 0.0405>, rot=<-0.0037 0.0213 0.298 -0.9543>}
 thumb_03_l {parent=thumb_02_l, loc=<2.2223 0.3682 -0.0907>, rot=<0.0118 -0.013 0.0066 -0.9998>}
 index_metacarpal_l {parent=hand_l, loc=<1.6535 0.0183 -2.96>, rot=<0.0158 -0.0124 -0.0173 -0.9996>}
 index_01_l {parent=index_metacarpal_l, loc=<8.3008 0.0357 -0.0524>, rot=<-0.0044 0.0485 0.0924 -0.9945>}
 index_02_l {parent=index_01_l, loc=<3.3587 -0.3007 0.0451>, rot=<-0.002 0.0027 0.3225 -0.9466>}
 index_03_l {parent=index_02_l, loc=<2.5578 -0.123 -0.0252>, rot=<-0.002 0.0067 0.2558 -0.9667>}
 middle_metacarpal_l {parent=hand_l, loc=<2.1856 -0.1725 -0.5984>, rot=<0.006 0.0168 -0.0329 -0.9993>}
 middle_01_l {parent=middle_metacarpal_l, loc=<7.3275 -0.2447 0.0129>, rot=<-0.0208 0.0042 0.0858 -0.9961>}
 middle_02_l {parent=middle_01_l, loc=<4.0071 0.2074 0.2322>, rot=<-0.0218 0.0258 0.3512 -0.9357>}
 middle_03_l {parent=middle_02_l, loc=<2.6077 0.0534 0.1373>, rot=<0.0082 0.0188 0.1708 -0.9851>}
 ring_metacarpal_l {parent=hand_l, loc=<2.2074 -0.3625 1.3638>, rot=<-0.0284 0.0722 -0.0458 -0.9959>}
 ring_01_l {parent=ring_metacarpal_l, loc=<7.0311 -0.6198 -0.145>, rot=<-0.0489 -0.0252 0.1892 -0.9804>}
 ring_02_l {parent=ring_01_l, loc=<3.6174 0.4647 -0.1027>, rot=<0 0 0.2806 -0.9598>}
 ring_03_l {parent=ring_02_l, loc=<2.4152 0.114 0.0199>, rot=<0 0 0.1189 -0.9929>}
 pinky_metacarpal_l {parent=hand_l, loc=<1.7616 -0.9753 3.3814>, rot=<-0.0397 0.0814 -0.069 -0.9935>}
 pinky_01_l {parent=pinky_metacarpal_l, loc=<7.1727 -0.4832 -0.1097>, rot=<-0.0654 -0.0035 0.2242 -0.9723>}
 pinky_02_l {parent=pinky_01_l, loc=<2.8274 0.0893 -0.1043>, rot=<0.0037 0.0011 0.2854 -0.9584>}
 pinky_03_l {parent=pinky_02_l, loc=<2.2046 0.0307 -0.0078>, rot=<-0.0002 0.0098 0.1445 -0.9895>}
 lowerarm_twist_02_l {parent=lowerarm_l, loc=<10.6707 0 0>, rot=<0 0 0 1>}
 lowerarm_twist_01_l {parent=lowerarm_l, loc=<18.316 0 0>, rot=<0 0 0 1>}
 deform_elbow_in_base_l {parent=lowerarm_l, loc=<0.1834 3.6062 -0.2572>, rot=<0 0 0 -1>}
 deform_elbow_in_l {parent=deform_elbow_in_base_l, loc=<0 0 0>, rot=<0 0 0 1>}
 deform_elbow_out_base_l {parent=lowerarm_l, loc=<0 -4 0>, rot=<0 0 0 1>}
 deform_elbow_out_l {parent=deform_elbow_out_base_l, loc=<0 0 0>, rot=<0 0 0 1>}
 dyn_l_wrist_guard_mstr {parent=lowerarm_l, loc=<20.3709 -0.7554 3.3403>, rot=<-0.0417 -0.068 0.0075 -0.9968>}
 dyn_l_wrist_guard {parent=dyn_l_wrist_guard_mstr, loc=<0 0 0>, rot=<0 0 0 1>}
 dyn_l_wrist_guard_01 {parent=dyn_l_wrist_guard, loc=<4.3658 0.294 0.6177>, rot=<-0.0127 -0.0298 0.0469 -0.9984>}
 upperarm_twist_01_l {parent=upperarm_l, loc=<7.2135 -0.1134 0>, rot=<0 0 0 1>}
 deform_upperarm_up_base_l {parent=upperarm_twist_01_l, loc=<-7.4671 0.5027 2.3566>, rot=<0 0 0 -1>}
 deform_upperarm_up_l {parent=deform_upperarm_up_base_l, loc=<0 0 0>, rot=<0 0 0 1>}
 deform_upperarm_bk_base_l {parent=upperarm_twist_01_l, loc=<-6.4053 -4.3746 0.314>, rot=<0 0 0 -1>}
 deform_upperarm_bk_l {parent=deform_upperarm_bk_base_l, loc=<0 0 0>, rot=<0 0 0 1>}
 deform_upperarm_fr_base_l {parent=upperarm_twist_01_l, loc=<-5.6312 5.092 0.3354>, rot=<0 0 0 -1>}
 deform_upperarm_fr_l {parent=deform_upperarm_fr_base_l, loc=<0 0 0>, rot=<0 0 0 1>}
 upperarm_twist_02_l {parent=upperarm_l, loc=<19.7998 -0.1134 0>, rot=<0 0 0 1>}
 dyn_l_shoulder_master {parent=upperarm_l, loc=<0.4017 -0.401 3.8772>, rot=<-0.0041 0.7381 0.0064 -0.6746>}
 dyn_l_shoulder {parent=dyn_l_shoulder_master, loc=<0 0 0>, rot=<0 0 0 1>}
 dyn_l_shoulder_bk_mstr {parent=dyn_l_shoulder_master, loc=<-8.3755 -5.6424 2.7954>, rot=<-0.2221 -0.6713 0.6713 -0.2221>}
 dyn_l_shoulder_bk {parent=dyn_l_shoulder_bk_mstr, loc=<0 0 0>, rot=<0 0 0 1>}
 dyn_l_shoulder_fr_mstr {parent=dyn_l_shoulder_master, loc=<-6.6277 8.955 -0.506>, rot=<0.2221 -0.6713 -0.6713 -0.2221>}
 dyn_l_shoulder_fr {parent=dyn_l_shoulder_fr_mstr, loc=<0 0 0>, rot=<0 0 0 1>}
 clavicle_r {parent=spine_05, loc=<5.4831 -0.1053 3.2946>, rot=<0.7289 0.0465 -0.6829 0.009>}
 upperarm_r {parent=clavicle_r, loc=<-9.8658 1.2136 0.4105>, rot=<0.0164 -0.3245 -0.0269 -0.9454>}
 lowerarm_r {parent=upperarm_r, loc=<-26.465 0.8345 -0.259>, rot=<0.0036 0.0161 -0.2298 -0.9731>}
 hand_r {parent=lowerarm_r, loc=<-22.7355 0.3203 0.6106>, rot=<-0.6784 -0.0223 0.0735 -0.7307>}
 thumb_01_r {parent=hand_r, loc=<-1.3118 0.3191 2.459>, rot=<0.6546 -0.4042 0.0664 -0.6354>}
 thumb_02_r {parent=thumb_01_r, loc=<-5.935 0.0462 -0.0408>, rot=<-0.0037 0.0213 0.298 -0.9543>}
 thumb_03_r {parent=thumb_02_r, loc=<-2.2223 -0.3682 0.0907>, rot=<0.0118 -0.013 0.0066 -0.9998>}
 index_metacarpal_r {parent=hand_r, loc=<-1.6535 -0.0183 2.96>, rot=<0.0158 -0.0124 -0.0173 -0.9996>}
 index_01_r {parent=index_metacarpal_r, loc=<-8.3008 -0.0357 0.0524>, rot=<-0.0044 0.0485 0.0924 -0.9945>}
 index_02_r {parent=index_01_r, loc=<-3.36 0.3003 -0.0452>, rot=<-0.002 0.0027 0.3225 -0.9466>}
 index_03_r {parent=index_02_r, loc=<-2.5578 0.123 0.0252>, rot=<-0.002 0.0067 0.2558 -0.9667>}
 middle_metacarpal_r {parent=hand_r, loc=<-2.1856 0.1725 0.5984>, rot=<0.006 0.0168 -0.0329 -0.9993>}
 middle_01_r {parent=middle_metacarpal_r, loc=<-7.3275 0.2447 -0.0129>, rot=<-0.0208 0.0042 0.0858 -0.9961>}
 middle_02_r {parent=middle_01_r, loc=<-4.0137 -0.1409 -0.1173>, rot=<-0.0218 0.0258 0.3512 -0.9357>}
 middle_03_r {parent=middle_02_r, loc=<-2.6077 -0.0534 -0.1373>, rot=<0.0082 0.0188 0.1708 -0.9851>}
 ring_metacarpal_r {parent=hand_r, loc=<-2.2074 0.3625 -1.3638>, rot=<-0.0284 0.0722 -0.0458 -0.9959>}
 ring_01_r {parent=ring_metacarpal_r, loc=<-7.0317 0.62 0.1452>, rot=<-0.0489 -0.0252 0.1892 -0.9804>}
 ring_02_r {parent=ring_01_r, loc=<-3.6174 -0.4647 0.1027>, rot=<0 0 0.2806 -0.9598>}
 ring_03_r {parent=ring_02_r, loc=<-2.4152 -0.114 -0.0199>, rot=<0 0 0.1189 -0.9929>}
 pinky_metacarpal_r {parent=hand_r, loc=<-1.7616 0.9753 -3.3814>, rot=<-0.0397 0.0814 -0.069 -0.9935>}
 pinky_01_r {parent=pinky_metacarpal_r, loc=<-7.1727 0.4835 0.1098>, rot=<-0.0654 -0.0035 0.2242 -0.9723>}
 pinky_02_r {parent=pinky_01_r, loc=<-2.8274 -0.0893 0.1043>, rot=<0.0037 0.0011 0.2854 -0.9584>}
 pinky_03_r {parent=pinky_02_r, loc=<-2.2046 -0.0307 0.0078>, rot=<-0.0002 0.0098 0.1445 -0.9895>}
 lowerarm_twist_02_r {parent=lowerarm_r, loc=<-10.6707 0 0>, rot=<0 0 0 -1>}
 lowerarm_twist_01_r {parent=lowerarm_r, loc=<-18.316 0 0>, rot=<0 0 0 -1>}
 deform_elbow_in_base_r {parent=lowerarm_r, loc=<-0.1834 -3.6062 0.2572>, rot=<0 0 0 -1>}
 deform_elbow_in_r {parent=deform_elbow_in_base_r, loc=<0 0 0>, rot=<0 0 0 1>}
 deform_elbow_out_base_r {parent=lowerarm_r, loc=<0 4 -0.0001>, rot=<0 0 0 -1>}
 deform_elbow_out_r {parent=deform_elbow_out_base_r, loc=<0 0 0>, rot=<0 0 0 1>}
 dyn_r_wrist_guard_mstr {parent=lowerarm_r, loc=<-20.3707 0.7553 -3.3406>, rot=<-0.0076 -0.9968 -0.0417 0.068>}
 dyn_r_wrist_guard {parent=dyn_r_wrist_guard_mstr, loc=<0 0 0>, rot=<0 0 0 1>}
 dyn_r_wrist_guard_01 {parent=dyn_r_wrist_guard, loc=<4.3662 -0.2939 0.6174>, rot=<0.0127 -0.0298 -0.0469 -0.9984>}
 upperarm_twist_01_r {parent=upperarm_r, loc=<-7.2135 0.1134 0>, rot=<0 0 0 -1>}
 deform_upperarm_up_base_r {parent=upperarm_twist_01_r, loc=<7.4671 -0.5027 -2.3566>, rot=<0 0 0 -1>}
 deform_upperarm_up_r {parent=deform_upperarm_up_base_r, loc=<0 0 0>, rot=<0 0 0 1>}
 deform_upperarm_bk_base_r {parent=upperarm_twist_01_r, loc=<6.4053 4.3746 -0.314>, rot=<0 0 0 -1>}
 deform_upperarm_bk_r {parent=deform_upperarm_bk_base_r, loc=<0 0 0>, rot=<0 0 0 1>}
 deform_upperarm_fr_base_r {parent=upperarm_twist_01_r, loc=<5.6312 -5.092 -0.3354>, rot=<0 0 0 -1>}
 deform_upperarm_fr_r {parent=deform_upperarm_fr_base_r, loc=<0 0 0>, rot=<0 0 0 1>}
 upperarm_twist_02_r {parent=upperarm_r, loc=<-19.7998 0.1134 0>, rot=<0 0 0 -1>}
 dyn_r_shoulder_master {parent=upperarm_r, loc=<-0.1152 0.3968 -3.6494>, rot=<-0.0063 -0.6918 -0.0042 -0.7221>}
 dyn_r_shoulder {parent=dyn_r_shoulder_master, loc=<0 0 0>, rot=<0 0 0 1>}
 dyn_r_shoulder_fr_mstr {parent=dyn_r_shoulder_master, loc=<-6.1149 -8.955 -1.4381>, rot=<-0.2378 -0.6659 0.6659 -0.2378>}
 dyn_r_shoulder_fr {parent=dyn_r_shoulder_fr_mstr, loc=<0 0 0>, rot=<0 0 0 1>}
 dyn_r_shoulder_bk_mstr {parent=dyn_r_shoulder_master, loc=<-8.0854 5.6424 1.8627>, rot=<0.2378 -0.6659 -0.6659 -0.2378>}
 dyn_r_shoulder_bk {parent=dyn_r_shoulder_bk_mstr, loc=<0 0 0>, rot=<0 0 0 1>}
 deform_upperarm_dn_base_l {parent=spine_05, loc=<-4.1088 -0.6837 -10.2924>, rot=<0.0284 -0.7065 -0.0284 -0.7065>}
 deform_upperarm_dn_l {parent=deform_upperarm_dn_base_l, loc=<0 0 0>, rot=<0 0 0 1>}
 deform_upperarm_dn_base_r {parent=spine_05, loc=<-4.1084 -0.6837 10.2924>, rot=<0.0284 -0.7065 -0.0284 -0.7065>}
 deform_upperarm_dn_r {parent=deform_upperarm_dn_base_r, loc=<0 0 0>, rot=<0 0 0 1>}
 dyn_belt_strap_fr_mstr {parent=spine_01, loc=<6.0725 0.8234 10.27>, rot=<0.2926 0.0558 0.9546 0.0029>}
 dyn_belt_strap_fr_01 {parent=dyn_belt_strap_fr_mstr, loc=<0 0 0>, rot=<0 0 0 1>}
 dyn_belt_strap_fr_02 {parent=dyn_belt_strap_fr_01, loc=<2.2622 -0.0711 0>, rot=<0.0002 0.017 0.0134 -0.9998>}
 dyn_belt_strap_fr_03 {parent=dyn_belt_strap_fr_02, loc=<2.5916 0 0.0059>, rot=<0 -0.0088 0 -1>}
 dyn_belt_strap_bk_mstr {parent=spine_01, loc=<6.1129 0.304 10.2212>, rot=<0.3205 -0.1972 0.9205 -0.1055>}
 dyn_belt_strap_bk_01 {parent=dyn_belt_strap_bk_mstr, loc=<0 0 0>, rot=<0 0 0 1>}
 dyn_belt_strap_bk_02 {parent=dyn_belt_strap_bk_01, loc=<2.2238 -0.0711 0>, rot=<-0.0008 -0.0275 0.0049 -0.9996>}
 dyn_belt_strap_bk_03 {parent=dyn_belt_strap_bk_02, loc=<3.9281 0 -0.0177>, rot=<0.1181 -0.0088 -0.001 -0.993>}
 thigh_l {parent=pelvis, loc=<-1.0855 -0.2109 -8.7053>, rot=<0.0009 0.0355 0.0155 -0.9993>}
 calf_l {parent=thigh_l, loc=<-42.5284 -0.0019 -0.7297>, rot=<0.0006 -0.0161 -0.0391 -0.9991>}
 foot_l {parent=calf_l, loc=<-46.7401 -2.5014 -0.592>, rot=<0.0427 -0.0238 0.0353 -0.9982>}
 ball_l {parent=foot_l, loc=<-3.5658 10.8066 -0.1544>, rot=<0.0001 -0.0001 -0.7063 -0.7079>}
 calf_twist_02_l {parent=calf_l, loc=<-19.1546 -1.7943 -0.2638>, rot=<0.0237 0.0095 -0.0035 -0.9997>}
 calf_twist_01_l {parent=calf_l, loc=<-35.8483 -1.9197 -0.5794>, rot=<0.0237 0.0095 -0.0035 -0.9997>}
 dyn_l_kneepad_master {parent=calf_l, loc=<-0.5483 4.3515 0.3404>, rot=<-0.4581 -0.5187 -0.5203 -0.5005>}
 dyn_l_kneepad {parent=dyn_l_kneepad_master, loc=<0 0 0>, rot=<0 0 0 1>}
 deform_knee_in_base_l {parent=calf_l, loc=<0.8031 -5.0231 0.2456>, rot=<-0.0309 -0.7201 0.0276 -0.6927>}
 deform_knee_in_l {parent=deform_knee_in_base_l, loc=<0 0 0>, rot=<0 0 0 1>}
 deform_knee_out_base_l {parent=calf_l, loc=<0 4 0>, rot=<0 0 0 -1>}
 deform_knee_out_l {parent=deform_knee_out_base_l, loc=<0 0 0>, rot=<0 0 0 1>}
 thigh_twist_01_l {parent=thigh_l, loc=<-27.952 0 0>, rot=<0.0013 0 0 -1>}
 thigh_r {parent=pelvis, loc=<-1.0855 -0.2109 8.7053>, rot=<0.0355 -0.0009 0.9993 0.0155>}
 calf_r {parent=thigh_r, loc=<42.5284 0.0019 0.7297>, rot=<0.0006 -0.0161 -0.0391 -0.9991>}
 foot_r {parent=calf_r, loc=<46.7401 2.5014 0.592>, rot=<0.0427 -0.0238 0.0353 -0.9982>}
 ball_r {parent=foot_r, loc=<3.5658 -10.8066 0.1544>, rot=<0.0001 -0.0001 -0.7063 -0.7079>}
 calf_twist_02_r {parent=calf_r, loc=<19.1546 1.7943 0.2638>, rot=<0.0237 0.0095 -0.0035 -0.9997>}
 calf_twist_01_r {parent=calf_r, loc=<35.8483 1.9197 0.5794>, rot=<0.0237 0.0095 -0.0035 -0.9997>}
 dyn_r_kneepad_master {parent=calf_r, loc=<0.8289 -4.3785 -0.2471>, rot=<-0.5274 0.5015 0.4613 -0.5075>}
 dyn_r_kneepad {parent=dyn_r_kneepad_master, loc=<0 0 0>, rot=<0 0 0 1>}
 deform_knee_in_base_r {parent=calf_r, loc=<-0.8031 5.0231 -0.2457>, rot=<0 0 0 -1>}
 deform_knee_in_r {parent=deform_knee_in_base_r, loc=<0 0 0>, rot=<0 0 0 1>}
 deform_knee_out_base_r {parent=calf_r, loc=<0 -4 0>, rot=<0 0 0 -1>}
 deform_knee_out_r {parent=deform_knee_out_base_r, loc=<0 0 0>, rot=<0 0 0 1>}
 thigh_twist_01_r {parent=thigh_r, loc=<27.952 0 0>, rot=<0.0013 0 0 -1>}
 dyn_skirt_base {parent=pelvis, loc=<0 0 0>, rot=<0 0 0 1>}
 dyn_skirt_c_fr__mstr {parent=dyn_skirt_base, loc=<0.1601 8.7943 0>, rot=<0.0738 0.7032 0.7032 0.0738>}
 dyn_skirt_c_fr_01 {parent=dyn_skirt_c_fr__mstr, loc=<0 0 0>, rot=<0 0.0167 0 -0.9999>}
 dyn_skirt_c_fr_02 {parent=dyn_skirt_c_fr_01, loc=<10.3406 0 -0.3866>, rot=<0 -0.0317 0 -0.9995>}
 dyn_skirt_c_fr_03 {parent=dyn_skirt_c_fr_02, loc=<11.5879 0 -0.111>, rot=<0 -0.0143 0 -0.9999>}
 dyn_l_bk_mstr {parent=dyn_skirt_base, loc=<-0.8033 -10.8794 -7.1131>, rot=<0.0249 -0.811 0.5475 -0.2049>}
 dyn_l_bk_01 {parent=dyn_l_bk_mstr, loc=<0 0 0>, rot=<0 0 0 1>}
 dyn_l_bk_02 {parent=dyn_l_bk_01, loc=<6.4825 -0.0428 -0.1903>, rot=<-0.0009 -0.0292 -0.0091 -0.9995>}
 dyn_l_bk_mid_mstr {parent=dyn_skirt_base, loc=<0.228 -7.9977 -11.9616>, rot=<-0.0346 -0.8883 0.3315 -0.3161>}
 dyn_l_bk_mid_01 {parent=dyn_l_bk_mid_mstr, loc=<0 0 0>, rot=<0 0 0 1>}
 dyn_l_bk_mid_02 {parent=dyn_l_bk_mid_01, loc=<6.2804 -0.0239 -0.1785>, rot=<-0.0003 -0.0293 0.0123 -0.9995>}
 dyn_l_mid_mstr {parent=dyn_skirt_base, loc=<2.1932 -1.6814 -13.9441>, rot=<-0.0364 -0.9371 0.0103 -0.347>}
 dyn_l_mid_01 {parent=dyn_l_mid_mstr, loc=<0 0 0>, rot=<0 0 0 1>}
 dyn_l_mid_02 {parent=dyn_l_mid_01, loc=<6.2804 -0.0239 -0.1785>, rot=<0 -0.0293 0.0211 -0.9993>}
 dyn_l_fr_mid_mstr {parent=dyn_skirt_base, loc=<1.2791 4.9194 -11.9718>, rot=<0.0166 -0.8653 -0.3722 -0.3354>}
 dyn_l_fr_mid_01 {parent=dyn_l_fr_mid_mstr, loc=<0 0 0>, rot=<0 0 0 1>}
 dyn_l_fr_mid_02 {parent=dyn_l_fr_mid_01, loc=<7.6124 0.0132 -0.2364>, rot=<-0.0007 -0.0293 0.0001 -0.9996>}
 dyn_l_fr_mstr {parent=dyn_skirt_base, loc=<0.2768 7.9357 -6.5448>, rot=<0.0796 -0.7371 -0.6438 -0.1895>}
 dyn_l_fr_01 {parent=dyn_l_fr_mstr, loc=<0 0 0>, rot=<0.0012 0.026 -0.0441 -0.9987>}
 dyn_l_fr_02 {parent=dyn_l_fr_01, loc=<9.2784 -0.0052 0.0082>, rot=<-0.0012 -0.0292 -0.0187 -0.9994>}
 dyn_r_fr_mstr {parent=dyn_skirt_base, loc=<0.3649 8.3224 6.5783>, rot=<0.1838 0.6512 0.7353 -0.0376>}
 dyn_r_fr_01 {parent=dyn_r_fr_mstr, loc=<0 0 0>, rot=<0.0498 0.0004 0.0003 -0.9988>}
 dyn_r_fr_02 {parent=dyn_r_fr_01, loc=<9.2784 0 0>, rot=<0.0016 -0.0309 0.0168 -0.9994>}
 dyn_r_fr_mid_mstr {parent=dyn_skirt_base, loc=<1.9315 5.0444 11.3812>, rot=<0.3287 0.3339 0.8832 -0.0214>}
 dyn_r_fr_mid_01 {parent=dyn_r_fr_mid_mstr, loc=<0 0 0>, rot=<0.0411 -0.0155 -0.0015 -0.999>}
 dyn_r_fr_mid_02 {parent=dyn_r_fr_mid_01, loc=<7.6161 0 0>, rot=<0.0015 -0.0138 0.0002 -0.9999>}
 dyn_r_mid_mstr {parent=dyn_skirt_base, loc=<2.3169 -1.5188 14.2534>, rot=<0.3324 -0.0418 0.9416 0.0346>}
 dyn_r_mid_01 {parent=dyn_r_mid_mstr, loc=<0 0 0>, rot=<0.0161 -0.0142 0.0017 -0.9998>}
 dyn_r_mid_02 {parent=dyn_r_mid_01, loc=<6.2831 0 0>, rot=<0.0002 -0.0127 -0.0208 -0.9997>}
 dyn_r_bk_mid_mstr {parent=dyn_skirt_base, loc=<1.2132 -6.9125 11.6124>, rot=<0.2779 -0.3626 0.8892 0.0244>}
 dyn_r_bk_mid_01 {parent=dyn_r_bk_mid_mstr, loc=<0 0 0>, rot=<-0.009 -0.0142 0.002 -0.9999>}
 dyn_r_bk_mid_02 {parent=dyn_r_bk_mid_01, loc=<6.283 0 0>, rot=<-0.0232 -0.013 -0.0129 -0.9996>}
 dyn_r_bk_mstr {parent=dyn_skirt_base, loc=<-0.4245 -10.6587 7.6197>, rot=<0.206 -0.5704 0.7949 -0.018>}
 dyn_r_bk_01 {parent=dyn_r_bk_mstr, loc=<0 0 0>, rot=<-0.0217 -0.0146 0.0036 -0.9997>}
 dyn_r_bk_02 {parent=dyn_r_bk_01, loc=<6.4855 0 0>, rot=<0.0378 -0.0164 0.0098 -0.9991>}
 dyn_skirt_c_bk__mstr {parent=dyn_skirt_base, loc=<-0.9613 -11.5883 0>, rot=<0.0572 -0.7048 0.7048 -0.0572>}
 dyn_skirt_c_bk_01 {parent=dyn_skirt_c_bk__mstr, loc=<-0.0157 0 0>, rot=<0 0.0447 0 -0.999>}
 dyn_skirt_c_bk_02 {parent=dyn_skirt_c_bk_01, loc=<10.3479 0 0>, rot=<0 -0.0056 0 -1>}
 dyn_skirt_c_bk_03 {parent=dyn_skirt_c_bk_02, loc=<11.5885 0 0>, rot=<0 -0.0095 0 -1>}
 dyn_diamond_l_fr_big_mstr {parent=dyn_skirt_base, loc=<-7.2537 9.1049 -13.9968>, rot=<0.1078 -0.7602 -0.5905 -0.2487>}
 dyn_diamond_l_fr_big {parent=dyn_diamond_l_fr_big_mstr, loc=<0 0 0>, rot=<0 0 0 1>}
 dyn_diamond_l_fr_sml_mstr {parent=dyn_skirt_base, loc=<-3.1045 2.8077 -18.1149>, rot=<0.0426 -0.9368 -0.1889 -0.2915>}
 dyn_diamond_l_fr_sml {parent=dyn_diamond_l_fr_sml_mstr, loc=<0 0 0>, rot=<0 0 0 1>}
 dyn_diamond_l_bk_big_mstr {parent=dyn_skirt_base, loc=<-10.4556 -11.8256 -14.861>, rot=<-0.0797 -0.8164 0.5281 -0.2196>}
 dyn_diamond_l_bk_big {parent=dyn_diamond_l_bk_big_mstr, loc=<0 0 0>, rot=<0 0 0 1>}
 dyn_diamond_l_bk_sml_mstr {parent=dyn_skirt_base, loc=<-4.6073 -5.209 -18.2312>, rot=<-0.0741 -0.9318 0.2 -0.2938>}
 dyn_diamond_l_bk_sml {parent=dyn_diamond_l_bk_sml_mstr, loc=<0 0 0>, rot=<0 0 0 1>}
 dyn_diamond_r_fr_big_mstr {parent=dyn_skirt_base, loc=<-7.838 9.1884 14.3034>, rot=<0.2487 0.5905 0.7602 -0.1078>}
 dyn_diamond_r_fr_big {parent=dyn_diamond_r_fr_big_mstr, loc=<0 0 0>, rot=<0 0 0 -1>}
 dyn_diamond_r_fr_sml_mstr {parent=dyn_skirt_base, loc=<-3.5452 2.7664 18.237>, rot=<0.2915 0.1889 0.9368 -0.0426>}
 dyn_diamond_r_fr_sml {parent=dyn_diamond_r_fr_sml_mstr, loc=<0 0 0>, rot=<0 0 0 -1>}
 dyn_diamond_r_bk_big_mstr {parent=dyn_skirt_base, loc=<-10.0796 -11.8479 14.6906>, rot=<0.224 -0.5263 0.8157 0.0865>}
 dyn_diamond_r_bk_big {parent=dyn_diamond_r_bk_big_mstr, loc=<0 0 0>, rot=<0 0 0 1>}
 dyn_diamond_r_bk_sml_mstr {parent=dyn_skirt_base, loc=<-4.2413 -5.1242 18.0045>, rot=<0.2938 -0.2 0.9318 0.0741>}
 dyn_diamond_r_bk_sml {parent=dyn_diamond_r_bk_sml_mstr, loc=<0 0 0>, rot=<0 0 0 -1>}
 dyn_skirt_l_bk_mstr {parent=dyn_skirt_base, loc=<-0.8033 -10.8794 -7.1131>, rot=<0.0249 -0.811 0.5475 -0.2049>}
 dyn_skirt_l_bk_01 {parent=dyn_skirt_l_bk_mstr, loc=<0 0 0>, rot=<0 0 0 1>}
 dyn_skirt_l_bk_02 {parent=dyn_skirt_l_bk_01, loc=<6.4825 -0.0428 -0.1903>, rot=<-0.0009 -0.0292 -0.0091 -0.9995>}
 dyn_skirt_l_bk_mid_mstr {parent=dyn_skirt_base, loc=<0.228 -7.9977 -11.9616>, rot=<-0.0346 -0.8883 0.3315 -0.3161>}
 dyn_skirt_l_bk_mid_01 {parent=dyn_skirt_l_bk_mid_mstr, loc=<0 0 0>, rot=<0 0 0 1>}
 dyn_skirt_l_bk_mid_02 {parent=dyn_skirt_l_bk_mid_01, loc=<6.2804 -0.0239 -0.1785>, rot=<-0.0003 -0.0293 0.0123 -0.9995>}
 dyn_skirt_l_mid_mstr {parent=dyn_skirt_base, loc=<2.1932 -1.6814 -13.9441>, rot=<-0.0364 -0.9371 0.0103 -0.347>}
 dyn_skirt_l_mid_01 {parent=dyn_skirt_l_mid_mstr, loc=<0 0 0>, rot=<0 0 0 1>}
 dyn_skirt_l_mid_02 {parent=dyn_skirt_l_mid_01, loc=<6.2804 -0.0239 -0.1785>, rot=<0 -0.0293 0.0211 -0.9993>}
 dyn_skirt_l_fr_mid_mstr {parent=dyn_skirt_base, loc=<1.2791 4.9194 -11.9718>, rot=<0.0166 -0.8653 -0.3722 -0.3354>}
 dyn_skirt_l_fr_mid_01 {parent=dyn_skirt_l_fr_mid_mstr, loc=<0 0 0>, rot=<0 0 0 1>}
 dyn_skirt_l_fr_mid_02 {parent=dyn_skirt_l_fr_mid_01, loc=<7.6124 0.0132 -0.2364>, rot=<-0.0007 -0.0293 0.0001 -0.9996>}
 dyn_skirt_l_fr_mstr {parent=dyn_skirt_base, loc=<0.2768 7.9357 -6.5448>, rot=<0.0796 -0.7371 -0.6438 -0.1895>}
 dyn_skirt_l_fr_01 {parent=dyn_skirt_l_fr_mstr, loc=<0 0 0>, rot=<0.0012 0.026 -0.0441 -0.9987>}
 dyn_skirt_l_fr_02 {parent=dyn_skirt_l_fr_01, loc=<9.2784 -0.0052 0.0082>, rot=<-0.0012 -0.0292 -0.0187 -0.9994>}
 dyn_skirt_r_fr_mstr {parent=dyn_skirt_base, loc=<0.3649 8.3224 6.5783>, rot=<0.1838 0.6512 0.7353 -0.0376>}
 dyn_skirt_r_fr_01 {parent=dyn_skirt_r_fr_mstr, loc=<0 0 0>, rot=<0.0498 0.0004 0.0003 -0.9988>}
 dyn_skirt_r_fr_02 {parent=dyn_skirt_r_fr_01, loc=<9.2784 0 0>, rot=<0.0016 -0.0309 0.0168 -0.9994>}
 dyn_skirt_r_fr_mid_mstr {parent=dyn_skirt_base, loc=<1.9315 5.0444 11.3812>, rot=<0.3287 0.3339 0.8832 -0.0214>}
 dyn_skirt_r_fr_mid_01 {parent=dyn_skirt_r_fr_mid_mstr, loc=<0 0 0>, rot=<0.0411 -0.0155 -0.0015 -0.999>}
 dyn_skirt_r_fr_mid_02 {parent=dyn_skirt_r_fr_mid_01, loc=<7.6161 0 0>, rot=<0.0015 -0.0138 0.0002 -0.9999>}
 dyn_skirt_r_mid_mstr {parent=dyn_skirt_base, loc=<2.3169 -1.5188 14.2534>, rot=<0.3324 -0.0418 0.9416 0.0346>}
 dyn_skirt_r_mid_01 {parent=dyn_skirt_r_mid_mstr, loc=<0 0 0>, rot=<0.0161 -0.0142 0.0017 -0.9998>}
 dyn_skirt_r_mid_02 {parent=dyn_skirt_r_mid_01, loc=<6.2831 0 0>, rot=<0.0002 -0.0127 -0.0208 -0.9997>}
 dyn_skirt_r_bk_mid_mstr {parent=dyn_skirt_base, loc=<1.2132 -6.9125 11.6124>, rot=<0.2779 -0.3626 0.8892 0.0244>}
 dyn_skirt_r_bk_mid_01 {parent=dyn_skirt_r_bk_mid_mstr, loc=<0 0 0>, rot=<-0.009 -0.0142 0.002 -0.9999>}
 dyn_skirt_r_bk_mid_02 {parent=dyn_skirt_r_bk_mid_01, loc=<6.283 0 0>, rot=<-0.0232 -0.013 -0.0129 -0.9996>}
 dyn_skirt_r_bk_mstr {parent=dyn_skirt_base, loc=<-0.4245 -10.6587 7.6197>, rot=<0.206 -0.5704 0.7949 -0.018>}
 dyn_skirt_r_bk_01 {parent=dyn_skirt_r_bk_mstr, loc=<0 0 0>, rot=<-0.0217 -0.0146 0.0036 -0.9997>}
 dyn_skirt_r_bk_02 {parent=dyn_skirt_r_bk_01, loc=<6.4855 0 0>, rot=<0.0378 -0.0164 0.0098 -0.9991>}
 dyn_belt_fr_r_lower_mstr {parent=pelvis, loc=<2.1863 9.3647 6>, rot=<-0.0127 -0.707 0.0127 -0.707>}
 dyn_belt_fr_r_lower {parent=dyn_belt_fr_r_lower_mstr, loc=<0 0 0>, rot=<0 0 0 1>}
 dyn_belt_r_mstr {parent=pelvis, loc=<10.0891 0.1688 11.4463>, rot=<-0.0127 -0.707 0.0127 -0.707>}
 dyn_belt_r {parent=dyn_belt_r_mstr, loc=<0 0 0>, rot=<0 0 0 1>}
 dyn_belt_l_mstr {parent=pelvis, loc=<7.7025 0.2544 -12.6216>, rot=<-0.0127 -0.707 0.0127 -0.707>}
 dyn_belt_l {parent=dyn_belt_l_mstr, loc=<0 0 0>, rot=<0 0 0 1>}
 dyn_belt_bk_mstr {parent=pelvis, loc=<5.8296 -10.823 0>, rot=<-0.0127 -0.707 0.0127 -0.707>}
 dyn_belt_bk {parent=dyn_belt_bk_mstr, loc=<0 0 0>, rot=<0 0 0 1>}
 dyn_belt_fr_l_lower_mstr {parent=pelvis, loc=<1.09 9.2963 -6.2466>, rot=<-0.0127 -0.707 0.0127 -0.707>}
 dyn_belt_fr_l_lower {parent=dyn_belt_fr_l_lower_mstr, loc=<0 0 0>, rot=<0 0 0 1>}
 dyn_belt_fr_upper_mstr {parent=pelvis, loc=<5.2536 9.9576 -0.3414>, rot=<-0.0127 -0.707 0.0127 -0.707>}
 dyn_belt_fr_upper {parent=dyn_belt_fr_upper_mstr, loc=<0 0 0>, rot=<0 0 0 1>}
 dyn_Cloth_volume_fn {parent=pelvis, loc=<5.5092 8.1614 0.1611>, rot=<0.994 -0.1097 0 0>}
 dyn_Cloth_volume_fn01 {parent=dyn_Cloth_volume_fn, loc=<-5.2419 -1.5293 0>, rot=<0 0 0 1>}
 dyn_Cloth_volume_fn02 {parent=dyn_Cloth_volume_fn01, loc=<-14.9076 -1.0417 0.0916>, rot=<0.024 0 0 -0.9997>}
 dyn_Cloth_volume_bk {parent=pelvis, loc=<5.1712 -9.396 0.2721>, rot=<0 0 -0.0741 -0.9973>}
 dyn_Cloth_volume_bk01 {parent=dyn_Cloth_volume_bk, loc=<-4.7247 -0.97 0>, rot=<0 0 0 1>}
 dyn_Cloth_volume_bk02 {parent=dyn_Cloth_volume_bk01, loc=<-12.0476 -1.6111 -0.2026>, rot=<0.024 0 0 -0.9997>}
 dyn_Cloth_volume_l_A {parent=pelvis, loc=<6.0145 5.286 -9.1939>, rot=<0.983 -0.1085 0.011 0.1476>}
 dyn_Cloth_volume_l_A01 {parent=dyn_Cloth_volume_l_A, loc=<-4.5252 -1.2971 0.6872>, rot=<0 0 0 1>}
 dyn_Cloth_volume_l_A02 {parent=dyn_Cloth_volume_l_A01, loc=<-14.6412 -1.2837 0.4445>, rot=<0 0 0 1>}
 dyn_Cloth_volume_l_B {parent=pelvis, loc=<5.891 -1.829 -13.2219>, rot=<-0.7061 0.078 -0.0521 -0.7019>}
 dyn_Cloth_volume_l_B01 {parent=dyn_Cloth_volume_l_B, loc=<-4.7771 -1.8458 0>, rot=<0 0 0 1>}
 dyn_Cloth_volume_l_B02 {parent=dyn_Cloth_volume_l_B01, loc=<-12.5849 -2.0721 0>, rot=<0 0 0 1>}
 dyn_Cloth_volume_l_C {parent=pelvis, loc=<5.5576 -7.461 -8.659>, rot=<-0.1371 0.0151 -0.0734 -0.9877>}
 dyn_Cloth_volume_l_C01 {parent=dyn_Cloth_volume_l_C, loc=<-4.5252 -1.5667 -0.6841>, rot=<0 0 0 1>}
 dyn_Cloth_volume_l_C02 {parent=dyn_Cloth_volume_l_C01, loc=<-12.5716 -1.8288 -0.8944>, rot=<0 0 0 1>}
 dyn_Cloth_volume_r_A {parent=pelvis, loc=<6.0145 5.286 9.1939>, rot=<0.983 -0.1085 -0.011 -0.1476>}
 dyn_Cloth_volume_r_A01 {parent=dyn_Cloth_volume_r_A, loc=<-4.5202 -1.5633 -0.6872>, rot=<0 0 0.0286 -0.9996>}
 dyn_Cloth_volume_r_A02 {parent=dyn_Cloth_volume_r_A01, loc=<-14.5343 -2.4471 -0.4444>, rot=<0 0 0 1>}
 dyn_Cloth_volume_r_B {parent=pelvis, loc=<5.891 -1.829 13.2219>, rot=<0.7054 -0.084 -0.0581 -0.7014>}
 dyn_Cloth_volume_r_B01 {parent=dyn_Cloth_volume_r_B, loc=<-4.8061 -1.5542 0>, rot=<0 0 0 1>}
 dyn_Cloth_volume_r_B02 {parent=dyn_Cloth_volume_r_B01, loc=<-12.6173 -1.7284 0>, rot=<0 0 0 1>}
 dyn_Cloth_volume_r_C {parent=pelvis, loc=<5.5576 -7.461 8.6589>, rot=<0.1369 -0.0169 -0.0861 -0.9867>}
 dyn_Cloth_volume_r_C01 {parent=dyn_Cloth_volume_r_C, loc=<-4.5599 -1.133 0.6841>, rot=<0 0 0 1>}
 dyn_Cloth_volume_r_C02 {parent=dyn_Cloth_volume_r_C01, loc=<-12.612 -1.3035 0.8944>, rot=<0 0 0 1>}
]
variables [Dual_Pistols_Twist {type=double, default=0}]
