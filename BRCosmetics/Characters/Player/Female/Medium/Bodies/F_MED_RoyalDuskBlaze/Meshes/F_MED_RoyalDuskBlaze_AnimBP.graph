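# Animation Blueprint for the F_MED_RoyalDuskBlaze female body cosmetic, adding secondary motion to its armor and cloth.
# It copies the attached parent mesh's pose, then weighted constraints drive shoulder and wrist-guard master bones.
# A skirt Control Rig takes a dual_pistols twist input, followed by additive yaw offsets on cloth-volume bones.
# Per-bone AnimDynamics then simulates that cloth with angular springs, limits and gravity overrides; a separate diamond-bone chain has no input pose.
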AnimBlueprint /BRCosmetics/Characters/Player/Female/Medium/Bodies/F_MED_RoyalDuskBlaze/Meshes/F_MED_RoyalDuskBlaze_AnimBP
target_skeleton=&/F_MED_RoyalDuskBlaze_Skeleton
graphs [
 AnimGraph {nodes [
   n1 = Anim Dynamics - Bone: dyn_diamond_l_fr_big(Alpha=0.3, ExternalForce="0,0,0") {
    angularDampingOverride=1
    angularSpringConstant=20000
    bAngularSpring=true
    bGravityOverrideInSimSpace=true
    bLinearSpring=true
    bOverrideAngularDamping=true
    bOverrideLinearDamping=true
    bUseGravityOverride=true
    gravityScale=0.35
    linearDampingOverride=1
    linearSpringConstant=100
    simulationSpace=RootRelative
    alphaBoolBlend {blendOption=Linear}
    boundBone {boneName=dyn_diamond_l_fr_big}
    componentAppliedLinearAccClamp=<3000 3000 3000>
    componentLinearAccScale=<0.3 0.3 0.4>
    componentLinearVelScale=<0.3 0.3 0.4>
    physicsBodyDefinitions [{
      sphereCollisionRadius=0
      boundBone {boneName=dyn_diamond_l_fr_big}
      boxExtents=<20 30 30>
      constraintSetup {angularLimitsMax=<0 0 0>, angularLimitsMin=<0 0 0>, angularTarget=<1 0 0>, linearAxesMax=<0 0 10>}
      localJointOffset=<0 0 0>
     }]
   }
   n2 = Anim Dynamics - Bone: dyn_diamond_l_fr_sml(ComponentPose=@n1, Alpha=0.3, ExternalForce="0,0,0") {
    angularDampingOverride=1
    angularSpringConstant=20000
    bAngularSpring=true
    bGravityOverrideInSimSpace=true
    bLinearSpring=true
    bOverrideAngularDamping=true
    bOverrideLinearDamping=true
    bUseGravityOverride=true
    gravityScale=0.35
    linearDampingOverride=1
    linearSpringConstant=100
    simulationSpace=RootRelative
    alphaBoolBlend {blendOption=Linear}
    boundBone {boneName=dyn_diamond_l_fr_sml}
    componentAppliedLinearAccClamp=<3000 3000 3000>
    componentLinearAccScale=<0.3 0.3 0.4>
    componentLinearVelScale=<0.3 0.3 0.4>
    physicsBodyDefinitions [{
      sphereCollisionRadius=0
      boundBone {boneName=dyn_diamond_l_fr_sml}
      boxExtents=<20 30 30>
      constraintSetup {angularLimitsMax=<0 0 0>, angularLimitsMin=<0 0 0>, angularTarget=<1 0 0>, linearAxesMax=<0 0 10>}
      localJointOffset=<0 0 0>
     }]
   }
   n3 = Anim Dynamics - Bone: dyn_diamond_l_bk_big(ComponentPose=@n2, Alpha=0.3, ExternalForce="0,0,0") {
    angularDampingOverride=1
    angularSpringConstant=20000
    bAngularSpring=true
    bGravityOverrideInSimSpace=true
    bLinearSpring=true
    bOverrideAngularDamping=true
    bOverrideLinearDamping=true
    bUseGravityOverride=true
    gravityScale=0.35
    linearDampingOverride=1
    linearSpringConstant=100
    simulationSpace=RootRelative
    alphaBoolBlend {blendOption=Linear}
    boundBone {boneName=dyn_diamond_l_bk_big}
    componentAppliedLinearAccClamp=<3000 3000 3000>
    componentLinearAccScale=<0.3 0.3 0.4>
    componentLinearVelScale=<0.3 0.3 0.4>
    physicsBodyDefinitions [{
      sphereCollisionRadius=0
      boundBone {boneName=dyn_diamond_l_bk_big}
      boxExtents=<20 30 30>
      constraintSetup {angularLimitsMax=<0 0 0>, angularLimitsMin=<0 0 0>, angularTarget=<1 0 0>, linearAxesMax=<0 0 10>}
      localJointOffset=<0 0 0>
     }]
   }
   n4 = Anim Dynamics - Bone: dyn_diamond_l_bk_sml(ComponentPose=@n3, Alpha=0.3, ExternalForce="0,0,0") {
    angularDampingOverride=1
    angularSpringConstant=20000
    bAngularSpring=true
    bGravityOverrideInSimSpace=true
    bLinearSpring=true
    bOverrideAngularDamping=true
    bOverrideLinearDamping=true
    bUseGravityOverride=true
    gravityScale=0.35
    linearDampingOverride=1
    linearSpringConstant=100
    simulationSpace=RootRelative
    alphaBoolBlend {blendOption=Linear}
    boundBone {boneName=dyn_diamond_l_bk_sml}
    componentAppliedLinearAccClamp=<3000 3000 3000>
    componentLinearAccScale=<0.3 0.3 0.4>
    componentLinearVelScale=<0.3 0.3 0.4>
    physicsBodyDefinitions [{
      sphereCollisionRadius=0
      boundBone {boneName=dyn_diamond_l_bk_sml}
      boxExtents=<20 30 30>
      constraintSetup {angularLimitsMax=<0 0 0>, angularLimitsMin=<0 0 0>, angularTarget=<1 0 0>, linearAxesMax=<0 0 10>}
      localJointOffset=<0 0 0>
     }]
   }
   n5 = Anim Dynamics - Bone: dyn_diamond_r_fr_big(ComponentPose=@n4, Alpha=0.3, ExternalForce="0,0,0") {
    angularDampingOverride=1
    angularSpringConstant=20000
    bAngularSpring=true
    bGravityOverrideInSimSpace=true
    bLinearSpring=true
    bOverrideAngularDamping=true
    bOverrideLinearDamping=true
    bUseGravityOverride=true
    gravityScale=0.35
    linearDampingOverride=1
    linearSpringConstant=100
    simulationSpace=RootRelative
    alphaBoolBlend {blendOption=Linear}
    boundBone {boneName=dyn_diamond_r_fr_big}
    componentAppliedLinearAccClamp=<3000 3000 3000>
    componentLinearAccScale=<0.3 0.3 0.4>
    componentLinearVelScale=<0.3 0.3 0.4>
    physicsBodyDefinitions [{
      sphereCollisionRadius=0
      boundBone {boneName=dyn_diamond_r_fr_big}
      boxExtents=<20 30 30>
      constraintSetup {angularLimitsMax=<0 0 0>, angularLimitsMin=<0 0 0>, angularTarget=<1 0 0>, linearAxesMax=<0 0 10>}
      localJointOffset=<0 0 0>
     }]
   }
   n6 = Anim Dynamics - Bone: dyn_diamond_r_fr_sml(ComponentPose=@n5, Alpha=0.3, ExternalForce="0,0,0") {
    angularDampingOverride=1
    angularSpringConstant=20000
    bAngularSpring=true
    bGravityOverrideInSimSpace=true
    bLinearSpring=true
    bOverrideAngularDamping=true
    bOverrideLinearDamping=true
    bUseGravityOverride=true
    gravityScale=0.35
    linearDampingOverride=1
    linearSpringConstant=100
    simulationSpace=RootRelative
    alphaBoolBlend {blendOption=Linear}
    boundBone {boneName=dyn_diamond_r_fr_sml}
    componentAppliedLinearAccClamp=<3000 3000 3000>
    componentLinearAccScale=<0.3 0.3 0.4>
    componentLinearVelScale=<0.3 0.3 0.4>
    physicsBodyDefinitions [{
      sphereCollisionRadius=0
      boundBone {boneName=dyn_diamond_r_fr_sml}
      boxExtents=<20 30 30>
      constraintSetup {angularLimitsMax=<0 0 0>, angularLimitsMin=<0 0 0>, angularTarget=<1 0 0>, linearAxesMax=<0 0 10>}
      localJointOffset=<0 0 0>
     }]
   }
   n7 = Anim Dynamics - Bone: dyn_diamond_r_bk_big(ComponentPose=@n6, Alpha=0.3, ExternalForce="0,0,0") {
    angularDampingOverride=1
    angularSpringConstant=20000
    bAngularSpring=true
    bGravityOverrideInSimSpace=true
    bLinearSpring=true
    bOverrideAngularDamping=true
    bOverrideLinearDamping=true
    bUseGravityOverride=true
    gravityScale=0.35
    linearDampingOverride=1
    linearSpringConstant=100
    simulationSpace=RootRelative
    alphaBoolBlend {blendOption=Linear}
    boundBone {boneName=dyn_diamond_r_bk_big}
    componentAppliedLinearAccClamp=<3000 3000 3000>
    componentLinearAccScale=<0.3 0.3 0.4>
    componentLinearVelScale=<0.3 0.3 0.4>
    physicsBodyDefinitions [{
      sphereCollisionRadius=0
      boundBone {boneName=dyn_diamond_r_bk_big}
      boxExtents=<20 30 30>
      constraintSetup {angularLimitsMax=<0 0 0>, angularLimitsMin=<0 0 0>, angularTarget=<1 0 0>, linearAxesMax=<0 0 10>}
      localJointOffset=<0 0 0>
     }]
   }
   Anim Dynamics - Bone: dyn_diamond_r_bk_sml(ComponentPose=@n7, Alpha=0.3, ExternalForce="0,0,0") {
    angularDampingOverride=1
    angularSpringConstant=20000
    bAngularSpring=true
    bGravityOverrideInSimSpace=true
    bLinearSpring=true
    bOverrideAngularDamping=true
    bOverrideLinearDamping=true
    bUseGravityOverride=true
    gravityScale=0.35
    linearDampingOverride=1
    linearSpringConstant=100
    simulationSpace=RootRelative
    alphaBoolBlend {blendOption=Linear}
    boundBone {boneName=dyn_diamond_r_bk_sml}
    componentAppliedLinearAccClamp=<3000 3000 3000>
    componentLinearAccScale=<0.3 0.3 0.4>
    componentLinearVelScale=<0.3 0.3 0.4>
    physicsBodyDefinitions [{
      sphereCollisionRadius=0
      boundBone {boneName=dyn_diamond_r_bk_sml}
      boxExtents=<20 30 30>
      constraintSetup {angularLimitsMax=<0 0 0>, angularLimitsMin=<0 0 0>, angularTarget=<1 0 0>, linearAxesMax=<0 0 10>}
      localJointOffset=<0 0 0>
     }]
   }
   Constraint!(
    ComponentPose
    Alpha=1
    ConstraintWeights_0=0
    ConstraintWeights_1:float=0.0007
    ConstraintWeights_2:float=0.3226
    ConstraintWeights_3:float=0.0009
    ConstraintWeights_4:float=0.4043
    ConstraintWeights_5:float=0
    ConstraintWeights_6:float=0.2067
    ConstraintWeights_7:float=0
    ConstraintWeights_8:float=0
    ConstraintWeights_9:float=0.0478
    ConstraintWeights_10:float=0
    ConstraintWeights_11:float=0.0103
    ConstraintWeights_12:float=0
    ConstraintWeights_13:float=0
    ConstraintWeights_14:float=0
    ConstraintWeights_15:float=0
    ConstraintWeights_16:float=0
    ConstraintWeights_17:float=0
    ConstraintWeights_18:float=0
    ConstraintWeights_19:float=0
    ConstraintWeights_20:float=0.0066
   ) -> (Pose) {
    alphaBoolBlend {blendOption=Linear}
    constraintSetup [
     {targetBone {boneName=upperarm_twist_01_l}}
     {targetBone {boneName=spine_02}}
     {targetBone {boneName=spine_04}}
     {targetBone {boneName=spine_03}}
     {targetBone {boneName=spine_05}}
     {targetBone {boneName=clavicle_l}}
     {targetBone {boneName=clavicle_r}}
     {targetBone {boneName=spine_01}}
     {targetBone {boneName=pelvis}}
     {targetBone {boneName=upperarm_twist_01_r}}
     {targetBone {boneName=deform_upperarm_up_l}}
     {targetBone {boneName=deform_upperarm_up_r}}
     {targetBone {boneName=deform_upperarm_dn_l}}
     {targetBone {boneName=deform_upperarm_dn_r}}
     {targetBone {boneName=dyn_Cloth_volume_bk}}
     {targetBone {boneName=dyn_Cloth_volume_r_A}}
     {targetBone {boneName=dyn_Cloth_volume_r_B}}
     {targetBone {boneName=deform_upperarm_bk_l}}
     {targetBone {boneName=deform_upperarm_bk_r}}
     {targetBone {boneName=deform_upperarm_fr_l}}
     {targetBone {boneName=deform_upperarm_fr_r}}
    ]
    constraintWeights [1,1,1,1,1,1,1,1,1,1,1,1,1,1,1,1,1,1,1,1,1]
    boneToModify=~
   }
   n8 = "Constraint - dyn_r_shoulder_master (PPPPP)"!(
    ComponentPose=Local To Component(LocalPose=Copy Pose From Mesh() {bCopyCustomAttributes=true, bUseAttachedParent=true, bUseMeshPose=true})
    Alpha=1
    ConstraintWeights_0=0
    ConstraintWeights_1:float=0.0929
    ConstraintWeights_2:float=0.1961
    ConstraintWeights_3:float=0.4556
    ConstraintWeights_4:float=0.2554
   ) -> (Pose) {
    alphaBoolBlend {blendOption=Linear}
    boneToModify {boneName=dyn_r_shoulder_master}
    constraintSetup [
     {targetBone {boneName=spine_04}}
     {targetBone {boneName=spine_05}}
     {targetBone {boneName=clavicle_r}}
     {targetBone {boneName=upperarm_twist_01_r}}
     {targetBone {boneName=deform_upperarm_up_r}}
    ]
    constraintWeights [1,1,1,1,1]
   }
   n9 = "Constraint - dyn_l_shoulder_master (PPPPP)"!(
    ComponentPose=@n8
    Alpha=1
    ConstraintWeights_0=0.4842
    ConstraintWeights_1:float=0
    ConstraintWeights_2:float=0.0844
    ConstraintWeights_3:float=0.1732
    ConstraintWeights_4:float=0.2582
   ) -> (Pose) {
    alphaBoolBlend {blendOption=Linear}
    boneToModify {boneName=dyn_l_shoulder_master}
    constraintSetup [
     {targetBone {boneName=upperarm_twist_01_l}}
     {targetBone {boneName=spine_04}}
     {targetBone {boneName=spine_05}}
     {targetBone {boneName=clavicle_l}}
     {targetBone {boneName=deform_upperarm_up_l}}
    ]
    constraintWeights [1,1,1,1,1]
   }
   n10 = "Constraint - dyn_r_shoulder_fr_mstr (PPPPPPPP)"!(
    ComponentPose=@n9
    Alpha=1
    ConstraintWeights_0=0.2154
    ConstraintWeights_1:float=0
    ConstraintWeights_2:float=0.3338
    ConstraintWeights_3:float=0.3113
    ConstraintWeights_4:float=0.0993
    ConstraintWeights_5:float=0.0215
    ConstraintWeights_6:float=0
    ConstraintWeights_7:float=0.0186
   ) -> (Pose) {
    alphaBoolBlend {blendOption=Linear}
    boneToModify {boneName=dyn_r_shoulder_fr_mstr}
    constraintSetup [
     {targetBone {boneName=spine_04}}
     {targetBone {boneName=spine_03}}
     {targetBone {boneName=spine_05}}
     {targetBone {boneName=clavicle_r}}
     {targetBone {boneName=upperarm_twist_01_r}}
     {targetBone {boneName=deform_upperarm_up_r}}
     {targetBone {boneName=deform_upperarm_bk_r}}
     {targetBone {boneName=deform_upperarm_fr_r}}
    ]
    constraintWeights [1,1,1,1,1,1,1,1]
   }
   n11 = "Constraint - dyn_l_shoulder_fr_mstr (PPPPPP)"!(
    ComponentPose=@n10
    Alpha=1
    ConstraintWeights_0=0.0441
    ConstraintWeights_1:float=0.3172
    ConstraintWeights_2:float=0.4063
    ConstraintWeights_3:float=0.219
    ConstraintWeights_4:float=0.0003
    ConstraintWeights_5:float=0.0131
   ) -> (Pose) {
    alphaBoolBlend {blendOption=Linear}
    boneToModify {boneName=dyn_l_shoulder_fr_mstr}
    constraintSetup [
     {targetBone {boneName=upperarm_twist_01_l}}
     {targetBone {boneName=spine_04}}
     {targetBone {boneName=spine_05}}
     {targetBone {boneName=clavicle_l}}
     {targetBone {boneName=deform_upperarm_up_l}}
     {targetBone {boneName=deform_upperarm_fr_l}}
    ]
    constraintWeights [1,1,1,1,1,1]
   }
   n12 = "Constraint - dyn_r_shoulder_bk_mstr (PPPPPP)"!(
    ComponentPose=@n11
    Alpha=1
    ConstraintWeights_0=0.0764
    ConstraintWeights_1:float=0
    ConstraintWeights_2:float=0.5285
    ConstraintWeights_3:float=0.3176
    ConstraintWeights_4:float=0.0771
    ConstraintWeights_5:float=0.0003
   ) -> (Pose) {
    alphaBoolBlend {blendOption=Linear}
    boneToModify {boneName=dyn_r_shoulder_bk_mstr}
    constraintSetup [
     {targetBone {boneName=spine_04}}
     {targetBone {boneName=spine_03}}
     {targetBone {boneName=spine_05}}
     {targetBone {boneName=clavicle_r}}
     {targetBone {boneName=upperarm_twist_01_r}}
     {targetBone {boneName=deform_upperarm_up_r}}
    ]
    constraintWeights [1,1,1,1,1,1]
   }
   n13 = "Constraint - dyn_l_shoulder_bk_mstr (PPPPPPPPP)"!(
    ComponentPose=@n12
    Alpha=1
    ConstraintWeights_0=0.0741
    ConstraintWeights_1:float=0
    ConstraintWeights_2:float=0.0675
    ConstraintWeights_3:float=0.0003
    ConstraintWeights_4:float=0.5296
    ConstraintWeights_5:float=0.3277
    ConstraintWeights_6:float=0
    ConstraintWeights_7:float=0.0008
    ConstraintWeights_8:float=0
   ) -> (Pose) {
    alphaBoolBlend {blendOption=Linear}
    boneToModify {boneName=dyn_l_shoulder_bk_mstr}
    constraintSetup [
     {targetBone {boneName=upperarm_twist_01_l}}
     {targetBone {boneName=spine_02}}
     {targetBone {boneName=spine_04}}
     {targetBone {boneName=spine_03}}
     {targetBone {boneName=spine_05}}
     {targetBone {boneName=clavicle_l}}
     {targetBone {boneName=deform_upperarm_up_l}}
     {targetBone {boneName=deform_upperarm_bk_l}}
     {targetBone {boneName=deform_upperarm_fr_l}}
    ]
    constraintWeights [1,1,1,1,1,1,1,1,1]
   }
   n14 = "Constraint - dyn_l_wrist_guard_mstr (PPPPP)"!(
    ComponentPose=@n13
    Alpha=1
    ConstraintWeights_0=0.0556
    ConstraintWeights_1:float=0.0007
    ConstraintWeights_2:float=0.9225
    ConstraintWeights_3:float=0.013
    ConstraintWeights_4:float=0.0083
   ) -> (Pose) {
    alphaBoolBlend {blendOption=Linear}
    boneToModify {boneName=dyn_l_wrist_guard_mstr}
    constraintSetup [
     {targetBone {boneName=hand_l}}
     {targetBone {boneName=middle_metacarpal_l}}
     {targetBone {boneName=lowerarm_twist_01_l}}
     {targetBone {boneName=lowerarm_l}}
     {targetBone {boneName=lowerarm_twist_02_l}}
    ]
    constraintWeights [1,1,1,1,1]
   }
   n15 = "Constraint - dyn_r_wrist_guard_mstr (PPPPP)"!(
    ComponentPose=@n14
    Alpha=1
    ConstraintWeights_0=0.0083
    ConstraintWeights_1:float=0.9225
    ConstraintWeights_2:float=0.013
    ConstraintWeights_3:float=0.0556
    ConstraintWeights_4:float=0.0007
   ) -> (Pose) {
    alphaBoolBlend {blendOption=Linear}
    boneToModify {boneName=dyn_r_wrist_guard_mstr}
    constraintSetup [
     {targetBone {boneName=lowerarm_twist_02_r}}
     {targetBone {boneName=lowerarm_twist_01_r}}
     {targetBone {boneName=lowerarm_r}}
     {targetBone {boneName=hand_r}}
     {targetBone {boneName=middle_metacarpal_r}}
    ]
    constraintWeights [1,1,1,1,1]
   }
   n16 = Control Rig!(Source=Component To Local(ComponentPose=@n15), Alpha, Dual_Pistols_Twist:double=Get dual_pistols() -> (dual_pistols:double)) -> (Pose) {
    alphaBoolBlend {blendOption=Linear}
    controlRigAssetReference {blueprintRigClass=&/F_MED_RoyalDuskBlaze_Skirt_CtrlRig.F_MED_RoyalDuskBlaze_Skirt_CtrlRig_C}
   }
   n17 = "Transform (Modify) Bone - Bone: dyn_Cloth_volume_l_A"(ComponentPose=Local To Component(LocalPose=@n16), Alpha=1, Translation="0,0,0", Rotation="0,10,0", Scale="1,1,1") {
    rotationMode=BMM_Additive
    rotationSpace=BCS_BoneSpace
    scaleMode=BMM_Additive
    scaleSpace=BCS_BoneSpace
    translationMode=BMM_Additive
    translationSpace=BCS_BoneSpace
    alphaBoolBlend {blendOption=Linear}
    boneToModify {boneName=dyn_Cloth_volume_l_A}
    rotation {yaw=10}
   }
   n18 = "Transform (Modify) Bone - Bone: dyn_Cloth_volume_l_A01"(ComponentPose=@n17, Alpha=1, Translation="0,0,0", Rotation="0,10,0", Scale="1,1,1") {
    rotationMode=BMM_Additive
    rotationSpace=BCS_BoneSpace
    scaleMode=BMM_Additive
    scaleSpace=BCS_BoneSpace
    translationMode=BMM_Additive
    translationSpace=BCS_BoneSpace
    alphaBoolBlend {blendOption=Linear}
    boneToModify {boneName=dyn_Cloth_volume_l_A01}
    rotation {yaw=10}
   }
   n19 = "Transform (Modify) Bone - Bone: dyn_Cloth_volume_l_A02"(ComponentPose=@n18, Alpha=1, Translation="0,0,0", Rotation="0,10,0", Scale="1,1,1") {
    rotationMode=BMM_Additive
    rotationSpace=BCS_BoneSpace
    scaleMode=BMM_Additive
    scaleSpace=BCS_BoneSpace
    translationMode=BMM_Additive
    translationSpace=BCS_BoneSpace
    alphaBoolBlend {blendOption=Linear}
    boneToModify {boneName=dyn_Cloth_volume_l_A02}
    rotation {yaw=10}
   }
   n20 = "Transform (Modify) Bone - Bone: dyn_Cloth_volume_l_B"(ComponentPose=@n19, Alpha=1, Translation="0,0,0", Rotation="0,10,0", Scale="1,1,1") {
    rotationMode=BMM_Additive
    rotationSpace=BCS_BoneSpace
    scaleMode=BMM_Additive
    scaleSpace=BCS_BoneSpace
    translationMode=BMM_Additive
    translationSpace=BCS_BoneSpace
    alphaBoolBlend {blendOption=Linear}
    boneToModify {boneName=dyn_Cloth_volume_l_B}
    rotation {yaw=10}
   }
   n21 = "Transform (Modify) Bone - Bone: dyn_Cloth_volume_l_B01"(ComponentPose=@n20, Alpha=1, Translation="0,0,0", Rotation="0,10,0", Scale="1,1,1") {
    rotationMode=BMM_Additive
    rotationSpace=BCS_BoneSpace
    scaleMode=BMM_Additive
    scaleSpace=BCS_BoneSpace
    translationMode=BMM_Additive
    translationSpace=BCS_BoneSpace
    alphaBoolBlend {blendOption=Linear}
    boneToModify {boneName=dyn_Cloth_volume_l_B01}
    rotation {yaw=10}
   }
   n22 = "Transform (Modify) Bone - Bone: dyn_Cloth_volume_l_B02"(ComponentPose=@n21, Alpha=1, Translation="0,0,0", Rotation="0,10,0", Scale="1,1,1") {
    rotationMode=BMM_Additive
    rotationSpace=BCS_BoneSpace
    scaleMode=BMM_Additive
    scaleSpace=BCS_BoneSpace
    translationMode=BMM_Additive
    translationSpace=BCS_BoneSpace
    alphaBoolBlend {blendOption=Linear}
    boneToModify {boneName=dyn_Cloth_volume_l_B02}
    rotation {yaw=10}
   }
   n23 = "Transform (Modify) Bone - Bone: dyn_Cloth_volume_l_C"(ComponentPose=@n22, Alpha=1, Translation="0,0,0", Rotation="0,10,0", Scale="1,1,1") {
    rotationMode=BMM_Additive
    rotationSpace=BCS_BoneSpace
    scaleMode=BMM_Additive
    scaleSpace=BCS_BoneSpace
    translationMode=BMM_Additive
    translationSpace=BCS_BoneSpace
    alphaBoolBlend {blendOption=Linear}
    boneToModify {boneName=dyn_Cloth_volume_l_C}
    rotation {yaw=10}
   }
   n24 = "Transform (Modify) Bone - Bone: dyn_Cloth_volume_l_C01"(ComponentPose=@n23, Alpha=1, Translation="0,0,0", Rotation="0,10,0", Scale="1,1,1") {
    rotationMode=BMM_Additive
    rotationSpace=BCS_BoneSpace
    scaleMode=BMM_Additive
    scaleSpace=BCS_BoneSpace
    translationMode=BMM_Additive
    translationSpace=BCS_BoneSpace
    alphaBoolBlend {blendOption=Linear}
    boneToModify {boneName=dyn_Cloth_volume_l_C01}
    rotation {yaw=10}
   }
   n25 = "Transform (Modify) Bone - Bone: dyn_Cloth_volume_l_C02"(ComponentPose=@n24, Alpha=1, Translation="0,0,0", Rotation="0,10,0", Scale="1,1,1") {
    rotationMode=BMM_Additive
    rotationSpace=BCS_BoneSpace
    scaleMode=BMM_Additive
    scaleSpace=BCS_BoneSpace
    translationMode=BMM_Additive
    translationSpace=BCS_BoneSpace
    alphaBoolBlend {blendOption=Linear}
    boneToModify {boneName=dyn_Cloth_volume_l_C02}
    rotation {yaw=10}
   }
   n26 = "Transform (Modify) Bone - Bone: dyn_Cloth_volume_fn"(ComponentPose=@n25, Alpha=1, Translation="0,0,0", Rotation="0,10,0", Scale="1,1,1") {
    rotationMode=BMM_Additive
    rotationSpace=BCS_BoneSpace
    scaleMode=BMM_Additive
    scaleSpace=BCS_BoneSpace
    translationMode=BMM_Additive
    translationSpace=BCS_BoneSpace
    alphaBoolBlend {blendOption=Linear}
    boneToModify {boneName=dyn_Cloth_volume_fn}
    rotation {yaw=10}
   }
   n27 = "Transform (Modify) Bone - Bone: dyn_Cloth_volume_fn01"(ComponentPose=@n26, Alpha=1, Translation="0,0,0", Rotation="0,10,0", Scale="1,1,1") {
    rotationMode=BMM_Additive
    rotationSpace=BCS_BoneSpace
    scaleMode=BMM_Additive
    scaleSpace=BCS_BoneSpace
    translationMode=BMM_Additive
    translationSpace=BCS_BoneSpace
    alphaBoolBlend {blendOption=Linear}
    boneToModify {boneName=dyn_Cloth_volume_fn01}
    rotation {yaw=10}
   }
   n28 = "Transform (Modify) Bone - Bone: dyn_Cloth_volume_fn02"(ComponentPose=@n27, Alpha=1, Translation="0,0,0", Rotation="0,10,0", Scale="1,1,1") {
    rotationMode=BMM_Additive
    rotationSpace=BCS_BoneSpace
    scaleMode=BMM_Additive
    scaleSpace=BCS_BoneSpace
    translationMode=BMM_Additive
    translationSpace=BCS_BoneSpace
    alphaBoolBlend {blendOption=Linear}
    boneToModify {boneName=dyn_Cloth_volume_fn02}
    rotation {yaw=10}
   }
   n29 = "Transform (Modify) Bone - Bone: dyn_Cloth_volume_r_A"(ComponentPose=@n28, Alpha=1, Translation="0,0,0", Rotation="0,10,0", Scale="1,1,1") {
    rotationMode=BMM_Additive
    rotationSpace=BCS_BoneSpace
    scaleMode=BMM_Additive
    scaleSpace=BCS_BoneSpace
    translationMode=BMM_Additive
    translationSpace=BCS_BoneSpace
    alphaBoolBlend {blendOption=Linear}
    boneToModify {boneName=dyn_Cloth_volume_r_A}
    rotation {yaw=10}
   }
   n30 = "Transform (Modify) Bone - Bone: dyn_Cloth_volume_r_A01"(ComponentPose=@n29, Alpha=1, Translation="0,0,0", Rotation="0,10,0", Scale="1,1,1") {
    rotationMode=BMM_Additive
    rotationSpace=BCS_BoneSpace
    scaleMode=BMM_Additive
    scaleSpace=BCS_BoneSpace
    translationMode=BMM_Additive
    translationSpace=BCS_BoneSpace
    alphaBoolBlend {blendOption=Linear}
    boneToModify {boneName=dyn_Cloth_volume_r_A01}
    rotation {yaw=10}
   }
   n31 = "Transform (Modify) Bone - Bone: dyn_Cloth_volume_r_A02"(ComponentPose=@n30, Alpha=1, Translation="0,0,0", Rotation="0,10,0", Scale="1,1,1") {
    rotationMode=BMM_Additive
    rotationSpace=BCS_BoneSpace
    scaleMode=BMM_Additive
    scaleSpace=BCS_BoneSpace
    translationMode=BMM_Additive
    translationSpace=BCS_BoneSpace
    alphaBoolBlend {blendOption=Linear}
    boneToModify {boneName=dyn_Cloth_volume_r_A02}
    rotation {yaw=10}
   }
   n32 = "Transform (Modify) Bone - Bone: dyn_Cloth_volume_r_B"(ComponentPose=@n31, Alpha=1, Translation="0,0,0", Rotation="0,10,0", Scale="1,1,1") {
    rotationMode=BMM_Additive
    rotationSpace=BCS_BoneSpace
    scaleMode=BMM_Additive
    scaleSpace=BCS_BoneSpace
    translationMode=BMM_Additive
    translationSpace=BCS_BoneSpace
    alphaBoolBlend {blendOption=Linear}
    boneToModify {boneName=dyn_Cloth_volume_r_B}
    rotation {yaw=10}
   }
   n33 = "Transform (Modify) Bone - Bone: dyn_Cloth_volume_r_B01"(ComponentPose=@n32, Alpha=1, Translation="0,0,0", Rotation="0,10,0", Scale="1,1,1") {
    rotationMode=BMM_Additive
    rotationSpace=BCS_BoneSpace
    scaleMode=BMM_Additive
    scaleSpace=BCS_BoneSpace
    translationMode=BMM_Additive
    translationSpace=BCS_BoneSpace
    alphaBoolBlend {blendOption=Linear}
    boneToModify {boneName=dyn_Cloth_volume_r_B01}
    rotation {yaw=10}
   }
   n34 = "Transform (Modify) Bone - Bone: dyn_Cloth_volume_r_B02"(ComponentPose=@n33, Alpha=1, Translation="0,0,0", Rotation="0,10,0", Scale="1,1,1") {
    rotationMode=BMM_Additive
    rotationSpace=BCS_BoneSpace
    scaleMode=BMM_Additive
    scaleSpace=BCS_BoneSpace
    translationMode=BMM_Additive
    translationSpace=BCS_BoneSpace
    alphaBoolBlend {blendOption=Linear}
    boneToModify {boneName=dyn_Cloth_volume_r_B02}
    rotation {yaw=10}
   }
   n35 = "Transform (Modify) Bone - Bone: dyn_Cloth_volume_r_C"(ComponentPose=@n34, Alpha=1, Translation="0,0,0", Rotation="0,10,0", Scale="1,1,1") {
    rotationMode=BMM_Additive
    rotationSpace=BCS_BoneSpace
    scaleMode=BMM_Additive
    scaleSpace=BCS_BoneSpace
    translationMode=BMM_Additive
    translationSpace=BCS_BoneSpace
    alphaBoolBlend {blendOption=Linear}
    boneToModify {boneName=dyn_Cloth_volume_r_C}
    rotation {yaw=10}
   }
   n36 = "Transform (Modify) Bone - Bone: dyn_Cloth_volume_r_C01"(ComponentPose=@n35, Alpha=1, Translation="0,0,0", Rotation="0,10,0", Scale="1,1,1") {
    rotationMode=BMM_Additive
    rotationSpace=BCS_BoneSpace
    scaleMode=BMM_Additive
    scaleSpace=BCS_BoneSpace
    translationMode=BMM_Additive
    translationSpace=BCS_BoneSpace
    alphaBoolBlend {blendOption=Linear}
    boneToModify {boneName=dyn_Cloth_volume_r_C01}
    rotation {yaw=10}
   }
   n37 = "Transform (Modify) Bone - Bone: dyn_Cloth_volume_r_C02"(ComponentPose=@n36, Alpha=1, Translation="0,0,0", Rotation="0,10,0", Scale="1,1,1") {
    rotationMode=BMM_Additive
    rotationSpace=BCS_BoneSpace
    scaleMode=BMM_Additive
    scaleSpace=BCS_BoneSpace
    translationMode=BMM_Additive
    translationSpace=BCS_BoneSpace
    alphaBoolBlend {blendOption=Linear}
    boneToModify {boneName=dyn_Cloth_volume_r_C02}
    rotation {yaw=10}
   }
   n38 = "Transform (Modify) Bone - Bone: dyn_Cloth_volume_bk"(ComponentPose=@n37, Alpha=1, Translation="0,0,0", Rotation="0,10,0", Scale="1,1,1") {
    rotationMode=BMM_Additive
    rotationSpace=BCS_BoneSpace
    scaleMode=BMM_Additive
    scaleSpace=BCS_BoneSpace
    translationMode=BMM_Additive
    translationSpace=BCS_BoneSpace
    alphaBoolBlend {blendOption=Linear}
    boneToModify {boneName=dyn_Cloth_volume_bk}
    rotation {yaw=10}
   }
   n39 = "Transform (Modify) Bone - Bone: dyn_Cloth_volume_bk01"(ComponentPose=@n38, Alpha=1, Translation="0,0,0", Rotation="0,10,0", Scale="1,1,1") {
    rotationMode=BMM_Additive
    rotationSpace=BCS_BoneSpace
    scaleMode=BMM_Additive
    scaleSpace=BCS_BoneSpace
    translationMode=BMM_Additive
    translationSpace=BCS_BoneSpace
    alphaBoolBlend {blendOption=Linear}
    boneToModify {boneName=dyn_Cloth_volume_bk01}
    rotation {yaw=10}
   }
   n40 = "Transform (Modify) Bone - Bone: dyn_Cloth_volume_bk02"(ComponentPose=@n39, Alpha=1, Translation="0,0,0", Rotation="0,10,0", Scale="1,1,1") {
    rotationMode=BMM_Additive
    rotationSpace=BCS_BoneSpace
    scaleMode=BMM_Additive
    scaleSpace=BCS_BoneSpace
    translationMode=BMM_Additive
    translationSpace=BCS_BoneSpace
    alphaBoolBlend {blendOption=Linear}
    boneToModify {boneName=dyn_Cloth_volume_bk02}
    rotation {yaw=10}
   }
   n41 = Anim Dynamics - Bone: dyn_Cloth_volume_l_A!(ComponentPose=@n40, Alpha=1, GravityOverride:Vector="0,0,500", ExternalForce="0,0,0") -> (Pose) {
    angularDampingOverride=0.99
    angularSpringConstant=800
    bAngularSpring=true
    bGravityOverrideInSimSpace=true
    bOverrideAngularDamping=true
    bUseGravityOverride=true
    gravityScale=0.3
    simulationSpace=BoneRelative
    alphaBoolBlend {blendOption=Linear}
    boundBone {boneName=dyn_Cloth_volume_l_A}
    componentAppliedLinearAccClamp=<2500 2500 1500>
    componentLinearAccScale=<0.5 0.5 0.5>
    componentLinearVelScale=<0.55 0.55 0.55>
    gravityOverride {z=500}
    physicsBodyDefinitions [{
      sphereCollisionRadius=0
      boundBone {boneName=dyn_Cloth_volume_l_A}
      boxExtents=<45 45 45>
      constraintSetup {angularLimitsMax=<0 0 20>, angularLimitsMin=<0 0 -10>, angularTarget=<0 0 0>}
      localJointOffset=<-5 0.2 -0.1>
     }]
    relativeSpaceBone {boneName=root}
   }
   n42 = Anim Dynamics - Bone: dyn_Cloth_volume_l_A01!(ComponentPose=@n41, Alpha=1, GravityOverride:Vector="0,0,500", ExternalForce="0,0,0") -> (Pose) {
    angularDampingOverride=0.99
    angularSpringConstant=800
    bAngularSpring=true
    bGravityOverrideInSimSpace=true
    bOverrideAngularDamping=true
    bUseGravityOverride=true
    gravityScale=0.3
    simulationSpace=BoneRelative
    alphaBoolBlend {blendOption=Linear}
    boundBone {boneName=dyn_Cloth_volume_l_A01}
    componentAppliedLinearAccClamp=<2500 2500 1500>
    componentLinearAccScale=<0.5 0.5 0.5>
    componentLinearVelScale=<0.55 0.55 0.55>
    gravityOverride {z=500}
    physicsBodyDefinitions [{
      sphereCollisionRadius=0
      boundBone {boneName=dyn_Cloth_volume_l_A01}
      boxExtents=<40 40 40>
      constraintSetup {angularLimitsMax=<0 0 20>, angularLimitsMin=<0 0 -10>, angularTarget=<0 0 0>}
      localJointOffset=<-5 0.2 -0.1>
     }]
    relativeSpaceBone {boneName=root}
   }
   n43 = Anim Dynamics - Bone: dyn_Cloth_volume_l_A02!(ComponentPose=@n42, Alpha=1, GravityOverride:Vector="0,0,500", ExternalForce="0,0,0") -> (Pose) {
    angularDampingOverride=0.99
    angularSpringConstant=800
    bAngularSpring=true
    bGravityOverrideInSimSpace=true
    bOverrideAngularDamping=true
    bUseGravityOverride=true
    gravityScale=0.3
    simulationSpace=BoneRelative
    alphaBoolBlend {blendOption=Linear}
    boundBone {boneName=dyn_Cloth_volume_l_A02}
    componentAppliedLinearAccClamp=<2500 2500 1500>
    componentLinearAccScale=<0.5 0.5 0.5>
    componentLinearVelScale=<0.55 0.55 0.55>
    gravityOverride {z=500}
    physicsBodyDefinitions [{
      sphereCollisionRadius=0
      boundBone {boneName=dyn_Cloth_volume_l_A02}
      boxExtents=<20 20 20>
      constraintSetup {angularLimitsMax=<0 0 20>, angularLimitsMin=<0 0 -10>, angularTarget=<0 0 0>}
      localJointOffset=<-5 0.2 -0.1>
     }]
    relativeSpaceBone {boneName=root}
   }
   n44 = Anim Dynamics - Bone: dyn_Cloth_volume_l_B!(ComponentPose=@n43, Alpha=1, GravityOverride:Vector="0,0,500", ExternalForce="0,0,0") -> (Pose) {
    angularDampingOverride=0.99
    angularSpringConstant=800
    bAngularSpring=true
    bGravityOverrideInSimSpace=true
    bOverrideAngularDamping=true
    bUseGravityOverride=true
    gravityScale=0.3
    simulationSpace=BoneRelative
    alphaBoolBlend {blendOption=Linear}
    boundBone {boneName=dyn_Cloth_volume_l_B}
    componentAppliedLinearAccClamp=<2500 2500 1500>
    componentLinearAccScale=<0.5 0.5 0.5>
    componentLinearVelScale=<0.55 0.55 0.55>
    gravityOverride {z=500}
    physicsBodyDefinitions [{
      sphereCollisionRadius=0
      boundBone {boneName=dyn_Cloth_volume_l_B}
      boxExtents=<45 45 45>
      constraintSetup {angularLimitsMax=<2 3.5 20>, angularLimitsMin=<-2 -3.5 -10>, angularTarget=<0 0 0>}
      localJointOffset=<-5 0.2 -0.1>
     }]
    relativeSpaceBone {boneName=root}
   }
   n45 = Anim Dynamics - Bone: dyn_Cloth_volume_l_B01!(ComponentPose=@n44, Alpha=1, GravityOverride:Vector="0,0,500", ExternalForce="0,0,0") -> (Pose) {
    angularDampingOverride=0.99
    angularSpringConstant=800
    bAngularSpring=true
    bGravityOverrideInSimSpace=true
    bOverrideAngularDamping=true
    bUseGravityOverride=true
    gravityScale=0.3
    simulationSpace=BoneRelative
    alphaBoolBlend {blendOption=Linear}
    boundBone {boneName=dyn_Cloth_volume_l_B01}
    componentAppliedLinearAccClamp=<2500 2500 1500>
    componentLinearAccScale=<0.5 0.5 0.5>
    componentLinearVelScale=<0.55 0.55 0.55>
    gravityOverride {z=500}
    physicsBodyDefinitions [{
      sphereCollisionRadius=0
      boundBone {boneName=dyn_Cloth_volume_l_B01}
      boxExtents=<40 40 40>
      constraintSetup {angularLimitsMax=<2 3.5 20>, angularLimitsMin=<-2 -3.5 -10>, angularTarget=<0 0 0>}
      localJointOffset=<-5 0.2 -0.1>
     }]
    relativeSpaceBone {boneName=root}
    sphericalLimits [{limitRadius=22, drivingBone {boneName=lowerarm_l}, sphereLocalOffset=<25 0 17>}]
   }
   n46 = Anim Dynamics - Bone: dyn_Cloth_volume_l_B02!(ComponentPose=@n45, Alpha=1, GravityOverride:Vector="0,0,500", ExternalForce="0,0,0") -> (Pose) {
    angularDampingOverride=0.99
    angularSpringConstant=800
    bAngularSpring=true
    bGravityOverrideInSimSpace=true
    bOverrideAngularDamping=true
    bUseGravityOverride=true
    bUseSphericalLimits=true
    gravityScale=0.3
    simulationSpace=BoneRelative
    alphaBoolBlend {blendOption=Linear}
    boundBone {boneName=dyn_Cloth_volume_l_B02}
    componentAppliedLinearAccClamp=<2500 2500 1500>
    componentLinearAccScale=<0.5 0.5 0.5>
    componentLinearVelScale=<0.55 0.55 0.55>
    gravityOverride {z=500}
    physicsBodyDefinitions [{
      sphereCollisionRadius=0
      boundBone {boneName=dyn_Cloth_volume_l_B02}
      boxExtents=<20 20 20>
      constraintSetup {angularLimitsMax=<2 3.5 20>, angularLimitsMin=<-2 -3.5 -10>, angularTarget=<0 0 0>}
      localJointOffset=<-5 0.2 -0.1>
     }]
    relativeSpaceBone {boneName=root}
    sphericalLimits [{limitRadius=22, drivingBone {boneName=lowerarm_l}, sphereLocalOffset=<25 0 17>}]
   }
   n47 = Anim Dynamics - Bone: dyn_Cloth_volume_l_C!(ComponentPose=@n46, Alpha=1, GravityOverride:Vector="0,0,500", ExternalForce="0,0,0") -> (Pose) {
    angularDampingOverride=0.99
    angularSpringConstant=800
    bAngularSpring=true
    bGravityOverrideInSimSpace=true
    bOverrideAngularDamping=true
    bUseGravityOverride=true
    gravityScale=0.3
    simulationSpace=BoneRelative
    alphaBoolBlend {blendOption=Linear}
    boundBone {boneName=dyn_Cloth_volume_l_C}
    componentAppliedLinearAccClamp=<2500 2500 1500>
    componentLinearAccScale=<0.5 0.5 0.5>
    componentLinearVelScale=<0.55 0.55 0.55>
    gravityOverride {z=500}
    physicsBodyDefinitions [{
      sphereCollisionRadius=0
      boundBone {boneName=dyn_Cloth_volume_l_C}
      boxExtents=<45 45 45>
      constraintSetup {angularLimitsMax=<0 5 20>, angularLimitsMin=<0 -5 -10>, angularTarget=<0 0 0>}
      localJointOffset=<-5 0.2 -0.1>
     }]
    relativeSpaceBone {boneName=root}
   }
   n48 = Anim Dynamics - Bone: dyn_Cloth_volume_l_C(ComponentPose=@n47, Alpha=1, GravityOverride="0,0,500", ExternalForce="0,0,0") {
    angularDampingOverride=0.99
    angularSpringConstant=800
    bAngularSpring=true
    bGravityOverrideInSimSpace=true
    bOverrideAngularDamping=true
    bUseGravityOverride=true
    gravityScale=0.3
    simulationSpace=BoneRelative
    alphaBoolBlend {blendOption=Linear}
    boundBone {boneName=dyn_Cloth_volume_l_C}
    componentAppliedLinearAccClamp=<2500 2500 1500>
    componentLinearAccScale=<0.5 0.5 0.5>
    componentLinearVelScale=<0.55 0.55 0.55>
    gravityOverride {z=500}
    physicsBodyDefinitions [{
      sphereCollisionRadius=0
      boundBone {boneName=dyn_Cloth_volume_l_C}
      boxExtents=<40 40 40>
      constraintSetup {angularLimitsMax=<0 5 20>, angularLimitsMin=<0 -5 -10>, angularTarget=<0 0 0>}
      localJointOffset=<-5 0.2 -0.1>
     }]
    relativeSpaceBone {boneName=root}
   }
   n49 = Anim Dynamics - Bone: dyn_Cloth_volume_l_C02!(ComponentPose=@n48, Alpha=1, GravityOverride:Vector="0,0,500", ExternalForce="0,0,0") -> (Pose) {
    angularDampingOverride=0.99
    angularSpringConstant=800
    bAngularSpring=true
    bGravityOverrideInSimSpace=true
    bOverrideAngularDamping=true
    bUseGravityOverride=true
    gravityScale=0.3
    simulationSpace=BoneRelative
    alphaBoolBlend {blendOption=Linear}
    boundBone {boneName=dyn_Cloth_volume_l_C02}
    componentAppliedLinearAccClamp=<2500 2500 1500>
    componentLinearAccScale=<0.5 0.5 0.5>
    componentLinearVelScale=<0.55 0.55 0.55>
    gravityOverride {z=500}
    physicsBodyDefinitions [{
      sphereCollisionRadius=0
      boundBone {boneName=dyn_Cloth_volume_l_C02}
      boxExtents=<20 20 20>
      constraintSetup {angularLimitsMax=<0 5 20>, angularLimitsMin=<0 -5 -15>, angularTarget=<0 0 0>}
      localJointOffset=<-5 0.2 -0.1>
     }]
    relativeSpaceBone {boneName=root}
   }
   n50 = Anim Dynamics - Bone: dyn_Cloth_volume_fn!(ComponentPose=@n49, Alpha=1, GravityOverride:Vector="0,0,500", ExternalForce="0,0,0") -> (Pose) {
    angularDampingOverride=0.99
    angularSpringConstant=800
    bAngularSpring=true
    bGravityOverrideInSimSpace=true
    bOverrideAngularDamping=true
    bUseGravityOverride=true
    gravityScale=0.3
    simulationSpace=BoneRelative
    alphaBoolBlend {blendOption=Linear}
    boundBone {boneName=dyn_Cloth_volume_fn}
    componentAppliedLinearAccClamp=<2500 2500 1500>
    componentLinearAccScale=<0.5 0.5 0.5>
    componentLinearVelScale=<0.55 0.55 0.55>
    gravityOverride {z=500}
    physicsBodyDefinitions [{
      sphereCollisionRadius=0
      boundBone {boneName=dyn_Cloth_volume_fn}
      boxExtents=<55 55 55>
      constraintSetup {angularLimitsMax=<0 3.5 20>, angularLimitsMin=<0 -3.5 -10>, angularTarget=<0 0 0>}
      localJointOffset=<-5 0.2 -0.1>
     }]
    relativeSpaceBone {boneName=root}
   }
   n51 = Anim Dynamics - Bone: dyn_Cloth_volume_fn01!(ComponentPose=@n50, Alpha=1, GravityOverride:Vector="0,0,500", ExternalForce="0,0,0") -> (Pose) {
    angularDampingOverride=0.99
    angularSpringConstant=800
    bAngularSpring=true
    bGravityOverrideInSimSpace=true
    bOverrideAngularDamping=true
    bUseGravityOverride=true
    gravityScale=0.3
    simulationSpace=BoneRelative
    alphaBoolBlend {blendOption=Linear}
    boundBone {boneName=dyn_Cloth_volume_fn01}
    componentAppliedLinearAccClamp=<2500 2500 1500>
    componentLinearAccScale=<0.5 0.5 0.5>
    componentLinearVelScale=<0.55 0.55 0.55>
    gravityOverride {z=500}
    physicsBodyDefinitions [{
      sphereCollisionRadius=0
      boundBone {boneName=dyn_Cloth_volume_fn01}
      boxExtents=<40 40 40>
      constraintSetup {angularLimitsMax=<0 3.5 20>, angularLimitsMin=<0 -3.5 -10>, angularTarget=<0 0 0>}
      localJointOffset=<-5 0.2 -0.1>
     }]
    relativeSpaceBone {boneName=root}
   }
   n52 = Anim Dynamics - Bone: dyn_Cloth_volume_fn02!(ComponentPose=@n51, Alpha=1, GravityOverride:Vector="0,0,500", ExternalForce="0,0,0") -> (Pose) {
    angularDampingOverride=0.99
    angularSpringConstant=800
    bAngularSpring=true
    bGravityOverrideInSimSpace=true
    bOverrideAngularDamping=true
    bUseGravityOverride=true
    gravityScale=0.3
    simulationSpace=BoneRelative
    alphaBoolBlend {blendOption=Linear}
    boundBone {boneName=dyn_Cloth_volume_fn02}
    componentAppliedLinearAccClamp=<2500 2500 1500>
    componentLinearAccScale=<0.5 0.5 0.5>
    componentLinearVelScale=<0.55 0.55 0.55>
    gravityOverride {z=500}
    physicsBodyDefinitions [{
      sphereCollisionRadius=0
      boundBone {boneName=dyn_Cloth_volume_fn02}
      boxExtents=<20 20 20>
      constraintSetup {angularLimitsMax=<0 3.5 20>, angularLimitsMin=<0 -3.5 -10>, angularTarget=<0 0 0>}
      localJointOffset=<-5 0.2 -0.1>
     }]
    relativeSpaceBone {boneName=root}
   }
   n53 = Anim Dynamics - Bone: dyn_Cloth_volume_r_A!(ComponentPose=@n52, Alpha=1, GravityOverride:Vector="0,0,-980", ExternalForce="0,0,0") -> (Pose) {
    angularDampingOverride=0.99
    angularSpringConstant=800
    bAngularSpring=true
    bGravityOverrideInSimSpace=true
    bOverrideAngularDamping=true
    bUseGravityOverride=true
    gravityScale=0.3
    simulationSpace=BoneRelative
    alphaBoolBlend {blendOption=Linear}
    boundBone {boneName=dyn_Cloth_volume_r_A}
    componentAppliedLinearAccClamp=<2500 2500 1500>
    componentLinearAccScale=<0.5 0.5 0.5>
    componentLinearVelScale=<0.55 0.55 0.55>
    gravityOverride {z=-980}
    physicsBodyDefinitions [{
      sphereCollisionRadius=0
      boundBone {boneName=dyn_Cloth_volume_r_A}
      boxExtents=<45 45 45>
      constraintSetup {angularLimitsMax=<1 1 20>, angularLimitsMin=<-1 -1 -10>, angularTarget=<0 0 0>}
      localJointOffset=<-5 0.2 -0.1>
     }]
    relativeSpaceBone {boneName=root}
   }
   n54 = Anim Dynamics - Bone: dyn_Cloth_volume_r_A01!(ComponentPose=@n53, Alpha=1, GravityOverride:Vector="0,0,-980", ExternalForce="0,0,0") -> (Pose) {
    angularDampingOverride=0.99
    angularSpringConstant=800
    bAngularSpring=true
    bGravityOverrideInSimSpace=true
    bOverrideAngularDamping=true
    bUseGravityOverride=true
    gravityScale=0.3
    simulationSpace=BoneRelative
    alphaBoolBlend {blendOption=Linear}
    boundBone {boneName=dyn_Cloth_volume_r_A01}
    componentAppliedLinearAccClamp=<2500 2500 1500>
    componentLinearAccScale=<0.5 0.5 0.5>
    componentLinearVelScale=<0.55 0.55 0.55>
    gravityOverride {z=-980}
    physicsBodyDefinitions [{
      sphereCollisionRadius=0
      boundBone {boneName=dyn_Cloth_volume_r_A01}
      boxExtents=<40 40 40>
      constraintSetup {angularLimitsMax=<1 1 20>, angularLimitsMin=<-1 -1 -10>, angularTarget=<0 0 0>}
      localJointOffset=<-5 0.2 -0.1>
     }]
    relativeSpaceBone {boneName=root}
   }
   n55 = Anim Dynamics - Bone: dyn_Cloth_volume_r_A02!(ComponentPose=@n54, Alpha=1, GravityOverride:Vector="0,0,-980", ExternalForce="0,0,0") -> (Pose) {
    angularDampingOverride=0.99
    angularSpringConstant=800
    bAngularSpring=true
    bGravityOverrideInSimSpace=true
    bOverrideAngularDamping=true
    bUseGravityOverride=true
    gravityScale=0.3
    simulationSpace=BoneRelative
    alphaBoolBlend {blendOption=Linear}
    boundBone {boneName=dyn_Cloth_volume_r_A02}
    componentAppliedLinearAccClamp=<2500 2500 1500>
    componentLinearAccScale=<0.5 0.5 0.5>
    componentLinearVelScale=<0.55 0.55 0.55>
    gravityOverride {z=-980}
    physicsBodyDefinitions [{
      sphereCollisionRadius=0
      boundBone {boneName=dyn_Cloth_volume_r_A02}
      boxExtents=<20 20 20>
      constraintSetup {angularLimitsMax=<1 1 20>, angularLimitsMin=<-1 -1 -10>, angularTarget=<0 0 0>}
      localJointOffset=<-5 0.2 -0.1>
     }]
    relativeSpaceBone {boneName=root}
   }
   n56 = Anim Dynamics - Bone: dyn_Cloth_volume_r_B!(ComponentPose=@n55, Alpha=1, GravityOverride:Vector="0,0,-980", ExternalForce="0,0,0") -> (Pose) {
    angularDampingOverride=0.99
    angularSpringConstant=800
    bAngularSpring=true
    bGravityOverrideInSimSpace=true
    bOverrideAngularDamping=true
    bUseGravityOverride=true
    gravityScale=0.3
    simulationSpace=BoneRelative
    alphaBoolBlend {blendOption=Linear}
    boundBone {boneName=dyn_Cloth_volume_r_B}
    componentAppliedLinearAccClamp=<2500 2500 1500>
    componentLinearAccScale=<0.5 0.5 0.5>
    componentLinearVelScale=<0.55 0.55 0.55>
    gravityOverride {z=-980}
    physicsBodyDefinitions [{
      sphereCollisionRadius=0
      boundBone {boneName=dyn_Cloth_volume_r_B}
      boxExtents=<45 45 45>
      constraintSetup {angularLimitsMax=<2 3.5 20>, angularLimitsMin=<-2 -3.5 -10>, angularTarget=<0 0 0>}
      localJointOffset=<-5 0.2 -0.1>
     }]
    relativeSpaceBone {boneName=root}
   }
   n57 = Anim Dynamics - Bone: dyn_Cloth_volume_r_B01!(ComponentPose=@n56, Alpha=1, GravityOverride:Vector="0,0,-980", ExternalForce="0,0,0") -> (Pose) {
    angularDampingOverride=0.99
    angularSpringConstant=800
    bAngularSpring=true
    bGravityOverrideInSimSpace=true
    bOverrideAngularDamping=true
    bUseGravityOverride=true
    gravityScale=0.3
    simulationSpace=BoneRelative
    alphaBoolBlend {blendOption=Linear}
    boundBone {boneName=dyn_Cloth_volume_r_B01}
    componentAppliedLinearAccClamp=<2500 2500 1500>
    componentLinearAccScale=<0.5 0.5 0.5>
    componentLinearVelScale=<0.55 0.55 0.55>
    gravityOverride {z=-980}
    physicsBodyDefinitions [{
      sphereCollisionRadius=0
      boundBone {boneName=dyn_Cloth_volume_r_B01}
      boxExtents=<40 40 40>
      constraintSetup {angularLimitsMax=<2 3.5 20>, angularLimitsMin=<-2 -3.5 -10>, angularTarget=<0 0 0>}
      localJointOffset=<-5 0.2 -0.1>
     }]
    relativeSpaceBone {boneName=root}
   }
   n58 = Anim Dynamics - Bone: dyn_Cloth_volume_r_B02!(ComponentPose=@n57, Alpha=1, GravityOverride:Vector="0,0,-980", ExternalForce="0,0,0") -> (Pose) {
    angularDampingOverride=0.99
    angularSpringConstant=800
    bAngularSpring=true
    bGravityOverrideInSimSpace=true
    bOverrideAngularDamping=true
    bUseGravityOverride=true
    bUseSphericalLimits=true
    gravityScale=0.3
    simulationSpace=BoneRelative
    alphaBoolBlend {blendOption=Linear}
    boundBone {boneName=dyn_Cloth_volume_r_B02}
    componentAppliedLinearAccClamp=<2500 2500 1500>
    componentLinearAccScale=<0.5 0.5 0.5>
    componentLinearVelScale=<0.55 0.55 0.55>
    gravityOverride {z=-980}
    physicsBodyDefinitions [{
      sphereCollisionRadius=0
      boundBone {boneName=dyn_Cloth_volume_r_B02}
      boxExtents=<20 20 20>
      constraintSetup {angularLimitsMax=<2 3.5 20>, angularLimitsMin=<-2 -3.5 -10>, angularTarget=<0 0 0>}
      localJointOffset=<-5 0.2 -0.1>
     }]
    relativeSpaceBone {boneName=root}
    sphericalLimits [
     {limitRadius=22, drivingBone {boneName=lowerarm_r}, sphereLocalOffset=<-25 0 -17>}
     {limitRadius=0, drivingBone {boneName=None}, sphereLocalOffset=<0 0 0>}
    ]
   }
   n59 = Anim Dynamics - Bone: dyn_Cloth_volume_r_C!(ComponentPose=@n58, Alpha=1, GravityOverride:Vector="0,0,-980", ExternalForce="0,0,0") -> (Pose) {
    angularDampingOverride=0.99
    angularSpringConstant=800
    bAngularSpring=true
    bGravityOverrideInSimSpace=true
    bOverrideAngularDamping=true
    bUseGravityOverride=true
    gravityScale=0.3
    simulationSpace=BoneRelative
    alphaBoolBlend {blendOption=Linear}
    boundBone {boneName=dyn_Cloth_volume_r_C}
    componentAppliedLinearAccClamp=<2500 2500 1500>
    componentLinearAccScale=<0.5 0.5 0.5>
    componentLinearVelScale=<0.55 0.55 0.55>
    gravityOverride {z=-980}
    physicsBodyDefinitions [{
      sphereCollisionRadius=0
      boundBone {boneName=dyn_Cloth_volume_r_C}
      boxExtents=<45 45 45>
      constraintSetup {angularLimitsMax=<0 5 20>, angularLimitsMin=<0 -5 -5>, angularTarget=<0 0 0>}
      localJointOffset=<-5 0.2 -0.1>
     }]
    relativeSpaceBone {boneName=root}
   }
   n60 = Anim Dynamics - Bone: dyn_Cloth_volume_r_C01!(ComponentPose=@n59, Alpha=1, GravityOverride:Vector="0,0,-980", ExternalForce="0,0,0") -> (Pose) {
    angularDampingOverride=0.99
    angularSpringConstant=800
    bAngularSpring=true
    bGravityOverrideInSimSpace=true
    bOverrideAngularDamping=true
    bUseGravityOverride=true
    gravityScale=0.3
    simulationSpace=BoneRelative
    alphaBoolBlend {blendOption=Linear}
    boundBone {boneName=dyn_Cloth_volume_r_C01}
    componentAppliedLinearAccClamp=<2500 2500 1500>
    componentLinearAccScale=<0.5 0.5 0.5>
    componentLinearVelScale=<0.55 0.55 0.55>
    gravityOverride {z=-980}
    physicsBodyDefinitions [{
      sphereCollisionRadius=0
      boundBone {boneName=dyn_Cloth_volume_r_C01}
      boxExtents=<33 33 33>
      constraintSetup {angularLimitsMax=<0 5 20>, angularLimitsMin=<0 -5 -10>, angularTarget=<0 0 0>}
      localJointOffset=<-5 0.2 -0.1>
     }]
    relativeSpaceBone {boneName=root}
   }
   n61 = Anim Dynamics - Bone: dyn_Cloth_volume_r_C02!(ComponentPose=@n60, Alpha=1, GravityOverride:Vector="0,0,-980", ExternalForce="0,0,0") -> (Pose) {
    angularDampingOverride=0.99
    angularSpringConstant=800
    bAngularSpring=true
    bGravityOverrideInSimSpace=true
    bOverrideAngularDamping=true
    bUseGravityOverride=true
    gravityScale=0.3
    simulationSpace=BoneRelative
    alphaBoolBlend {blendOption=Linear}
    boundBone {boneName=dyn_Cloth_volume_r_C02}
    componentAppliedLinearAccClamp=<2500 2500 1500>
    componentLinearAccScale=<0.5 0.5 0.5>
    componentLinearVelScale=<0.55 0.55 0.55>
    gravityOverride {z=-980}
    physicsBodyDefinitions [{
      sphereCollisionRadius=0
      boundBone {boneName=dyn_Cloth_volume_r_C02}
      boxExtents=<16 16 16>
      constraintSetup {angularLimitsMax=<0 5 20>, angularLimitsMin=<0 -5 -15>, angularTarget=<0 0 0>}
      localJointOffset=<-5 0.2 -0.1>
     }]
    relativeSpaceBone {boneName=root}
   }
   n62 = Anim Dynamics - Bone: dyn_Cloth_volume_bk!(ComponentPose=@n61, Alpha=1, GravityOverride:Vector="0,0,-980", ExternalForce="0,0,0") -> (Pose) {
    angularDampingOverride=0.99
    angularSpringConstant=800
    bAngularSpring=true
    bGravityOverrideInSimSpace=true
    bOverrideAngularDamping=true
    bUseGravityOverride=true
    gravityScale=0.3
    simulationSpace=BoneRelative
    alphaBoolBlend {blendOption=Linear}
    boundBone {boneName=dyn_Cloth_volume_bk}
    componentAppliedLinearAccClamp=<2500 2500 1500>
    componentLinearAccScale=<0.5 0.5 0.5>
    componentLinearVelScale=<0.55 0.55 0.55>
    gravityOverride {z=-980}
    physicsBodyDefinitions [{
      sphereCollisionRadius=0
      boundBone {boneName=dyn_Cloth_volume_bk}
      boxExtents=<55 55 55>
      constraintSetup {angularLimitsMax=<0 5 20>, angularLimitsMin=<0 -5 -10>, angularTarget=<0 0 0>}
      localJointOffset=<-5 0.2 -0.1>
     }]
    relativeSpaceBone {boneName=root}
   }
   n63 = Anim Dynamics - Bone: dyn_Cloth_volume_bk01!(ComponentPose=@n62, Alpha=1, GravityOverride:Vector="0,0,-980", ExternalForce="0,0,0") -> (Pose) {
    angularDampingOverride=0.99
    angularSpringConstant=800
    bAngularSpring=true
    bGravityOverrideInSimSpace=true
    bOverrideAngularDamping=true
    bUseGravityOverride=true
    gravityScale=0.3
    simulationSpace=BoneRelative
    alphaBoolBlend {blendOption=Linear}
    boundBone {boneName=dyn_Cloth_volume_bk01}
    componentAppliedLinearAccClamp=<2500 2500 1500>
    componentLinearAccScale=<0.5 0.5 0.5>
    componentLinearVelScale=<0.55 0.55 0.55>
    gravityOverride {z=-980}
    physicsBodyDefinitions [{
      sphereCollisionRadius=0
      boundBone {boneName=dyn_Cloth_volume_bk01}
      boxExtents=<45 45 45>
      constraintSetup {angularLimitsMax=<0 5 20>, angularLimitsMin=<0 -5 -10>, angularTarget=<0 0 0>}
      localJointOffset=<-5 0.2 -0.1>
     }]
    relativeSpaceBone {boneName=root}
   }
   n64 = Anim Dynamics - Bone: dyn_Cloth_volume_bk02!(ComponentPose=@n63, Alpha=1, GravityOverride:Vector="0,0,-980", ExternalForce="0,0,0") -> (Pose) {
    angularDampingOverride=0.99
    angularSpringConstant=800
    bAngularSpring=true
    bGravityOverrideInSimSpace=true
    bOverrideAngularDamping=true
    bUseGravityOverride=true
    gravityScale=0.3
    simulationSpace=BoneRelative
    alphaBoolBlend {blendOption=Linear}
    boundBone {boneName=dyn_Cloth_volume_bk02}
    componentAppliedLinearAccClamp=<2500 2500 1500>
    componentLinearAccScale=<0.5 0.5 0.5>
    componentLinearVelScale=<0.55 0.55 0.55>
    gravityOverride {z=-980}
    physicsBodyDefinitions [{
      sphereCollisionRadius=0
      boundBone {boneName=dyn_Cloth_volume_bk02}
      boxExtents=<25 25 25>
      constraintSetup {angularLimitsMax=<0 5 20>, angularLimitsMin=<0 -5 -15>, angularTarget=<0 0 0>}
      localJointOffset=<-5 0.2 -0.1>
     }]
    relativeSpaceBone {boneName=root}
   }
   n65 = "Constraint - dyn_diamond_r_fr_big_mstr (PPPPPPPPPPPPPPP)"!(
    ComponentPose=@n64
    Alpha=1
    ConstraintWeights_0=0
    ConstraintWeights_1:float=0.0122
    ConstraintWeights_2:float=0.0763
    ConstraintWeights_3:float=0
    ConstraintWeights_4:float=0
    ConstraintWeights_5:float=0
    ConstraintWeights_6:float=0
    ConstraintWeights_7:float=0.2598
    ConstraintWeights_8:float=0.4804
    ConstraintWeights_9:float=0.0428
    ConstraintWeights_10:float=0.1282
    ConstraintWeights_11:float=0
    ConstraintWeights_12:float=0
    ConstraintWeights_13:float=0.0001
    ConstraintWeights_14:float=0.0001
   ) -> (Pose) {
    alphaBoolBlend {blendOption=Linear}
    boneToModify {boneName=dyn_diamond_r_fr_big_mstr}
    constraintSetup [
     {targetBone {boneName=pelvis}}
     {targetBone {boneName=dyn_Cloth_volume_fn01}}
     {targetBone {boneName=dyn_Cloth_volume_fn02}}
     {targetBone {boneName=dyn_Cloth_volume_l_A01}}
     {targetBone {boneName=dyn_Cloth_volume_l_A02}}
     {targetBone {boneName=dyn_Cloth_volume_l_B01}}
     {targetBone {boneName=dyn_Cloth_volume_l_B02}}
     {targetBone {boneName=dyn_Cloth_volume_r_A01}}
     {targetBone {boneName=dyn_Cloth_volume_r_A02}}
     {targetBone {boneName=dyn_Cloth_volume_r_B01}}
     {targetBone {boneName=dyn_Cloth_volume_r_B02}}
     {targetBone {boneName=dyn_Cloth_volume_fn}}
     {targetBone {boneName=dyn_Cloth_volume_l_A}}
     {targetBone {boneName=dyn_Cloth_volume_r_A}}
     {targetBone {boneName=dyn_Cloth_volume_r_B}}
    ]
    constraintWeights [1,1,1,1,1,1,1,1,1,1,1,1,1,1,1]
   }
   n66 = "Constraint - dyn_diamond_r_fr_sml_mstr (PPPPPPPPPPP)"!(
    ComponentPose=@n65
    Alpha=1
    ConstraintWeights_0=0.0005
    ConstraintWeights_1:float=0
    ConstraintWeights_2:float=0.0007
    ConstraintWeights_3:float=0.1637
    ConstraintWeights_4:float=0.2265
    ConstraintWeights_5:float=0.2692
    ConstraintWeights_6:float=0.3213
    ConstraintWeights_7:float=0.0009
    ConstraintWeights_8:float=0.0079
    ConstraintWeights_9:float=0.0006
    ConstraintWeights_10:float=0.0086
   ) -> (Pose) {
    alphaBoolBlend {blendOption=Linear}
    boneToModify {boneName=dyn_diamond_r_fr_sml_mstr}
    constraintSetup [
     {targetBone {boneName=pelvis}}
     {targetBone {boneName=dyn_Cloth_volume_fn01}}
     {targetBone {boneName=dyn_Cloth_volume_fn02}}
     {targetBone {boneName=dyn_Cloth_volume_r_A01}}
     {targetBone {boneName=dyn_Cloth_volume_r_A02}}
     {targetBone {boneName=dyn_Cloth_volume_r_B01}}
     {targetBone {boneName=dyn_Cloth_volume_r_B02}}
     {targetBone {boneName=dyn_Cloth_volume_r_C01}}
     {targetBone {boneName=dyn_Cloth_volume_r_C02}}
     {targetBone {boneName=dyn_Cloth_volume_r_A}}
     {targetBone {boneName=dyn_Cloth_volume_r_B}}
    ]
    constraintWeights [1,1,1,1,1,1,1,1,1,1,1]
   }
   n67 = "Constraint - dyn_diamond_r_bk_sml_mstr (PPPPPPPPPPPP)"!(
    ComponentPose=@n66
    Alpha=1
    ConstraintWeights_0=0.0001
    ConstraintWeights_1:float=0
    ConstraintWeights_2:float=0
    ConstraintWeights_3:float=0.0118
    ConstraintWeights_4:float=0.0411
    ConstraintWeights_5:float=0.3444
    ConstraintWeights_6:float=0.3961
    ConstraintWeights_7:float=0.0696
    ConstraintWeights_8:float=0.1361
    ConstraintWeights_9:float=0
    ConstraintWeights_10:float=0.0009
    ConstraintWeights_11:float=0
   ) -> (Pose) {
    alphaBoolBlend {blendOption=Linear}
    boneToModify {boneName=dyn_diamond_r_bk_sml_mstr}
    constraintSetup [
     {targetBone {boneName=pelvis}}
     {targetBone {boneName=dyn_Cloth_volume_bk01}}
     {targetBone {boneName=dyn_Cloth_volume_bk02}}
     {targetBone {boneName=dyn_Cloth_volume_r_A01}}
     {targetBone {boneName=dyn_Cloth_volume_r_A02}}
     {targetBone {boneName=dyn_Cloth_volume_r_B01}}
     {targetBone {boneName=dyn_Cloth_volume_r_B02}}
     {targetBone {boneName=dyn_Cloth_volume_r_C01}}
     {targetBone {boneName=dyn_Cloth_volume_r_C02}}
     {targetBone {boneName=dyn_Cloth_volume_r_A}}
     {targetBone {boneName=dyn_Cloth_volume_r_B}}
     {targetBone {boneName=dyn_Cloth_volume_r_C}}
    ]
    constraintWeights [1,1,1,1,1,1,1,1,1,1,1,1]
   }
   n68 = "Constraint - dyn_diamond_r_bk_big_mstr (PPPPPPP)"!(
    ComponentPose=@n67
    Alpha=1
    ConstraintWeights_0=0.0014
    ConstraintWeights_1:float=0.0191
    ConstraintWeights_2:float=0.0006
    ConstraintWeights_3:float=0.087
    ConstraintWeights_4:float=0.2978
    ConstraintWeights_5:float=0.0986
    ConstraintWeights_6:float=0.4955
   ) -> (Pose) {
    alphaBoolBlend {blendOption=Linear}
    boneToModify {boneName=dyn_diamond_r_bk_big_mstr}
    constraintSetup [
     {targetBone {boneName=dyn_Cloth_volume_bk01}}
     {targetBone {boneName=dyn_Cloth_volume_bk02}}
     {targetBone {boneName=dyn_Cloth_volume_r_A02}}
     {targetBone {boneName=dyn_Cloth_volume_r_B01}}
     {targetBone {boneName=dyn_Cloth_volume_r_B02}}
     {targetBone {boneName=dyn_Cloth_volume_r_C01}}
     {targetBone {boneName=dyn_Cloth_volume_r_C02}}
    ]
    constraintWeights [1,1,1,1,1,1,1]
   }
   n69 = "Constraint - dyn_diamond_l_bk_big_mstr (PPPPPPPPP)"!(
    ComponentPose=@n68
    Alpha=1
    ConstraintWeights_0=0.0044
    ConstraintWeights_1:float=0.0317
    ConstraintWeights_2:float=0.0007
    ConstraintWeights_3:float=0.087
    ConstraintWeights_4:float=0.2738
    ConstraintWeights_5:float=0.1026
    ConstraintWeights_6:float=0.4998
    ConstraintWeights_7:float=0
    ConstraintWeights_8:float=0
   ) -> (Pose) {
    alphaBoolBlend {blendOption=Linear}
    boneToModify {boneName=dyn_diamond_l_bk_big_mstr}
    constraintSetup [
     {targetBone {boneName=dyn_Cloth_volume_bk01}}
     {targetBone {boneName=dyn_Cloth_volume_bk02}}
     {targetBone {boneName=dyn_Cloth_volume_l_A02}}
     {targetBone {boneName=dyn_Cloth_volume_l_B01}}
     {targetBone {boneName=dyn_Cloth_volume_l_B02}}
     {targetBone {boneName=dyn_Cloth_volume_l_C01}}
     {targetBone {boneName=dyn_Cloth_volume_l_C02}}
     {targetBone {boneName=dyn_Cloth_volume_bk}}
     {targetBone {boneName=dyn_Cloth_volume_l_C}}
    ]
    constraintWeights [1,1,1,1,1,1,1,1,1]
   }
   n70 = "Constraint - dyn_diamond_l_bk_sml_mstr (PPPPPPPPPPPPPP)"!(
    ComponentPose=@n69
    Alpha=1
    ConstraintWeights_0=0.0004
    ConstraintWeights_1:float=0
    ConstraintWeights_2:float=0.0001
    ConstraintWeights_3:float=0.0001
    ConstraintWeights_4:float=0.0134
    ConstraintWeights_5:float=0.039
    ConstraintWeights_6:float=0.3404
    ConstraintWeights_7:float=0.3865
    ConstraintWeights_8:float=0.075
    ConstraintWeights_9:float=0.1439
    ConstraintWeights_10:float=0
    ConstraintWeights_11:float=0.0002
    ConstraintWeights_12:float=0.001
    ConstraintWeights_13:float=0.0001
   ) -> (Pose) {
    alphaBoolBlend {blendOption=Linear}
    boneToModify {boneName=dyn_diamond_l_bk_sml_mstr}
    constraintSetup [
     {targetBone {boneName=pelvis}}
     {targetBone {boneName=dyn_Cloth_volume_fn02}}
     {targetBone {boneName=dyn_Cloth_volume_bk01}}
     {targetBone {boneName=dyn_Cloth_volume_bk02}}
     {targetBone {boneName=dyn_Cloth_volume_l_A01}}
     {targetBone {boneName=dyn_Cloth_volume_l_A02}}
     {targetBone {boneName=dyn_Cloth_volume_l_B01}}
     {targetBone {boneName=dyn_Cloth_volume_l_B02}}
     {targetBone {boneName=dyn_Cloth_volume_l_C01}}
     {targetBone {boneName=dyn_Cloth_volume_l_C02}}
     {targetBone {boneName=dyn_Cloth_volume_bk}}
     {targetBone {boneName=dyn_Cloth_volume_l_A}}
     {targetBone {boneName=dyn_Cloth_volume_l_B}}
     {targetBone {boneName=dyn_Cloth_volume_l_C}}
    ]
    constraintWeights [1,1,1,1,1,1,1,1,1,1,1,1,1,1]
   }
   n71 = "Constraint - dyn_diamond_l_fr_sml_mstr (PPPPPPPPPPPPPPPP)"!(
    ComponentPose=@n70
    Alpha=1
    ConstraintWeights_0=0
    ConstraintWeights_1:float=0.0029
    ConstraintWeights_2:float=0
    ConstraintWeights_3:float=0.0007
    ConstraintWeights_4:float=0.1575
    ConstraintWeights_5:float=0.211
    ConstraintWeights_6:float=0.2807
    ConstraintWeights_7:float=0.3168
    ConstraintWeights_8:float=0.0026
    ConstraintWeights_9:float=0.0109
    ConstraintWeights_10:float=0
    ConstraintWeights_11:float=0
    ConstraintWeights_12:float=0
    ConstraintWeights_13:float=0.0039
    ConstraintWeights_14:float=0.0129
    ConstraintWeights_15:float=0
   ) -> (Pose) {
    alphaBoolBlend {blendOption=Linear}
    boneToModify {boneName=dyn_diamond_l_fr_sml_mstr}
    constraintSetup [
     {targetBone {boneName=spine_01}}
     {targetBone {boneName=pelvis}}
     {targetBone {boneName=dyn_Cloth_volume_fn01}}
     {targetBone {boneName=dyn_Cloth_volume_fn02}}
     {targetBone {boneName=dyn_Cloth_volume_l_A01}}
     {targetBone {boneName=dyn_Cloth_volume_l_A02}}
     {targetBone {boneName=dyn_Cloth_volume_l_B01}}
     {targetBone {boneName=dyn_Cloth_volume_l_B02}}
     {targetBone {boneName=dyn_Cloth_volume_l_C01}}
     {targetBone {boneName=dyn_Cloth_volume_l_C02}}
     {targetBone {boneName=dyn_Cloth_volume_r_A01}}
     {targetBone {boneName=dyn_Cloth_volume_r_A02}}
     {targetBone {boneName=dyn_Cloth_volume_fn}}
     {targetBone {boneName=dyn_Cloth_volume_l_A}}
     {targetBone {boneName=dyn_Cloth_volume_l_B}}
     {targetBone {boneName=dyn_Cloth_volume_l_C}}
    ]
    constraintWeights [1,1,1,1,1,1,1,1,1,1,1,1,1,1,1,1]
   }
   n72 = "Constraint - dyn_diamond_l_fr_big_mstr (PPPPPPPPPPPPPPPPP)"!(
    ComponentPose=@n71
    Alpha=1
    ConstraintWeights_0=0
    ConstraintWeights_1:float=0.0183
    ConstraintWeights_2:float=0.0964
    ConstraintWeights_3:float=0.263
    ConstraintWeights_4:float=0.4765
    ConstraintWeights_5:float=0.0354
    ConstraintWeights_6:float=0.1096
    ConstraintWeights_7:float=0
    ConstraintWeights_8:float=0
    ConstraintWeights_9:float=0
    ConstraintWeights_10:float=0
    ConstraintWeights_11:float=0.0002
    ConstraintWeights_12:float=0
    ConstraintWeights_13:float=0.0004
    ConstraintWeights_14:float=0.0001
    ConstraintWeights_15:float=0
    ConstraintWeights_16:float=0
   ) -> (Pose) {
    alphaBoolBlend {blendOption=Linear}
    boneToModify {boneName=dyn_diamond_l_fr_big_mstr}
    constraintSetup [
     {targetBone {boneName=pelvis}}
     {targetBone {boneName=dyn_Cloth_volume_fn01}}
     {targetBone {boneName=dyn_Cloth_volume_fn02}}
     {targetBone {boneName=dyn_Cloth_volume_l_A01}}
     {targetBone {boneName=dyn_Cloth_volume_l_A02}}
     {targetBone {boneName=dyn_Cloth_volume_l_B01}}
     {targetBone {boneName=dyn_Cloth_volume_l_B02}}
     {targetBone {boneName=dyn_Cloth_volume_r_A01}}
     {targetBone {boneName=dyn_Cloth_volume_r_A02}}
     {targetBone {boneName=dyn_Cloth_volume_r_B01}}
     {targetBone {boneName=dyn_Cloth_volume_r_B02}}
     {targetBone {boneName=dyn_Cloth_volume_fn}}
     {targetBone {boneName=dyn_Cloth_volume_bk}}
     {targetBone {boneName=dyn_Cloth_volume_l_A}}
     {targetBone {boneName=dyn_Cloth_volume_l_B}}
     {targetBone {boneName=dyn_Cloth_volume_l_C}}
     {targetBone {boneName=dyn_Cloth_volume_r_A}}
    ]
    constraintWeights [1,1,1,1,1,1,1,1,1,1,1,1,1,1,1,1,1]
   }
   n73 = "Transform (Modify) Bone - Bone: dyn_Cloth_volume_l_C"(ComponentPose=@n72, Alpha=1, Translation="0,-2,0", Rotation="0,0,0", Scale="1,1,1") {
    rotationMode=BMM_Additive
    rotationSpace=BCS_BoneSpace
    translationMode=BMM_Additive
    translationSpace=BCS_BoneSpace
    alphaBoolBlend {blendOption=Linear}
    boneToModify {boneName=dyn_Cloth_volume_l_C}
    translation {y=-2}
   }
   n74 = "Transform (Modify) Bone - Bone: dyn_Cloth_volume_bk"(ComponentPose=@n73, Alpha=1, Translation="0,-2,0", Rotation="0,0,0", Scale="1,1,1") {
    rotationMode=BMM_Additive
    rotationSpace=BCS_BoneSpace
    translationMode=BMM_Additive
    translationSpace=BCS_BoneSpace
    alphaBoolBlend {blendOption=Linear}
    boneToModify {boneName=dyn_Cloth_volume_bk}
    translation {y=-2}
   }
   n75 = "Transform (Modify) Bone - Bone: dyn_Cloth_volume_r_C"(ComponentPose=@n74, Alpha=1, Translation="0,-2,0", Rotation="0,0,0", Scale="1,1,1") {
    rotationMode=BMM_Additive
    rotationSpace=BCS_BoneSpace
    translationMode=BMM_Additive
    translationSpace=BCS_BoneSpace
    alphaBoolBlend {blendOption=Linear}
    boneToModify {boneName=dyn_Cloth_volume_r_C}
    translation {y=-2}
   }
   n76 = "Transform (Modify) Bone - Bone: dyn_Cloth_volume_l_B"(ComponentPose=@n75, Alpha=1, Translation="0,-1,0", Rotation="0,0,0", Scale="1,1,1") {
    rotationMode=BMM_Additive
    rotationSpace=BCS_BoneSpace
    translationMode=BMM_Additive
    translationSpace=BCS_BoneSpace
    alphaBoolBlend {blendOption=Linear}
    boneToModify {boneName=dyn_Cloth_volume_l_B}
    translation {y=-1}
   }
   n77 = "Transform (Modify) Bone - Bone: dyn_Cloth_volume_r_C"(ComponentPose=@n76, Alpha=1, Translation="0,-1,0", Rotation="0,0,0", Scale="1,1,1") {
    rotationMode=BMM_Additive
    rotationSpace=BCS_BoneSpace
    translationMode=BMM_Additive
    translationSpace=BCS_BoneSpace
    alphaBoolBlend {blendOption=Linear}
    boneToModify {boneName=dyn_Cloth_volume_r_C}
    translation {y=-1}
   }
   n78 = "Constraint - dyn_r_kneepad_master (PPP)"!(ComponentPose=@n77, Alpha=1, ConstraintWeights_0=0.25, ConstraintWeights_1:float=0.75, ConstraintWeights_2:float=0) -> (Pose) {
    alphaBoolBlend {blendOption=Linear}
    boneToModify {boneName=dyn_r_kneepad_master}
    constraintSetup [{targetBone {boneName=thigh_r}}, {targetBone {boneName=calf_r}}, {targetBone {boneName=deform_knee_out_r}}]
    constraintWeights [1,1,1]
   }
   n79 = "Constraint - dyn_l_kneepad_master (PPP)"!(ComponentPose=@n78, Alpha=1, ConstraintWeights_0=0.25, ConstraintWeights_1:float=0.75, ConstraintWeights_2:float=0) -> (Pose) {
    alphaBoolBlend {blendOption=Linear}
    boneToModify {boneName=dyn_l_kneepad_master}
    constraintSetup [{targetBone {boneName=thigh_l}}, {targetBone {boneName=calf_l}}, {targetBone {boneName=deform_knee_out_l}}]
    constraintWeights [1,1,1]
   }
   n80 = RigidBody!(ComponentPose=@n79, Alpha, OverrideWorldGravity:Vector=Get Gravity_Vector() -> (Gravity_Vector:Vector), ExternalForce) -> (Pose) {
    bClampLinearTranslationLimitToRefPose=true
    bOverrideWorldGravity=true
    overridePhysicsAsset=&/F_MED_RoyalDuskBlaze_Physics
    simulationSpace=BaseBoneSpace
    alphaBoolBlend {blendOption=Linear}
    baseBoneRef {boneName=pelvis}
    componentAppliedLinearAccClamp=<5000 5000 5000>
    componentLinearAccScale=<0.4 0.4 0.5>
    overrideWorldGravity {z=-980}
    simSpaceSettings {
     maxAngularAcceleration=100
     maxAngularVelocity=3
     maxLinearAcceleration=700
     maxLinearVelocity=500
     velocityScaleZ=0
     worldAlpha=1
     externalLinearDragV=<0.4 0.4 0.4>
    }
   }
   n81 = "Transform (Modify) Bone - Bone: attach_backpack"!(ComponentPose=@n80, bAlphaBoolEnabled:bool=Get PonytailsOff() -> (PonytailsOff:bool), Translation="0,4,0", Rotation="0,0,0", Scale="1,1,1") -> (Pose) {
    alphaInputType=Bool
    translationMode=BMM_Additive
    translationSpace=BCS_BoneSpace
    alphaBoolBlend {blendOption=Linear}
    boneToModify {boneName=attach_backpack}
    translation {y=4}
   }
   n82 = "Transform (Modify) Bone - Bone: attach_cape"!(ComponentPose=@n81, bAlphaBoolEnabled:bool=Get PonytailsOff(), Translation="0,2.5,0", Rotation="0,0,0", Scale="1,1,1") -> (Pose) {
    alphaInputType=Bool
    translationMode=BMM_Additive
    translationSpace=BCS_BoneSpace
    alphaBoolBlend {blendOption=Linear}
    boneToModify {boneName=attach_cape}
    translation {y=2.5}
   }
   Output Pose(Result=Component To Local(ComponentPose=@n82)) {layerGroup=None}
   Save cached pose 'SkirtDynPose'(Pose=Component To Local()) {cacheName=SkirtDynPose}
  ]}
 EventGraph {nodes [
   Event BlueprintUpdateAnimation()
   > n1 = Sequence()
   > n2 = Gravity Override State Function DEPRECATED(CharacterPartAnimInstance=Self-Reference(), GameTimeInSeconds=GetGameTimeInSeconds(), GravityOverrideParams=Get Gravity Override Params())
   > Set Gravity Override Cape(Gravity Override Cape:Vector=@n2) -> (Output_Get:Vector)
   > Set is_skydiving(is_skydiving:double=SelectFloat(A=1, B=0, bPickA=Get bIsSkyDiving())) -> (Output_Get:double)
   n3 = Cast To FortPlayerAnimInstance(execute=@n1.then_1, Object=GetAnimInstance(self=GetOwnerSkeletalMeshComponent()))
   > Set dual_pistols(dual_pistols:double=0) -> (Output_Get:double)
   > n4 = Switch on EFortWeaponCoreAnimation!(execute, Selection:EFortWeaponCoreAnimation Enum=Get WeaponCoreAnim(self=@n3)) -> (
    Melee:Exec
    Pistol:Exec
    Shotgun:Exec
    PaperBlueprint:Exec
    Rifle:Exec
    MeleeOneHand:Exec
    MachinePistol:Exec
    RocketLauncher:Exec
    GrenadeLauncher:Exec
    GoingCommando:Exec
    AssaultRifle:Exec
    TacticalShotgun:Exec
    SniperRifle:Exec
    TrapPlacement:Exec
    ShoulderLauncher:Exec
    AbilityDecoTool:Exec
    Crossbow:Exec
    C4:Exec
    RemoteControl:Exec
    DualWield:Exec
    AR_BullPup:Exec
    AR_ForwardGrip:Exec
    MedPackPaddles:Exec
    SMG_P90:Exec
    AR_DrumGun:Exec
    Consumable_Small:Exec
    Consumable_Large:Exec
    Balloon:Exec
    MountedTurret:Exec
    CreativeTool:Exec
    ExplosiveBow:Exec
    AshtonIndigo:Exec
    AshtonChicago:Exec
    MeleeDualWield:Exec
    Unarmed:Exec
   )
   n5 = BreakTransform(InTransform=GetDeltaTransformFromRefPose(self=GetOwnerSkeletalMeshComponent(), BoneName=upperarm_twist_01_r, BaseName=clavicle_r))
   Set dual_pistols(execute=@n4.DualWield, dual_pistols=MapRangeClamped(Value=BreakRotator(InRot=@n5.Rotation), InRangeA=50, InRangeB=65))
   n6 = Event OnEndSkydiving()
   Delay(execute=[@n6.then, @n7.then], Duration=0.1)
   > ResetDynamics(InTeleportType=TeleportPhysics)
   > ResetDynamics(InTeleportType=ResetPhysics)
   n7 = Event OnEndPlayingEmote() {is=title K2Node_Event}
   Event OnCharacterCustomizationLoaded()
   > n8 = Branch(Condition=CharacterItemDefinitionHasAnyMetaTag(Tags=Make Array("[0]"=(TagName="Cosmetics.Variant.Property.Mat2"))))
   > Set PonytailsOff(PonytailsOff:bool=true) -> (Output_Get:bool)
   Set PonytailsOff(execute=@n8.else, PonytailsOff=false)
   n9 = BreakTransform(InTransform=GetDeltaTransformFromRefPose(self=GetOwnerSkeletalMeshComponent(), BoneName=upperarm_twist_01_r, BaseName=clavicle_r))
   Set dual_pistols(dual_pistols=MapRangeClamped(Value=BreakRotator(InRot=@n9.Rotation), InRangeA=50, InRangeB=65))
  ]}
]
variables [
 dual_pistols {type=number}
 Gravity Override Cape {type=Vector}
 Gravity Override Params {type=GravityOverrideParamsStruct}
 Gravity_Vector {type=Vector, default=<0 0 -700>}
 is_skydiving {type=number}
 PonytailsOff {type=boolean}
]
