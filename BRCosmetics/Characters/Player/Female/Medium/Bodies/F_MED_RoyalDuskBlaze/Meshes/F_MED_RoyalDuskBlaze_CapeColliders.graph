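# Physics asset of cape colliders for the female medium RoyalDuskBlaze player body, apparently collision shapes for cape simulation.
# Capsules cover head, clavicles, pelvis, spine and feet, with crossed pairs on mid-spine; arm, thigh, calf and neck bodies are empty.
# More capsules sit on dynamic cloth-volume chains and back diamond bones attached to the pelvis.
# All constraints leave swing and twist free, so they impose no angular limits.
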PhysicsAsset /BRCosmetics/Characters/Player/Female/Medium/Bodies/F_MED_RoyalDuskBlaze/Meshes/F_MED_RoyalDuskBlaze_CapeColliders
bodies {
 calf_l []
 calf_r []
 clavicle_l [{shapeType=Capsule, length=6.5623, radius=7.5194, center=<3.9566 -1.0862 -2.1127>, rotation=<rot -77.5566 177.3915 7.7314>}]
 clavicle_r [{shapeType=Capsule, length=6.5623, radius=7.5194, center=<-3.9566 1.0862 2.1127>, rotation=<rot -77.5566 177.3915 -172.2686>}]
 dyn_Cloth_volume_bk01 [{shapeType=Capsule, length=3.0539, radius=6.251, shapeName=dyn_Cloth_volume_bk01_capsule, center=<-5.7782 0 0>, rotation=<rot 90 0 0>}]
 dyn_Cloth_volume_bk02 [{shapeType=Capsule, length=18.1763, radius=7.9969, shapeName=dyn_Cloth_volume_bk02_capsule, center=<-15.9662 0 0>}]
 dyn_Cloth_volume_l_B01 [{shapeType=Capsule, length=9.3092, radius=7.9566, shapeName=dyn_Cloth_volume_l_B01_capsule, center=<-7.7741 0 0>}]
 dyn_Cloth_volume_l_B02 [{shapeType=Capsule, length=11.7804, radius=7.7183, shapeName=dyn_Cloth_volume_l_B02_capsule, center=<-7.7741 0 0>}]
 dyn_Cloth_volume_l_C01 [{shapeType=Capsule, length=2.2167, radius=7.0882, shapeName=dyn_Cloth_volume_l_C01_capsule, center=<-8.2571 0 0>, rotation=<rot 90 0 0>}]
 dyn_Cloth_volume_l_C02 [{shapeType=Capsule, length=12.2954, radius=8.834, shapeName=dyn_Cloth_volume_l_C02_capsule, center=<-8.2571 0 0>}]
 dyn_Cloth_volume_r_B01 [{shapeType=Capsule, length=9.1214, radius=7.589, shapeName=dyn_Cloth_volume_r_B01_capsule, center=<-6.6921 0 0>}]
 dyn_Cloth_volume_r_B02 [{shapeType=Capsule, length=12.1821, radius=7.585, shapeName=dyn_Cloth_volume_r_B02_capsule, center=<-6.6921 0 0>}]
 dyn_Cloth_volume_r_C01 [{shapeType=Capsule, length=1.5515, radius=7.7534, shapeName=dyn_Cloth_volume_r_C01_capsule, center=<-5.7782 0 0>, rotation=<rot 90 0 0>}]
 dyn_Cloth_volume_r_C02 [{shapeType=Capsule, length=11.6301, radius=9.4993, shapeName=dyn_Cloth_volume_r_C02_capsule, center=<-9.5449 0 0>}]
 dyn_diamond_l_bk_big [{
   shapeType=Capsule
   length=12.9212
   radius=3.1185
   shapeName=dyn_diamond_l_bk_big_capsule
   center=<10.0995 0.1185 0.4685>
   rotation=<rot 0 91.0211 -88.6489>
  }]
 dyn_diamond_l_bk_sml [{
   shapeType=Capsule
   length=11.5714
   radius=2.6636
   shapeName=dyn_diamond_l_bk_sml_capsule
   center=<8.2743 -0.0597 -0.0412>
   rotation=<rot 0 87.9078 -89.8951>
  }]
 dyn_diamond_r_bk_big [{
   shapeType=Capsule
   length=12.9212
   radius=3.1185
   shapeName=dyn_diamond_r_bk_big_capsule
   center=<10.5084 0.0432 0.4042>
   rotation=<rot 0 89.9393 -88.6489>
  }]
 dyn_diamond_r_bk_sml [{
   shapeType=Capsule
   length=11.5714
   radius=2.6636
   shapeName=dyn_diamond_r_bk_sml_capsule
   center=<8.7021 0.1557 -0.026>
   rotation=<rot 0 92.0922 -89.8951>
  }]
 foot_l [{shapeType=Capsule, length=11.0765, radius=8.0344, center=<-4.1811 6.0027 -0.338>, rotation=<rot 0.0388 1.4701 -89.3982>}]
 foot_r [{shapeType=Capsule, length=11.0765, radius=8.0344, center=<4.1811 -6.0027 0.338>, rotation=<rot 0.0388 1.4701 90.6018>}]
 head [{shapeType=Capsule, length=0, radius=18, center=<9.0697 0.0746 0>, rotation=<rot 0 -90 0>}]
 lowerarm_l []
 lowerarm_r []
 neck_02 []
 pelvis [{shapeType=Capsule, length=4.0259, radius=16.7223, center=<-2.4471 -2.9207 -0.1981>}]
 spine_01 [{shapeType=Capsule, length=4.3888, radius=14.3638, center=<-8.9021 -0.8973 0>}]
 spine_02 [
  {shapeType=Capsule, length=5.1113, radius=11.4547, center=<-2.6885 -0.05 0>}
  {shapeType=Capsule, length=20.6875, radius=9.3125, center=<-1.8317 1.9362 0>, rotation=<rot 90 0 0>}
 ]
 spine_03 [
  {shapeType=Capsule, length=1.8622, radius=12.282, center=<0.0574 0.9887 0>}
  {shapeType=Capsule, length=20.7188, radius=9.2812, center=<-1.2163 0.549 0>, rotation=<rot 90 180 -175.0732>}
 ]
 spine_04 [
  {shapeType=Capsule, length=5.5715, radius=12.3277, center=<0 0.4184 0>}
  {shapeType=Capsule, length=10.0312, radius=9.2812, center=<-1.2326 1.308 0>, rotation=<rot 90 0 0>}
 ]
 spine_05 [{shapeType=Capsule, length=7.2996, radius=12.2354, center=<1.0979 -0.5836 0>}]
 thigh_l []
 thigh_r []
 upperarm_l []
 upperarm_r []
}
constraints [
 {
  bone1Name=spine_01
  bone2Name=pelvis
  swing1LimitDegrees=45
  swing1Motion=Free
  swing2LimitDegrees=45
  swing2Motion=Free
  twistLimitDegrees=45
  twistMotion=Free
 }
 {
  bone1Name=spine_02
  bone2Name=spine_01
  swing1LimitDegrees=45
  swing1Motion=Free
  swing2LimitDegrees=45
  swing2Motion=Free
  twistLimitDegrees=45
  twistMotion=Free
 }
 {
  bone1Name=spine_03
  bone2Name=spine_02
  swing1LimitDegrees=45
  swing1Motion=Free
  swing2LimitDegrees=45
  swing2Motion=Free
  twistLimitDegrees=45
  twistMotion=Free
 }
 {
  bone1Name=spine_04
  bone2Name=spine_03
  swing1LimitDegrees=45
  swing1Motion=Free
  swing2LimitDegrees=45
  swing2Motion=Free
  twistLimitDegrees=45
  twistMotion=Free
 }
 {
  bone1Name=spine_05
  bone2Name=spine_04
  swing1LimitDegrees=45
  swing1Motion=Free
  swing2LimitDegrees=45
  swing2Motion=Free
  twistLimitDegrees=45
  twistMotion=Free
 }
 {
  bone1Name=clavicle_l
  bone2Name=spine_05
  swing1LimitDegrees=45
  swing1Motion=Free
  swing2LimitDegrees=45
  swing2Motion=Free
  twistLimitDegrees=45
  twistMotion=Free
 }
 {
  bone1Name=upperarm_l
  bone2Name=clavicle_l
  swing1LimitDegrees=45
  swing1Motion=Free
  swing2LimitDegrees=45
  swing2Motion=Free
  twistLimitDegrees=45
  twistMotion=Free
 }
 {
  bone1Name=lowerarm_l
  bone2Name=upperarm_l
  swing1LimitDegrees=45
  swing1Motion=Free
  swing2LimitDegrees=45
  swing2Motion=Free
  twistLimitDegrees=45
  twistMotion=Free
 }
 {
  bone1Name=neck_02
  bone2Name=spine_05
  swing1LimitDegrees=45
  swing1Motion=Free
  swing2LimitDegrees=45
  swing2Motion=Free
  twistLimitDegrees=45
  twistMotion=Free
 }
 {
  bone1Name=head
  bone2Name=neck_02
  swing1LimitDegrees=45
  swing1Motion=Free
  swing2LimitDegrees=45
  swing2Motion=Free
  twistLimitDegrees=45
  twistMotion=Free
 }
 {
  bone1Name=calf_l
  bone2Name=thigh_l
  swing1LimitDegrees=45
  swing1Motion=Free
  swing2LimitDegrees=45
  swing2Motion=Free
  twistLimitDegrees=45
  twistMotion=Free
 }
 {
  bone1Name=thigh_l
  bone2Name=pelvis
  swing1LimitDegrees=45
  swing1Motion=Free
  swing2LimitDegrees=45
  swing2Motion=Free
  twistLimitDegrees=45
  twistMotion=Free
 }
 {
  bone1Name=foot_l
  bone2Name=calf_l
  swing1LimitDegrees=45
  swing1Motion=Free
  swing2LimitDegrees=45
  swing2Motion=Free
  twistLimitDegrees=45
  twistMotion=Free
 }
 {
  bone1Name=lowerarm_r
  bone2Name=upperarm_r
  swing1LimitDegrees=45
  swing1Motion=Free
  swing2LimitDegrees=45
  swing2Motion=Free
  twistLimitDegrees=45
  twistMotion=Free
 }
 {
  bone1Name=upperarm_r
  bone2Name=clavicle_r
  swing1LimitDegrees=45
  swing1Motion=Free
  swing2LimitDegrees=45
  swing2Motion=Free
  twistLimitDegrees=45
  twistMotion=Free
 }
 {
  bone1Name=clavicle_r
  bone2Name=spine_05
  swing1LimitDegrees=45
  swing1Motion=Free
  swing2LimitDegrees=45
  swing2Motion=Free
  twistLimitDegrees=45
  twistMotion=Free
 }
 {
  bone1Name=thigh_r
  bone2Name=pelvis
  swing1LimitDegrees=45
  swing1Motion=Free
  swing2LimitDegrees=45
  swing2Motion=Free
  twistLimitDegrees=45
  twistMotion=Free
 }
 {
  bone1Name=foot_r
  bone2Name=calf_r
  swing1LimitDegrees=45
  swing1Motion=Free
  swing2LimitDegrees=45
  swing2Motion=Free
  twistLimitDegrees=45
  twistMotion=Free
 }
 {
  bone1Name=calf_r
  bone2Name=thigh_r
  swing1LimitDegrees=45
  swing1Motion=Free
  swing2LimitDegrees=45
  swing2Motion=Free
  twistLimitDegrees=45
  twistMotion=Free
 }
 {
  bone1Name=dyn_Cloth_volume_l_C01
  bone2Name=pelvis
  swing1LimitDegrees=45
  swing1Motion=Free
  swing2LimitDegrees=45
  swing2Motion=Free
  twistLimitDegrees=45
  twistMotion=Free
 }
 {
  bone1Name=dyn_Cloth_volume_r_C01
  bone2Name=pelvis
  swing1LimitDegrees=45
  swing1Motion=Free
  swing2LimitDegrees=45
  swing2Motion=Free
  twistLimitDegrees=45
  twistMotion=Free
 }
 {
  bone1Name=dyn_Cloth_volume_bk01
  bone2Name=pelvis
  swing1LimitDegrees=45
  swing1Motion=Free
  swing2LimitDegrees=45
  swing2Motion=Free
  twistLimitDegrees=45
  twistMotion=Free
 }
 {
  bone1Name=dyn_Cloth_volume_bk02
  bone2Name=dyn_Cloth_volume_bk01
  swing1LimitDegrees=45
  swing1Motion=Free
  swing2LimitDegrees=45
  swing2Motion=Free
  twistLimitDegrees=45
  twistMotion=Free
 }
 {
  bone1Name=dyn_Cloth_volume_l_C02
  bone2Name=dyn_Cloth_volume_l_C01
  swing1LimitDegrees=45
  swing1Motion=Free
  swing2LimitDegrees=45
  swing2Motion=Free
  twistLimitDegrees=45
  twistMotion=Free
 }
 {
  bone1Name=dyn_Cloth_volume_r_C02
  bone2Name=dyn_Cloth_volume_r_C01
  swing1LimitDegrees=45
  swing1Motion=Free
  swing2LimitDegrees=45
  swing2Motion=Free
  twistLimitDegrees=45
  twistMotion=Free
 }
 {
  bone1Name=dyn_diamond_l_bk_big
  bone2Name=pelvis
  swing1LimitDegrees=45
  swing1Motion=Free
  swing2LimitDegrees=45
  swing2Motion=Free
  twistLimitDegrees=45
  twistMotion=Free
 }
 {
  bone1Name=dyn_diamond_l_bk_sml
  bone2Name=pelvis
  swing1LimitDegrees=45
  swing1Motion=Free
  swing2LimitDegrees=45
  swing2Motion=Free
  twistLimitDegrees=45
  twistMotion=Free
 }
 {
  bone1Name=dyn_diamond_r_bk_big
  bone2Name=pelvis
  swing1LimitDegrees=45
  swing1Motion=Free
  swing2LimitDegrees=45
  swing2Motion=Free
  twistLimitDegrees=45
  twistMotion=Free
 }
 {
  bone1Name=dyn_diamond_r_bk_sml
  bone2Name=pelvis
  swing1LimitDegrees=45
  swing1Motion=Free
  swing2LimitDegrees=45
  swing2Motion=Free
  twistLimitDegrees=45
  twistMotion=Free
 }
 {
  bone1Name=dyn_Cloth_volume_r_B01
  bone2Name=pelvis
  swing1LimitDegrees=45
  swing1Motion=Free
  swing2LimitDegrees=45
  swing2Motion=Free
  twistLimitDegrees=45
  twistMotion=Free
 }
 {
  bone1Name=dyn_Cloth_volume_r_B02
  bone2Name=dyn_Cloth_volume_r_B01
  swing1LimitDegrees=45
  swing1Motion=Free
  swing2LimitDegrees=45
  swing2Motion=Free
  twistLimitDegrees=45
  twistMotion=Free
 }
 {
  bone1Name=dyn_Cloth_volume_l_B01
  bone2Name=pelvis
  swing1LimitDegrees=45
  swing1Motion=Free
  swing2LimitDegrees=45
  swing2Motion=Free
  twistLimitDegrees=45
  twistMotion=Free
 }
 {
  bone1Name=dyn_Cloth_volume_l_B02
  bone2Name=dyn_Cloth_volume_l_B01
  swing1LimitDegrees=45
  swing1Motion=Free
  swing2LimitDegrees=45
  swing2Motion=Free
  twistLimitDegrees=45
  twistMotion=Free
 }
]
