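# Control Rig for the RoyalDuskBlaze female face accessory mesh that applies pose-driven adjustments to pigtail and earring master bones.
# One forward chain runs four spherical pose readers, each weighting a Modify Transforms node.
# Two head readers, relative to spine_05, rotate and shift both pigtail masters.
# Left and right upper-arm readers, also relative to spine_05, rotate the earring master on the same side.
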ControlRigBlueprint /BRCosmetics/Characters/Player/Female/Medium/Bodies/F_MED_RoyalDuskBlaze/Meshes/Parts/F_MED_RoyalDuskBlaze_FaceAcc_CtrlRig
graphs [
 RigVMFunctionLibrary {links [], nodes []}
 RigVMModel {nodes [
   RigUnit_BeginExecution = &RigUnit_BeginExecution()
   n1 = &RigUnit_SphericalPoseReader(
    ActiveRegionScaleFactors=(PositiveWidth=1,NegativeWidth=1,PositiveHeight=1,NegativeHeight=1)
    ActiveRegionSize=0.2
    Debug=(bDrawDebug=false,DebugScale=25,DebugSegments=20,DebugThickness=0.25)
    DriverAxis=(X=0,Y=1,Z=0)
    DriverItem=(Type=Bone,Name="head")
    FalloffRegionScaleFactors=(PositiveWidth=1,NegativeWidth=1,PositiveHeight=1,NegativeHeight=1)
    FalloffSize=0.5
    OptionalParentItem=(Type=Bone,Name="spine_05")
    RotationOffset=(X=0,Y=-90,Z=0)
    ExecutePin=@RigUnit_BeginExecution.ExecutePin
   )
   RigVMFunction_MathQuaternionFromEuler = &RigVMFunction_MathQuaternionFromEuler(Euler=(X=0,Y=-30,Z=0))
   > &RigUnit_ModifyTransforms(
    ItemToModify="((Item=(Type=Bone,Name=\"dyn_pigtail_l_master\"),Transform=(Rotation=(X=0,Y=0,Z=0,W=1),Translation=(X=-2,Y=0,Z=0),Scale3D=(X=1,Y=1,Z=1))),(Item=(Type=Bone,Name=\"dyn_pigtail_r_master\"),Transform=(Rotation=(X=0,Y=0,Z=0,W=1),Translation=(X=-2,Y=0,Z=0),Scale3D=( ...(truncated)"
    ItemToModify.0.Transform.Rotation=@RigVMFunction_MathQuaternionFromEuler.Result
    ItemToModify.1.Transform.Rotation=&RigVMFunction_MathQuaternionFromEuler(Euler=(X=0,Y=-20,Z=0)).Result
    Weight=@n1.OutputParam
   )
   > n2 = &RigUnit_SphericalPoseReader(
    ActiveRegionScaleFactors=(PositiveWidth=1,NegativeWidth=1,PositiveHeight=1,NegativeHeight=1)
    ActiveRegionSize=0.2
    Debug=(bDrawDebug=false,DebugScale=25,DebugSegments=20,DebugThickness=0.25)
    DriverAxis=(X=0,Y=1,Z=0)
    DriverItem=(Type=Bone,Name="head")
    FalloffRegionScaleFactors=(PositiveWidth=1,NegativeWidth=1,PositiveHeight=1,NegativeHeight=1)
    FalloffSize=0.33
    OptionalParentItem=(Type=Bone,Name="spine_05")
    RotationOffset=(X=0,Y=-90,Z=-30)
   )
   > &RigUnit_ModifyTransforms(
    ItemToModify="((Item=(Type=Bone,Name=\"dyn_pigtail_l_master\"),Transform=(Rotation=(X=0,Y=0,Z=0,W=1),Translation=(X=0,Y=0,Z=2),Scale3D=(X=1,Y=1,Z=1))),(Item=(Type=Bone,Name=\"dyn_pigtail_r_master\"),Transform=(Rotation=(X=0,Y=0,Z=0,W=1),Translation=(X=0,Y=0,Z=2),Scale3D=(X= ...(truncated)"
    ItemToModify.1.Transform.Rotation=&RigVMFunction_MathQuaternionFromEuler(Euler=(X=0,Y=-5,Z=0)).Result
    ItemToModify.0.Transform.Rotation=&RigVMFunction_MathQuaternionFromEuler(Euler=(X=0,Y=-5,Z=0)).Result
    Weight=@n2.OutputParam
   )
   > n3 = &RigUnit_SphericalPoseReader(
    ActiveRegionScaleFactors=(PositiveWidth=1.5,NegativeWidth=1.5,PositiveHeight=2,NegativeHeight=2)
    ActiveRegionSize=0.08
    Debug=(bDrawDebug=false,bDraw2D=false,bDrawLocalAxes=false,DebugScale=25,DebugSegments=20,DebugThickness=0.25)
    DriverAxis=(X=-1,Y=0,Z=0)
    DriverItem=(Type=Bone,Name="upperarm_r")
    FalloffRegionScaleFactors=(PositiveWidth=1,NegativeWidth=1,PositiveHeight=2,NegativeHeight=2)
    FalloffSize=0.25
    OptionalParentItem=(Type=Bone,Name="spine_05")
    RotationOffset=(X=-80,Y=30,Z=40)
   )
   > &RigUnit_ModifyTransforms(
    ItemToModify=((Item=(Type=Bone,Name="dyn_r_earring_mstr"),Transform=(Rotation=(X=0,Y=0,Z=0,W=1),Translation=(X=0,Y=0,Z=0),Scale3D=(X=1,Y=1,Z=1))))
    ItemToModify.0.Transform.Rotation=&RigVMFunction_MathQuaternionFromEuler(Euler=(X=0,Y=0,Z=-15)).Result
    Weight=@n3.OutputParam
   )
   > n4 = &RigUnit_SphericalPoseReader(
    ActiveRegionScaleFactors=(PositiveWidth=1.5,NegativeWidth=1.5,PositiveHeight=2,NegativeHeight=2)
    ActiveRegionSize=0.1
    Debug=(bDrawDebug=false,bDraw2D=false,bDrawLocalAxes=false,DebugScale=25,DebugSegments=20,DebugThickness=0.25)
    DriverAxis=(X=-1,Y=0,Z=0)
    DriverItem=(Type=Bone,Name="upperarm_l")
    FalloffRegionScaleFactors=(PositiveWidth=1,NegativeWidth=1.6,PositiveHeight=2,NegativeHeight=2)
    FalloffSize=0.35
    FlipHeightScaling=false
    FlipWidthScaling=false
    OptionalParentItem=(Type=Bone,Name="spine_05")
    RotationOffset=(X=-60,Y=30,Z=40)
   )
   > &RigUnit_ModifyTransforms(
    ItemToModify=((Item=(Type=Bone,Name="dyn_l_earring_mstr"),Transform=(Rotation=(X=0,Y=0,Z=0,W=1),Translation=(X=0,Y=0,Z=0),Scale3D=(X=1,Y=1,Z=1))))
    ItemToModify.0.Transform.Rotation=&RigVMFunction_MathQuaternionFromEuler(Euler=(X=0,Y=0,Z=15)).Result
    Weight=@n4.OutputParam
   )
  ]}
]
hierarchy [
 root {loc=<0 0 0>, rot=<0 0 0 1>}
 pelvis {parent=root, loc=<0 -0.5301 98.0571>, rot=<0.0127 0.707 -0.0127 -0.707>}
 spine_01 {parent=pelvis, loc=<6.1208 -0.9328 0>, rot=<0 0 0.034 -0.9994>}
 spine_02 {parent=spine_01, loc=<6.5996 -0.2095 0>, rot=<0 0 0.0238 -0.9997>}
 spine_03 {parent=spine_02, loc=<6.4914 -0.1237 0>, rot=<0 0 0.0149 -0.9999>}
 spine_04 {parent=spine_03, loc=<6.9598 0.0794 0>, rot=<0 0 -0.0132 -0.9999>}
 spine_05 {parent=spine_04, loc=<6.9987 0 0>, rot=<0 0 -0.0015 -1>}
 neck_01 {parent=spine_05, loc=<11.3138 0.1529 0>, rot=<0 0 -0.2003 -0.9797>}
 neck_02 {parent=neck_01, loc=<4.5627 0 0>, rot=<0 0 0.0504 -0.9987>}
 head {parent=neck_02, loc=<5.3914 0 0>, rot=<0 0 0.115 -0.9934>}
 dyn_pigtail_l_master {parent=head, loc=<-0.6885 -5.5179 -1.9631>, rot=<0.104 -0.6755 0.699 -0.2107>}
 dyn_pigtail_l_01 {parent=dyn_pigtail_l_master, loc=<0 0 0>, rot=<0 0 0 1>}
 dyn_pigtail_l_02 {parent=dyn_pigtail_l_01, loc=<6.6814 0 0>, rot=<-0.0142 0.0177 -0.0286 -0.9993>}
 dyn_pigtail_l_03 {parent=dyn_pigtail_l_02, loc=<5.6705 0 0>, rot=<0.0059 -0.0596 0.0133 -0.9981>}
 dyn_pigtail_l_04 {parent=dyn_pigtail_l_03, loc=<6.8958 0 0>, rot=<-0.0034 0.0287 -0.008 -0.9996>}
 dyn_pigtail_l_05 {parent=dyn_pigtail_l_04, loc=<7.4695 0 0>, rot=<0.0044 -0.1075 0.0131 -0.9941>}
 dyn_pigtail_l_06 {parent=dyn_pigtail_l_05, loc=<5.6309 0.5071 0.0741>, rot=<0.0179 -0.0144 0.0876 -0.9959>}
 dyn_pigtail_r_master {parent=head, loc=<-0.6886 -5.5179 1.9632>, rot=<0.2107 -0.699 0.6755 -0.104>}
 dyn_pigtail_r_01 {parent=dyn_pigtail_r_master, loc=<0 0 0>, rot=<0 0 0 1>}
 dyn_pigtail_r_02 {parent=dyn_pigtail_r_01, loc=<6.681 0 0>, rot=<0.0062 0.0178 0.0125 -0.9997>}
 dyn_pigtail_r_03 {parent=dyn_pigtail_r_02, loc=<5.669 0 0>, rot=<0.1664 -0.0136 -0.022 -0.9857>}
 dyn_pigtail_r_04 {parent=dyn_pigtail_r_03, loc=<6.8962 0 0>, rot=<0.0033 0.0154 0.008 -0.9998>}
 dyn_pigtail_r_05 {parent=dyn_pigtail_r_04, loc=<7.4281 0 0>, rot=<-0.0042 -0.1072 -0.0136 -0.9941>}
 dyn_pigtail_r_06 {parent=dyn_pigtail_r_05, loc=<8.4788 0 0>, rot=<0.0031 -0.0758 0.0269 -0.9968>}
 dyn_hair_l_long_mstr {parent=head, loc=<10.9426 -1.7967 -8.1091>, rot=<0.5156 -0.7656 0.2123 -0.3209>}
 dyn_hair_l_long_01 {parent=dyn_hair_l_long_mstr, loc=<0 0 0>, rot=<0 0 0 1>}
 dyn_hair_l_long_02 {parent=dyn_hair_l_long_01, loc=<6.0538 0 0>, rot=<0.0246 0.1579 0.0462 -0.9861>}
 dyn_hair_bk_l_mstr {parent=head, loc=<4.3715 -3.2142 -4.6266>, rot=<-0.0393 -0.671 0.6683 -0.3189>}
 dyn_hair_bk_l_01 {parent=dyn_hair_bk_l_mstr, loc=<0 0 0>, rot=<0 0 0 1>}
 dyn_hair_bk_l_02 {parent=dyn_hair_bk_l_01, loc=<3.7023 0 0>, rot=<-0.0269 -0.0052 -0.0692 -0.9972>}
 dyn_hair_l_mid_mstr {parent=head, loc=<7.6565 3.4702 -8.4266>, rot=<0.2897 -0.9445 -0.1047 -0.1142>}
 dyn_hair_l_mid_01 {parent=dyn_hair_l_mid_mstr, loc=<0 0 0>, rot=<0 0 0 1>}
 dyn_hair_l_mid_02 {parent=dyn_hair_l_mid_01, loc=<4.9329 -0.4663 -0.7093>, rot=<-0.0571 0.1691 0.3598 -0.9158>}
 dyn_hair_bk_r_mstr {parent=head, loc=<8.666 -9.4162 -1.5735>, rot=<0.2445 -0.6666 0.6616 -0.2413>}
 dyn_hair_bk_r_01 {parent=dyn_hair_bk_r_mstr, loc=<0 0 0>, rot=<0 0 0 1>}
 dyn_hair_top_l_mstr {parent=head, loc=<16.3035 1.2391 -4.3862>, rot=<0.2908 -0.7769 0.1914 -0.5246>}
 dyn_hair_top_l_01 {parent=dyn_hair_top_l_mstr, loc=<0 0 0>, rot=<0 0 0 1>}
 dyn_hair_top_l_02 {parent=dyn_hair_top_l_01, loc=<4.8035 0 0>, rot=<-0.1336 0.1906 0.0432 -0.9716>}
 dyn_hair_top_mid_mstr {parent=head, loc=<16.9408 -0.3741 1.3314>, rot=<0.48 -0.5829 0.1771 -0.6312>}
 dyn_hair_top_mid_01 {parent=dyn_hair_top_mid_mstr, loc=<0 0 0>, rot=<-0.158 0 0 -0.9874>}
 dyn_hair_top_mid_02 {parent=dyn_hair_top_mid_01, loc=<4.4731 0 0>, rot=<-0.1336 0.1529 0.0482 -0.978>}
 dyn_hair_top_r_01_mstr {parent=head, loc=<15.3833 -3.7339 4.1745>, rot=<0.8143 -0.2861 0.3984 -0.3105>}
 dyn_hair_top_r_01 {parent=dyn_hair_top_r_01_mstr, loc=<0 0 0>, rot=<0 0 0 1>}
 dyn_hair_top_r_02_mstr {parent=head, loc=<9.6722 -7.5842 6.9532>, rot=<0.3545 -0.3556 0.8571 -0.115>}
 dyn_hair_top_r_02 {parent=dyn_hair_top_r_02_mstr, loc=<0 0 0>, rot=<0 0 0 1>}
 dyn_hair_r_mid_03_mstr {parent=head, loc=<1.5942 -4.088 4.5621>, rot=<0.1928 -0.5179 0.8274 0.0997>}
 dyn_hair_r_mid_03_a {parent=dyn_hair_r_mid_03_mstr, loc=<0 0 0>, rot=<0 0 0 1>}
 dyn_hair_r_mid_03_b {parent=dyn_hair_r_mid_03_a, loc=<2.8852 0 0>, rot=<-0.0013 0.2106 -0.0053 -0.9776>}
 dyn_hair_r_mid_01_mstr {parent=head, loc=<9.6404 3.7459 6.5629>, rot=<0.2197 -0.0203 0.9704 -0.0978>}
 dyn_hair_r_mid_01_a {parent=dyn_hair_r_mid_01_mstr, loc=<0 0 0>, rot=<0 0 0 1>}
 dyn_hair_r_mid_01_b {parent=dyn_hair_r_mid_01_a, loc=<5.7487 0 0>, rot=<-0.0121 -0.2917 -0.0541 -0.9549>}
 dyn_r_fr_bang_mstr {parent=head, loc=<9.3187 7.5053 2.9129>, rot=<0.9205 0.2644 0.2772 0.0772>}
 dyn_r_fr_bang_01 {parent=dyn_r_fr_bang_mstr, loc=<0 0 0>, rot=<0 0 0 1>}
 dyn_r_fr_bang_02 {parent=dyn_r_fr_bang_01, loc=<3.7226 0 0>, rot=<-0.0252 -0.8158 -0.0266 -0.5771>}
 dyn_r_fr_bang_03 {parent=dyn_r_fr_bang_02, loc=<7.8189 0 0>, rot=<0.0748 -0.2456 0.032 -0.966>}
 dyn_l_fr_bang_mstr {parent=head, loc=<9.5886 7.3432 0.1087>, rot=<-0.01 -0.2319 -0.1015 -0.9674>}
 dyn_l_fr_bang_01 {parent=dyn_l_fr_bang_mstr, loc=<0 0 0>, rot=<0 0 0 1>}
 dyn_l_fr_bang_02 {parent=dyn_l_fr_bang_01, loc=<5.4245 0 0>, rot=<0.191 -0.8217 0.0446 -0.535>}
 dyn_l_fr_bang_03 {parent=dyn_l_fr_bang_02, loc=<7.4213 2.4962 0.1184>, rot=<-0.0432 -0.3367 -0.218 -0.915>}
 dyn_earring_base_r_master {parent=head, loc=<0.1532 -0.5371 6.4553>, rot=<0.6137 -0.1784 0.0357 -0.7683>}
 dyn_earring_base_r {parent=dyn_earring_base_r_master, loc=<0 0 0>, rot=<0 0 0 1>}
 dyn_r_earring_mstr {parent=dyn_earring_base_r, loc=<-3.6609 0.0201 0.0832>, rot=<-0.0465 0.0419 0.9851 -0.1599>}
 dyn_r_earring {parent=dyn_r_earring_mstr, loc=<0 0 0>, rot=<0 0 0 1>}
 dyn_earring_base_l_master {parent=head, loc=<0.1536 -0.5371 -6.4553>, rot=<-0.0357 -0.7683 -0.6137 -0.1784>}
 dyn_earring_base_l {parent=dyn_earring_base_l_master, loc=<0 0 0>, rot=<0 0 0 1>}
 dyn_l_earring_mstr {parent=dyn_earring_base_l, loc=<3.7354 -0.0283 -0.0136>, rot=<0.0348 -0.0398 0.1247 -0.9908>}
 dyn_l_earring {parent=dyn_l_earring_mstr, loc=<0 0 0>, rot=<0 0 0 1>}
 dyn_hair_r_mid_02__base_mstr {parent=head, loc=<10.1997 -1.8215 6.7124>, rot=<0.2725 -0.1784 0.9413 -0.0882>}
 dyn_hair_r_mid_02__base {parent=dyn_hair_r_mid_02__base_mstr, loc=<0 0 0>, rot=<0 0 0 1>}
 dyn_hair_r_mid_02_mstr {parent=dyn_hair_r_mid_02__base, loc=<4.3459 0 0>, rot=<0.044 0 0 -0.999>}
 dyn_hair_r_mid_02_a {parent=dyn_hair_r_mid_02_mstr, loc=<0 0 0>, rot=<-0.1638 0 0 -0.9865>}
 dyn_hair_r_mid_02_b {parent=dyn_hair_r_mid_02_a, loc=<3.4778 0 0>, rot=<0.0077 0.2387 -0.03 -0.9706>}
 dyn_hair_r_bang_bk_mstr {parent=head, loc=<13.0529 3.5977 7.724>, rot=<0.3155 -0.0699 0.9222 -0.2125>}
 dyn_hair_r_bang_bk_01 {parent=dyn_hair_r_bang_bk_mstr, loc=<0 0 0>, rot=<0 0 0 1>}
 dyn_hair_r_bang_bk_02 {parent=dyn_hair_r_bang_bk_01, loc=<4.3899 0 0>, rot=<0.0215 0.047 0.0147 -0.9986>}
 dyn_hair_r_bang_bk_03 {parent=dyn_hair_r_bang_bk_02, loc=<5.1609 0 0>, rot=<0.0902 0.2801 0.0989 -0.9506>}
 clavicle_l {parent=spine_05, loc=<5.4831 -0.1053 -3.2946>, rot=<0.0465 -0.7289 -0.009 -0.6829>}
 upperarm_l {parent=clavicle_l, loc=<9.8658 -1.2136 -0.4105>, rot=<0.0164 -0.3245 -0.0269 -0.9454>}
 upperarm_twist_01_l {parent=upperarm_l, loc=<7.2135 -0.1134 0>, rot=<0 0 0 1>}
 lowerarm_l {parent=upperarm_l, loc=<26.465 -0.8345 0.259>, rot=<0.0036 0.0161 -0.2298 -0.9731>}
 hand_l {parent=lowerarm_l, loc=<22.7355 -0.3203 -0.6105>, rot=<-0.6784 -0.0223 0.0735 -0.7307>}
 weapon_l {parent=hand_l, loc=<-0.1974 -3.4111 0.3811>, rot=<0 0 0 1>}
 hand_attach_l {parent=hand_l, loc=<8.5 -3 0>, rot=<0.7071 0.7071 0 0>}
 thumb_01_l {parent=hand_l, loc=<1.3118 -0.3191 -2.459>, rot=<0.6546 -0.4042 0.0664 -0.6354>}
 thumb_02_l {parent=thumb_01_l, loc=<5.9336 -0.0448 0.0401>, rot=<-0.0037 0.0213 0.298 -0.9543>}
 thumb_03_l {parent=thumb_02_l, loc=<2.2223 0.3682 -0.0907>, rot=<0.0118 -0.013 0.0066 -0.9998>}
 index_metacarpal_l {parent=hand_l, loc=<1.6535 0.0183 -2.96>, rot=<0.0158 -0.0124 -0.0173 -0.9996>}
 index_01_l {parent=index_metacarpal_l, loc=<8.3008 0.0357 -0.0524>, rot=<-0.0044 0.0485 0.0924 -0.9945>}
 index_02_l {parent=index_01_l, loc=<3.3587 -0.3004 0.0452>, rot=<-0.002 0.0027 0.3225 -0.9466>}
 index_03_l {parent=index_02_l, loc=<2.5578 -0.123 -0.0252>, rot=<-0.002 0.0067 0.2558 -0.9667>}
 middle_metacarpal_l {parent=hand_l, loc=<2.1856 -0.1725 -0.5984>, rot=<0.006 0.0168 -0.0329 -0.9993>}
 middle_01_l {parent=middle_metacarpal_l, loc=<7.3275 -0.2447 0.0129>, rot=<-0.0208 0.0042 0.0858 -0.9961>}
 middle_02_l {parent=middle_01_l, loc=<4.0123 0.1401 0.1175>, rot=<-0.0218 0.0258 0.3512 -0.9357>}
 middle_03_l {parent=middle_02_l, loc=<2.6077 0.0534 0.1373>, rot=<0.0082 0.0188 0.1708 -0.9851>}
 ring_metacarpal_l {parent=hand_l, loc=<2.2074 -0.3625 1.3638>, rot=<-0.0284 0.0722 -0.0458 -0.9959>}
 ring_01_l {parent=ring_metacarpal_l, loc=<7.0295 -0.6197 -0.1446>, rot=<-0.0489 -0.0252 0.1892 -0.9804>}
 ring_02_l {parent=ring_01_l, loc=<3.6174 0.4647 -0.1027>, rot=<0 0 0.2806 -0.9598>}
 ring_03_l {parent=ring_02_l, loc=<2.4152 0.114 0.0199>, rot=<0 0 0.1189 -0.9929>}
 pinky_metacarpal_l {parent=hand_l, loc=<1.7616 -0.9753 3.3814>, rot=<-0.0397 0.0814 -0.069 -0.9935>}
 pinky_01_l {parent=pinky_metacarpal_l, loc=<7.1709 -0.4831 -0.1094>, rot=<-0.0654 -0.0035 0.2242 -0.9723>}
 pinky_02_l {parent=pinky_01_l, loc=<2.8274 0.0893 -0.1043>, rot=<0.0037 0.0011 0.2854 -0.9584>}
 pinky_03_l {parent=pinky_02_l, loc=<2.2046 0.0307 -0.0078>, rot=<-0.0002 0.0098 0.1445 -0.9895>}
 lowerarm_twist_02_l {parent=lowerarm_l, loc=<10.6707 0 0>, rot=<0 0 0 1>}
 lowerarm_twist_01_l {parent=lowerarm_l, loc=<18.316 0 0>, rot=<0 0 0 1>}
 upperarm_twist_02_l {parent=upperarm_l, loc=<19.7998 -0.1134 0>, rot=<0 0 0 1>}
 clavicle_r {parent=spine_05, loc=<5.4831 -0.1053 3.2946>, rot=<0.7289 0.0465 -0.6829 0.009>}
 upperarm_r {parent=clavicle_r, loc=<-9.8658 1.2136 0.4105>, rot=<0.0164 -0.3245 -0.0269 -0.9454>}
 upperarm_twist_01_r {parent=upperarm_r, loc=<-7.2135 0.1134 0>, rot=<0 0 0 -1>}
 lowerarm_r {parent=upperarm_r, loc=<-26.465 0.8345 -0.259>, rot=<0.0036 0.0161 -0.2298 -0.9731>}
 hand_r {parent=lowerarm_r, loc=<-22.7355 0.3203 0.6106>, rot=<-0.6784 -0.0223 0.0735 -0.7307>}
 weapon_r {parent=hand_r, loc=<0.1974 3.4111 -0.3811>, rot=<0 0 0 1>}
 thumb_01_r {parent=hand_r, loc=<-1.3118 0.3191 2.459>, rot=<0.6546 -0.4042 0.0664 -0.6354>}
 thumb_02_r {parent=thumb_01_r, loc=<-5.9336 0.0448 -0.0401>, rot=<-0.0037 0.0213 0.298 -0.9543>}
 thumb_03_r {parent=thumb_02_r, loc=<-2.2223 -0.3682 0.0907>, rot=<0.0118 -0.013 0.0066 -0.9998>}
 index_metacarpal_r {parent=hand_r, loc=<-1.6535 -0.0183 2.96>, rot=<0.0158 -0.0124 -0.0173 -0.9996>}
 index_01_r {parent=index_metacarpal_r, loc=<-8.3008 -0.0357 0.0524>, rot=<-0.0044 0.0485 0.0924 -0.9945>}
 index_02_r {parent=index_01_r, loc=<-3.3587 0.3004 -0.0452>, rot=<-0.002 0.0027 0.3225 -0.9466>}
 index_03_r {parent=index_02_r, loc=<-2.5578 0.123 0.0252>, rot=<-0.002 0.0067 0.2558 -0.9667>}
 middle_metacarpal_r {parent=hand_r, loc=<-2.1856 0.1725 0.5984>, rot=<0.006 0.0168 -0.0329 -0.9993>}
 middle_01_r {parent=middle_metacarpal_r, loc=<-7.3275 0.2447 -0.0129>, rot=<-0.0208 0.0042 0.0858 -0.9961>}
 middle_02_r {parent=middle_01_r, loc=<-4.0123 -0.1401 -0.1175>, rot=<-0.0218 0.0258 0.3512 -0.9357>}
 middle_03_r {parent=middle_02_r, loc=<-2.6077 -0.0534 -0.1373>, rot=<0.0082 0.0188 0.1708 -0.9851>}
 ring_metacarpal_r {parent=hand_r, loc=<-2.2074 0.3625 -1.3638>, rot=<-0.0284 0.0722 -0.0458 -0.9959>}
 ring_01_r {parent=ring_metacarpal_r, loc=<-7.0295 0.6197 0.1446>, rot=<-0.0489 -0.0252 0.1892 -0.9804>}
 ring_02_r {parent=ring_01_r, loc=<-3.6174 -0.4647 0.1027>, rot=<0 0 0.2806 -0.9598>}
 ring_03_r {parent=ring_02_r, loc=<-2.4152 -0.114 -0.0199>, rot=<0 0 0.1189 -0.9929>}
 pinky_metacarpal_r {parent=hand_r, loc=<-1.7616 0.9753 -3.3814>, rot=<-0.0397 0.0814 -0.069 -0.9935>}
 pinky_01_r {parent=pinky_metacarpal_r, loc=<-7.1709 0.4831 0.1094>, rot=<-0.0654 -0.0035 0.2242 -0.9723>}
 pinky_02_r {parent=pinky_01_r, loc=<-2.8274 -0.0893 0.1043>, rot=<0.0037 0.0011 0.2854 -0.9584>}
 pinky_03_r {parent=pinky_02_r, loc=<-2.2046 -0.0307 0.0078>, rot=<-0.0002 0.0098 0.1445 -0.9895>}
 hand_attach_r {parent=hand_r, loc=<-8.5001 2.9999 0>, rot=<0 0 -0.7071 -0.7071>}
 lowerarm_twist_02_r {parent=lowerarm_r, loc=<-10.6707 0 0>, rot=<0 0 0 -1>}
 lowerarm_twist_01_r {parent=lowerarm_r, loc=<-18.316 0 0>, rot=<0 0 0 -1>}
 upperarm_twist_02_r {parent=upperarm_r, loc=<-19.7998 0.1134 0>, rot=<0 0 0 -1>}
 attach_cape {parent=spine_05, loc=<9.3216 -8.4501 0>, rot=<0.0616 -0.7044 -0.0616 -0.7044>}
 attach_backpack {parent=spine_05, loc=<0 -8.35 0>, rot=<-0.0093 -0.707 0.0093 -0.707>}
 thigh_l {parent=pelvis, loc=<-1.0855 -0.2109 -8.7053>, rot=<0.0009 0.0355 0.0155 -0.9993>}
 calf_l {parent=thigh_l, loc=<-42.5284 -0.0019 -0.7297>, rot=<0.0006 -0.0161 -0.0391 -0.9991>}
 foot_l {parent=calf_l, loc=<-46.7401 -2.5014 -0.592>, rot=<0.0427 -0.0238 0.0353 -0.9982>}
 ball_l {parent=foot_l, loc=<-3.5658 10.8066 -0.1544>, rot=<0.0001 -0.0001 -0.7063 -0.7079>}
 calf_twist_02_l {parent=calf_l, loc=<-19.1546 -1.7943 -0.2638>, rot=<0.0237 0.0095 -0.0035 -0.9997>}
 calf_twist_01_l {parent=calf_l, loc=<-35.8483 -1.9197 -0.5794>, rot=<0.0237 0.0095 -0.0035 -0.9997>}
 thigh_twist_01_l {parent=thigh_l, loc=<-27.952 0 0>, rot=<0.0013 0 0 -1>}
 thigh_r {parent=pelvis, loc=<-1.0855 -0.2109 8.7053>, rot=<0.0355 -0.0009 0.9993 0.0155>}
 calf_r {parent=thigh_r, loc=<42.5284 0.0019 0.7297>, rot=<0.0006 -0.0161 -0.0391 -0.9991>}
 foot_r {parent=calf_r, loc=<46.7401 2.5014 0.592>, rot=<0.0427 -0.0238 0.0353 -0.9982>}
 ball_r {parent=foot_r, loc=<3.5658 -10.8066 0.1544>, rot=<0.0001 -0.0001 -0.7063 -0.7079>}
 calf_twist_02_r {parent=calf_r, loc=<19.1546 1.7943 0.2638>, rot=<0.0237 0.0095 -0.0035 -0.9997>}
 calf_twist_01_r {parent=calf_r, loc=<35.8483 1.9197 0.5794>, rot=<0.0237 0.0095 -0.0035 -0.9997>}
 thigh_twist_01_r {parent=thigh_r, loc=<27.952 0 0>, rot=<0.0013 0 0 -1>}
 attach {parent=root, loc=<0 0 0>, rot=<0 0 0 1>}
 ik_hand_root {parent=root, loc=<0 0 0>, rot=<0 0 0 1>}
 ik_hand_gun {parent=ik_hand_root, loc=<-48.1743 3.3545 104.4381>, rot=<0.6276 0.1309 0.4183 -0.6434>}
 ik_hand_r {parent=ik_hand_gun, loc=<0 0 0>, rot=<0 0 0 1>}
 ik_hand_l {parent=ik_hand_gun, loc=<59.3271 67.6997 34.3524>, rot=<-0.6158 -0.7027 -0.3565 0>}
 ik_foot_root {parent=root, loc=<0 0 0>, rot=<0 0 0 1>}
 ik_foot_l {parent=ik_foot_root, loc=<14.8262 -7.3545 8.296>, rot=<0.0357 0.7045 0.0284 -0.7082>}
 ik_foot_r {parent=ik_foot_root, loc=<-14.8262 -7.3545 8.296>, rot=<0.7082 0.0284 -0.7045 0.0357>}
 attach_fp {parent=root, loc=<0 0 0>, rot=<0 0 0 1>}
 camera_root {parent=root, loc=<0 0 0>, rot=<0 0 0 1>}
 camera {parent=camera_root, loc=<0 0 0>, rot=<0 0 0 1>}
 warp_point_01 {parent=root, loc=<0 0 0>, rot=<0 0 0 1>}
 warp_point_02 {parent=root, loc=<0 0 0>, rot=<0 0 0 1>}
 warp_point_03 {parent=root, loc=<0 0 0>, rot=<0 0 0 1>}
]
