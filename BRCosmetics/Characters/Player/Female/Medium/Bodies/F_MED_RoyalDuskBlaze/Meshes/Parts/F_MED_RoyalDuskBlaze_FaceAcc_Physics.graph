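# Physics asset for the RoyalDuskBlaze medium female cosmetic's FaceAcc mesh, setting up capsule bodies and joints for two dynamic pigtails.
# Each pigtail chains six segments from a master body on the head, with narrow twist and swing limits that are tightest near the root.
# Head, neck, spine, clavicle and upper-arm capsules form the upper body, linked by free angular constraints, as are the pigtail masters to the head.
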PhysicsAsset /BRCosmetics/Characters/Player/Female/Medium/Bodies/F_MED_RoyalDuskBlaze/Meshes/Parts/F_MED_RoyalDuskBlaze_FaceAcc_Physics
bodies {
 clavicle_l [{shapeType=Capsule, length=5.7962, radius=7.3301, shapeName=clavicle_l_capsule, center=<7.7741 0 0>, rotation=<rot 90 0 0>}]
 clavicle_r [{shapeType=Capsule, length=5.7962, radius=7.3301, shapeName=clavicle_r_capsule, center=<-7.6676 0 0>, rotation=<rot 90 0 0>}]
 dyn_pigtail_l_01 [{shapeType=Capsule, length=3.9207, radius=2.2004, shapeName=dyn_pigtail_r_01_capsule, center=<3.0364 0 3>, rotation=<rot 90 0 0>}]
 dyn_pigtail_l_02 [{shapeType=Capsule, length=4.0692, radius=1.9861, shapeName=dyn_pigtail_r_02_capsule, center=<2.7582 0 3>, rotation=<rot 90 0 0>}]
 dyn_pigtail_l_03 [{shapeType=Capsule, length=3.7449, radius=3.4024, shapeName=dyn_pigtail_r_03_capsule, center=<3.824 0 3>, rotation=<rot 90 0 0>}]
 dyn_pigtail_l_04 [{shapeType=Capsule, length=5.3401, radius=5.6219, center=<5 0 3>, rotation=<rot 81.3082 -0.0001 -0.0001>}]
 dyn_pigtail_l_05 [{shapeType=Capsule, length=4.4904, radius=6.4716, shapeName=dyn_pigtail_l_05_capsule, center=<5 0 3>, rotation=<rot 90 0 0>}]
 dyn_pigtail_l_06 [{shapeType=Capsule, length=4.5827, radius=6.3794, shapeName=dyn_pigtail_l_06_capsule, center=<6.9875 0 3>, rotation=<rot 90 0 0>}]
 dyn_pigtail_l_master [{shapeType=Capsule, length=0.505, radius=0.505, shapeName=dyn_pigtail_l_master_capsule, rotation=<rot 0 -90 0>}]
 dyn_pigtail_r_01 [{shapeType=Capsule, length=3.9207, radius=2.2004, shapeName=dyn_pigtail_r_01_capsule, center=<3.0364 0 3>, rotation=<rot 90 0 0>}]
 dyn_pigtail_r_02 [{shapeType=Capsule, length=3.643, radius=2.4123, shapeName=dyn_pigtail_r_02_capsule, center=<2.7582 0 3>, rotation=<rot 90 0 0>}]
 dyn_pigtail_r_03 [{shapeType=Capsule, length=3.7449, radius=3.4024, shapeName=dyn_pigtail_r_03_capsule, center=<3.824 0 3>, rotation=<rot 90 0 0>}]
 dyn_pigtail_r_04 [{shapeType=Capsule, length=5.3401, radius=5.6219, center=<5 0 3>, rotation=<rot 83.1682 0.0001 0>}]
 dyn_pigtail_r_05 [{shapeType=Capsule, length=3.9098, radius=7.0522, shapeName=dyn_pigtail_r_05_capsule, center=<5 0 3>, rotation=<rot 90 0 0>}]
 dyn_pigtail_r_06 [{shapeType=Capsule, length=4.1297, radius=6.8324, shapeName=dyn_pigtail_r_06_capsule, center=<7.2669 0.3433 3>, rotation=<rot 80.925 0 0>}]
 dyn_pigtail_r_master [{shapeType=Capsule, length=0.505, radius=0.505, shapeName=dyn_pigtail_r_master_capsule, rotation=<rot 0 -90 0>}]
 head [{shapeType=Capsule, length=2, radius=2, shapeName=head_capsule}]
 neck_01 [{shapeType=Capsule, length=5.3653, radius=5.8582, shapeName=neck_01_capsule, rotation=<rot 90 0 0>}]
 neck_02 [{shapeType=Capsule, length=5.3653, radius=5.3398, shapeName=neck_02_capsule, rotation=<rot 90 0 0>}]
 spine_01 [{shapeType=Capsule, length=16.7181, radius=9.3268, shapeName=spine_02_capsule, rotation=<rot 0 -90 0>}]
 spine_03 [{shapeType=Capsule, length=15.7639, radius=6.8844, shapeName=spine_03_capsule, rotation=<rot 0 -90 0>}]
 spine_05 [
  {shapeType=Capsule, length=24.1971, radius=8.6644, shapeName=spine_05_capsule, center=<2.6372 0.1982 0>}
  {shapeType=Capsule, length=24.1971, radius=8.6644, shapeName=spine_05_capsule, center=<8.6901 -7.8828 -19.7707>, rotation=<rot 0 0 54.8203>}
 ]
 upperarm_l [{shapeType=Capsule, length=21.0025, radius=2.6116, shapeName=upperarm_l_capsule, center=<12.3593 0 0>, rotation=<rot 90 0 0>}]
 upperarm_r [{shapeType=Capsule, length=17.7017, radius=5.9124, shapeName=upperarm_r_capsule, center=<-13.503 0 0>, rotation=<rot 90 0 0>}]
}
constraints [
 {bone1Name=dyn_pigtail_l_04, bone2Name=dyn_pigtail_l_03, swing1LimitDegrees=20, swing2LimitDegrees=30, twistLimitDegrees=5}
 {bone1Name=dyn_pigtail_r_04, bone2Name=dyn_pigtail_r_03, swing1LimitDegrees=20, swing2LimitDegrees=30, twistLimitDegrees=5}
 {
  bone1Name=neck_02
  bone2Name=neck_01
  swing1LimitDegrees=45
  swing1Motion=Free
  swing2LimitDegrees=45
  swing2Motion=Free
  twistLimitDegrees=45
  twistMotion=Free
 }
 {
  bone1Name=neck_01
  bone2Name=spine_05
  swing1LimitDegrees=45
  swing1Motion=Free
  swing2LimitDegrees=45
  swing2Motion=Free
  twistLimitDegrees=45
  twistMotion=Free
 }
 {
  bone1Name=spine_05
  bone2Name=spine_03
  swing1LimitDegrees=45
  swing1Motion=Free
  swing2LimitDegrees=45
  swing2Motion=Free
  twistLimitDegrees=45
  twistMotion=Free
 }
 {
  bone1Name=spine_03
  bone2Name=spine_01
  swing1LimitDegrees=45
  swing1Motion=Free
  swing2LimitDegrees=45
  swing2Motion=Free
  twistLimitDegrees=45
  twistMotion=Free
 }
 {
  bone1Name=clavicle_l
  bone2Name=spine_05
  swing1LimitDegrees=45
  swing1Motion=Free
  swing2LimitDegrees=45
  swing2Motion=Free
  twistLimitDegrees=45
  twistMotion=Free
 }
 {
  bone1Name=clavicle_r
  bone2Name=spine_05
  swing1LimitDegrees=45
  swing1Motion=Free
  swing2LimitDegrees=45
  swing2Motion=Free
  twistLimitDegrees=45
  twistMotion=Free
 }
 {
  bone1Name=upperarm_l
  bone2Name=clavicle_l
  swing1LimitDegrees=45
  swing1Motion=Free
  swing2LimitDegrees=45
  swing2Motion=Free
  twistLimitDegrees=45
  twistMotion=Free
 }
 {
  bone1Name=upperarm_r
  bone2Name=clavicle_r
  swing1LimitDegrees=45
  swing1Motion=Free
  swing2LimitDegrees=45
  swing2Motion=Free
  twistLimitDegrees=45
  twistMotion=Free
 }
 {bone1Name=dyn_pigtail_l_06, bone2Name=dyn_pigtail_l_05, swing1LimitDegrees=20, swing2LimitDegrees=30, twistLimitDegrees=5}
 {bone1Name=dyn_pigtail_l_05, bone2Name=dyn_pigtail_l_04, swing1LimitDegrees=20, swing2LimitDegrees=30, twistLimitDegrees=5}
 {bone1Name=dyn_pigtail_l_02, bone2Name=dyn_pigtail_l_01, swing1LimitDegrees=15, swing2LimitDegrees=15, twistLimitDegrees=5}
 {bone1Name=dyn_pigtail_l_03, bone2Name=dyn_pigtail_l_02, swing1LimitDegrees=20, swing2LimitDegrees=30, twistLimitDegrees=5}
 {bone1Name=dyn_pigtail_l_01, bone2Name=dyn_pigtail_l_master, swing1LimitDegrees=7, swing2LimitDegrees=17, twistLimitDegrees=5}
 {bone1Name=dyn_pigtail_r_01, bone2Name=dyn_pigtail_r_master, swing1LimitDegrees=7, swing2LimitDegrees=17, twistLimitDegrees=5}
 {bone1Name=dyn_pigtail_r_02, bone2Name=dyn_pigtail_r_01, swing1LimitDegrees=15, swing2LimitDegrees=15, twistLimitDegrees=5}
 {bone1Name=dyn_pigtail_r_03, bone2Name=dyn_pigtail_r_02, swing1LimitDegrees=20, swing2LimitDegrees=30, twistLimitDegrees=5}
 {bone1Name=dyn_pigtail_r_05, bone2Name=dyn_pigtail_r_04, swing1LimitDegrees=20, swing2LimitDegrees=30, twistLimitDegrees=5}
 {bone1Name=dyn_pigtail_r_06, bone2Name=dyn_pigtail_r_05, swing1LimitDegrees=20, swing2LimitDegrees=30, twistLimitDegrees=5}
 {
  bone1Name=head
  bone2Name=neck_02
  swing1LimitDegrees=45
  swing1Motion=Free
  swing2LimitDegrees=45
  swing2Motion=Free
  twistLimitDegrees=45
  twistMotion=Free
 }
 {
  bone1Name=dyn_pigtail_l_master
  bone2Name=head
  swing1LimitDegrees=45
  swing1Motion=Free
  swing2LimitDegrees=45
  swing2Motion=Free
  twistLimitDegrees=45
  twistMotion=Free
 }
 {
  bone1Name=dyn_pigtail_r_master
  bone2Name=head
  swing1LimitDegrees=45
  swing1Motion=Free
  swing2LimitDegrees=45
  swing2Motion=Free
  twistLimitDegrees=45
  twistMotion=Free
 }
]
