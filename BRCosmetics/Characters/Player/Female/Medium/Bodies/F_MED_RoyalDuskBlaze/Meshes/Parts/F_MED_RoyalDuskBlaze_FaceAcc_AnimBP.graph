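# Post-copy animation Blueprint for the RoyalDuskBlaze female face accessory mesh, adding secondary motion to hair and posing pigtails around the body.
# It copies the parent pose, layers in Body_Mesh arms and clavicles, then chains spring-based AnimDynamics through hair and bang bones.
# Output reads cached Base_Pose, applying weighted multi-bone constraints to pigtails: strongly with offsets when backbling is equipped, otherwise lighter during grounded tactical sprint.
# Final additive rotations adjust both pigtails.
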AnimBlueprint /BRCosmetics/Characters/Player/Female/Medium/Bodies/F_MED_RoyalDuskBlaze/Meshes/Parts/F_MED_RoyalDuskBlaze_FaceAcc_AnimBP
target_skeleton=&/F_MED_RoyalDuskBlaze_FaceAcc_Skeleton
graphs [
 AnimGraph {nodes [
   n1 = Anim Dynamics - Bone: dyn_hair_top_l_01(Alpha=1, ExternalForce="0,0,0") {
    angularDampingOverride=1
    angularSpringConstant=5000
    bAngularSpring=true
    bGravityOverrideInSimSpace=true
    bOverrideAngularDamping=true
    bUseGravityOverride=true
    simulationSpace=BoneRelative
    alphaBoolBlend {blendOption=Linear}
    boundBone {boneName=dyn_hair_top_l_01}
    componentAppliedLinearAccClamp=<1500 1500 1500>
    componentLinearAccScale=<0.4 0.4 1>
    componentLinearVelScale=<0.2 0.2 0.4>
    gravityOverride {z=-980}
    physicsBodyDefinitions [{
      sphereCollisionRadius=0
      boundBone {boneName=dyn_hair_top_l_01}
      boxExtents=<20 20 20>
      constraintSetup {angularLimitsMax=<3 5 5>, angularLimitsMin=<-3 -5 -5>, angularTarget=<1 0 0>}
      localJointOffset=<3 0 0>
     }]
    relativeSpaceBone {boneName=root}
   }
   Anim Dynamics - Bone: dyn_hair_top_l_02(ComponentPose=@n1, Alpha=1, ExternalForce="0,0,0") {
    angularDampingOverride=1
    angularSpringConstant=5000
    bAngularSpring=true
    bGravityOverrideInSimSpace=true
    bOverrideAngularDamping=true
    bUseGravityOverride=true
    simulationSpace=BoneRelative
    alphaBoolBlend {blendOption=Linear}
    boundBone {boneName=dyn_hair_top_l_02}
    componentAppliedLinearAccClamp=<1500 1500 1500>
    componentLinearAccScale=<0.4 0.4 1>
    componentLinearVelScale=<0.2 0.2 0.4>
    gravityOverride {z=-980}
    physicsBodyDefinitions [{
      sphereCollisionRadius=0
      boundBone {boneName=dyn_hair_top_l_02}
      boxExtents=<20 20 20>
      constraintSetup {angularLimitsMax=<3 10 10>, angularLimitsMin=<-3 -10 -10>, angularTarget=<1 0 0>}
      localJointOffset=<3 0 0>
     }]
    relativeSpaceBone {boneName=root}
   }
   n2 = Get IsTargeting() -> (IsTargeting:float)
   n3 = Anim Dynamics - Bone: dyn_earring_base_r(Alpha=@n2, ExternalForce="0,0,0") {
    angularDampingOverride=1
    angularSpringConstant=20000
    bAngularSpring=true
    bGravityOverrideInSimSpace=true
    bLinearSpring=true
    bOverrideAngularDamping=true
    bOverrideLinearDamping=true
    bUseGravityOverride=true
    gravityScale=0.35
    linearDampingOverride=1
    linearSpringConstant=135
    simulationSpace=RootRelative
    alphaBoolBlend {blendOption=Linear}
    boundBone {boneName=dyn_earring_base_r}
    componentAppliedLinearAccClamp=<3000 3000 3000>
    componentLinearAccScale=<0.3 0.3 0.4>
    componentLinearVelScale=<0.3 0.3 0.4>
    gravityOverride {z=-980}
    physicsBodyDefinitions [{
      sphereCollisionRadius=0
      boundBone {boneName=dyn_earring_base_r}
      boxExtents=<20 30 30>
      constraintSetup {angularLimitsMax=<5 80 40>, angularLimitsMin=<-5 -80 -40>, angularTarget=<1 0 0>}
      localJointOffset=<0 0 0>
     }]
   }
   Anim Dynamics - Bone: dyn_earring_base_l(ComponentPose=@n3, Alpha=@n2, ExternalForce="0,0,0") {
    angularDampingOverride=1
    angularSpringConstant=20000
    bAngularSpring=true
    bGravityOverrideInSimSpace=true
    bLinearSpring=true
    bOverrideAngularDamping=true
    bOverrideLinearDamping=true
    bUseGravityOverride=true
    gravityScale=0.35
    linearDampingOverride=1
    linearSpringConstant=135
    simulationSpace=RootRelative
    alphaBoolBlend {blendOption=Linear}
    boundBone {boneName=dyn_earring_base_l}
    componentAppliedLinearAccClamp=<3000 3000 3000>
    componentLinearAccScale=<0.3 0.3 0.4>
    componentLinearVelScale=<0.3 0.3 0.4>
    gravityOverride {z=-980}
    physicsBodyDefinitions [{
      sphereCollisionRadius=0
      boundBone {boneName=dyn_earring_base_l}
      boxExtents=<20 30 30>
      constraintSetup {angularLimitsMax=<5 80 40>, angularLimitsMin=<-5 -80 -40>, angularTarget=<1 0 0>}
      localJointOffset=<0 0 0>
     }]
   }
   "Constraint - dyn_pigtail_r_03 (PPP)"!(ComponentPose, Alpha=0.5, ConstraintWeights_0=0.6532, ConstraintWeights_1:float=0.013, ConstraintWeights_2:float=0.3338) -> (Pose) {
    alphaBoolBlend {blendOption=Linear}
    boneToModify {boneName=dyn_pigtail_r_03}
    constraintSetup [{targetBone {boneName=spine_05}}, {targetBone {boneName=clavicle_l}}, {targetBone {boneName=clavicle_r}}]
    constraintWeights [1,1,1]
   }
   n4 = "Constraint - dyn_pigtail_l_02 (PPPPPP)"!(
    ComponentPose=Local To Component(LocalPose=Use cached pose 'Base_Pose'())
    Alpha=0.5
    ConstraintWeights_0=0.5881
    ConstraintWeights_1:float=0.0012
    ConstraintWeights_2:float=0.0007
    ConstraintWeights_3:float=0.1265
    ConstraintWeights_4:float=0.0658
    ConstraintWeights_5:float=0.2177
   ) -> (Pose) {
    alphaBoolBlend {blendOption=Linear}
    boneToModify {boneName=dyn_pigtail_l_02}
    constraintSetup [
     {targetBone {boneName=spine_05}}
     {targetBone {boneName=clavicle_l}}
     {targetBone {boneName=clavicle_r}}
     {targetBone {boneName=neck_01}}
     {targetBone {boneName=head}}
     {targetBone {boneName=neck_02}}
    ]
    constraintWeights [1,1,1,1,1,1]
   }
   n5 = "Constraint - dyn_pigtail_r_02 (PPPPPP)"!(
    ComponentPose=@n4
    Alpha=0.5
    ConstraintWeights_0=0.5899
    ConstraintWeights_1:float=0.0001
    ConstraintWeights_2:float=0.0018
    ConstraintWeights_3:float=0.126
    ConstraintWeights_4:float=0.0655
    ConstraintWeights_5:float=0.2168
   ) -> (Pose) {
    alphaBoolBlend {blendOption=Linear}
    boneToModify {boneName=dyn_pigtail_r_02}
    constraintSetup [
     {targetBone {boneName=spine_05}}
     {targetBone {boneName=clavicle_l}}
     {targetBone {boneName=clavicle_r}}
     {targetBone {boneName=neck_01}}
     {targetBone {boneName=head}}
     {targetBone {boneName=neck_02}}
    ]
    constraintWeights [1,1,1,1,1,1]
   }
   n6 = "Constraint - dyn_pigtail_l_03 (PPP)"!(ComponentPose=@n5, Alpha=0.5, ConstraintWeights_0=0.6527, ConstraintWeights_1:float=0.3344, ConstraintWeights_2:float=0.013) -> (Pose) {
    alphaBoolBlend {blendOption=Linear}
    boneToModify {boneName=dyn_pigtail_l_03}
    constraintSetup [{targetBone {boneName=spine_05}}, {targetBone {boneName=clavicle_l}}, {targetBone {boneName=clavicle_r}}]
    constraintWeights [1,1,1]
   }
   n7 = "Constraint - dyn_pigtail_l_04 (PPPPPPPP)"!(
    ComponentPose=@n6
    Alpha=0.35
    ConstraintWeights_0=0.0138
    ConstraintWeights_1:float=0.0051
    ConstraintWeights_2:float=0.0974
    ConstraintWeights_3:float=0.0092
    ConstraintWeights_4:float=0.3632
    ConstraintWeights_5:float=0.4797
    ConstraintWeights_6:float=0.0306
    ConstraintWeights_7:float=0.0011
   ) -> (Pose) {
    alphaBoolBlend {blendOption=Linear}
    boneToModify {boneName=dyn_pigtail_l_04}
    constraintSetup [
     {targetBone {boneName=upperarm_twist_01_l}}
     {targetBone {boneName=spine_02}}
     {targetBone {boneName=spine_04}}
     {targetBone {boneName=spine_03}}
     {targetBone {boneName=spine_05}}
     {targetBone {boneName=clavicle_l}}
     {targetBone {boneName=clavicle_r}}
     {targetBone {boneName=spine_01}}
    ]
    constraintWeights [1,1,1,1,1,1,1,1]
   }
   n8 = Get bIsJumping()
   n9 = Get bIsFalling()
   n10 = Blend Poses by bool(
    BlendPose_0=Component To Local(ComponentPose=@n7)
    BlendPose_1=Use cached pose 'Base_Pose'()
    BlendTime_0=0.25
    BlendTime_1=0.25
    bActiveValue=AND Boolean!(A=Get bIsTacticalSprinting(), B=NOT Boolean(A=@n8), C:bool=NOT Boolean(A=@n9)) -> (ReturnValue)
   ) {blendType=Linear}
   n11 = "Constraint - dyn_pigtail_l_02 (PPPPPP)"(
    ComponentPose=Local To Component(LocalPose=Use cached pose 'Base_Pose'())
    Alpha=0.5
    ConstraintWeights_0=0.5881
    ConstraintWeights_1=0.0012
    ConstraintWeights_2=0.0007
    ConstraintWeights_3=0.1265
    ConstraintWeights_4=0.0658
    ConstraintWeights_5=0.2177
   ) {
    alphaBoolBlend {blendOption=Linear}
    boneToModify {boneName=dyn_pigtail_l_02}
    constraintSetup [
     {targetBone {boneName=spine_05}}
     {targetBone {boneName=clavicle_l}}
     {targetBone {boneName=clavicle_r}}
     {targetBone {boneName=neck_01}}
     {targetBone {boneName=head}}
     {targetBone {boneName=neck_02}}
    ]
    constraintWeights [1,1,1,1,1,1]
   }
   n12 = "Constraint - dyn_pigtail_r_02 (PPPPPP)"(
    ComponentPose=@n11
    Alpha=0.5
    ConstraintWeights_0=0.5899
    ConstraintWeights_1=0.0001
    ConstraintWeights_2=0.0018
    ConstraintWeights_3=0.126
    ConstraintWeights_4=0.0655
    ConstraintWeights_5=0.2168
   ) {
    alphaBoolBlend {blendOption=Linear}
    boneToModify {boneName=dyn_pigtail_r_02}
    constraintSetup [
     {targetBone {boneName=spine_05}}
     {targetBone {boneName=clavicle_l}}
     {targetBone {boneName=clavicle_r}}
     {targetBone {boneName=neck_01}}
     {targetBone {boneName=head}}
     {targetBone {boneName=neck_02}}
    ]
    constraintWeights [1,1,1,1,1,1]
   }
   n13 = "Constraint - dyn_pigtail_l_03 (PPP)"(ComponentPose=@n12, Alpha=0.5, ConstraintWeights_0=0.6527, ConstraintWeights_1=0.3344, ConstraintWeights_2=0.013) {
    alphaBoolBlend {blendOption=Linear}
    boneToModify {boneName=dyn_pigtail_l_03}
    constraintSetup [{targetBone {boneName=spine_05}}, {targetBone {boneName=clavicle_l}}, {targetBone {boneName=clavicle_r}}]
    constraintWeights [1,1,1]
   }
   n14 = "Constraint - dyn_pigtail_r_03 (PPP)"(ComponentPose=@n13, Alpha=0.5, ConstraintWeights_0=0.6532, ConstraintWeights_1=0.013, ConstraintWeights_2=0.3338) {
    alphaBoolBlend {blendOption=Linear}
    boneToModify {boneName=dyn_pigtail_r_03}
    constraintSetup [{targetBone {boneName=spine_05}}, {targetBone {boneName=clavicle_l}}, {targetBone {boneName=clavicle_r}}]
    constraintWeights [1,1,1]
   }
   n15 = "Constraint - dyn_pigtail_l_04 (PPPPPPPP)"(
    ComponentPose=@n14
    Alpha=0.8
    ConstraintWeights_0=0.0138
    ConstraintWeights_1=0.0051
    ConstraintWeights_2=0.0974
    ConstraintWeights_3=0.0092
    ConstraintWeights_4=0.3632
    ConstraintWeights_5=0.4797
    ConstraintWeights_6=0.0306
    ConstraintWeights_7=0.0011
   ) {
    alphaBoolBlend {blendOption=Linear}
    boneToModify {boneName=dyn_pigtail_l_04}
    constraintSetup [
     {targetBone {boneName=upperarm_twist_01_l}}
     {targetBone {boneName=spine_02}}
     {targetBone {boneName=spine_04}}
     {targetBone {boneName=spine_03}}
     {targetBone {boneName=spine_05}}
     {targetBone {boneName=clavicle_l}}
     {targetBone {boneName=clavicle_r}}
     {targetBone {boneName=spine_01}}
    ]
    constraintWeights [1,1,1,1,1,1,1,1]
   }
   n16 = "Constraint - dyn_pigtail_l_05 (PPPPPPPP)"!(
    ComponentPose=@n15
    Alpha=0.8
    ConstraintWeights_0=0.0098
    ConstraintWeights_1:float=0.0525
    ConstraintWeights_2:float=0.3795
    ConstraintWeights_3:float=0.0814
    ConstraintWeights_4:float=0.2137
    ConstraintWeights_5:float=0.235
    ConstraintWeights_6:float=0.0163
    ConstraintWeights_7:float=0.0118
   ) -> (Pose) {
    alphaBoolBlend {blendOption=Linear}
    boneToModify {boneName=dyn_pigtail_l_05}
    constraintSetup [
     {targetBone {boneName=upperarm_twist_01_l}}
     {targetBone {boneName=spine_02}}
     {targetBone {boneName=spine_04}}
     {targetBone {boneName=spine_03}}
     {targetBone {boneName=spine_05}}
     {targetBone {boneName=clavicle_l}}
     {targetBone {boneName=clavicle_r}}
     {targetBone {boneName=spine_01}}
    ]
    constraintWeights [1,1,1,1,1,1,1,1]
   }
   n17 = "Constraint - dyn_pigtail_r_04 (PPPPPPPP)"!(
    ComponentPose=@n16
    Alpha=0.8
    ConstraintWeights_0=0.0049
    ConstraintWeights_1:float=0.0947
    ConstraintWeights_2:float=0.009
    ConstraintWeights_3:float=0.3669
    ConstraintWeights_4:float=0.034
    ConstraintWeights_5:float=0.4759
    ConstraintWeights_6:float=0.001
    ConstraintWeights_7:float=0.0136
   ) -> (Pose) {
    alphaBoolBlend {blendOption=Linear}
    boneToModify {boneName=dyn_pigtail_r_04}
    constraintSetup [
     {targetBone {boneName=spine_02}}
     {targetBone {boneName=spine_04}}
     {targetBone {boneName=spine_03}}
     {targetBone {boneName=spine_05}}
     {targetBone {boneName=clavicle_l}}
     {targetBone {boneName=clavicle_r}}
     {targetBone {boneName=spine_01}}
     {targetBone {boneName=upperarm_twist_01_r}}
    ]
    constraintWeights [1,1,1,1,1,1,1,1]
   }
   n18 = "Constraint - dyn_pigtail_r_05 (PPPPPPPP)"!(
    ComponentPose=@n17
    Alpha=0.8
    ConstraintWeights_0=0.0049
    ConstraintWeights_1:float=0.0947
    ConstraintWeights_2:float=0.009
    ConstraintWeights_3:float=0.3669
    ConstraintWeights_4:float=0.034
    ConstraintWeights_5:float=0.4759
    ConstraintWeights_6:float=0.001
    ConstraintWeights_7:float=0.0136
   ) -> (Pose) {
    alphaBoolBlend {blendOption=Linear}
    boneToModify {boneName=dyn_pigtail_r_05}
    constraintSetup [
     {targetBone {boneName=spine_02}}
     {targetBone {boneName=spine_04}}
     {targetBone {boneName=spine_03}}
     {targetBone {boneName=spine_05}}
     {targetBone {boneName=clavicle_l}}
     {targetBone {boneName=clavicle_r}}
     {targetBone {boneName=spine_01}}
     {targetBone {boneName=upperarm_twist_01_r}}
    ]
    constraintWeights [1,1,1,1,1,1,1,1]
   }
   n19 = "Transform (Modify) Bone - Bone: dyn_pigtail_l_02"(ComponentPose=@n18, Alpha=1, Translation="0,0,1", Rotation="0,0,0", Scale="1,1,1") {
    rotationMode=BMM_Additive
    rotationSpace=BCS_BoneSpace
    translationMode=BMM_Additive
    translationSpace=BCS_BoneSpace
    alphaBoolBlend {blendOption=Linear}
    boneToModify {boneName=dyn_pigtail_l_02}
    translation {z=1}
   }
   n20 = "Transform (Modify) Bone - Bone: dyn_pigtail_r_02"(ComponentPose=@n19, Alpha=1, Translation="0,0,1", Rotation="0,0,0", Scale="1,1,1") {
    rotationMode=BMM_Additive
    rotationSpace=BCS_BoneSpace
    translationMode=BMM_Additive
    translationSpace=BCS_BoneSpace
    alphaBoolBlend {blendOption=Linear}
    boneToModify {boneName=dyn_pigtail_r_02}
    translation {z=1}
   }
   n21 = "Transform (Modify) Bone - Bone: dyn_pigtail_l_03"(ComponentPose=@n20, Alpha=1, Translation="0,0,-1", Rotation="-5,0,0", Scale="1,1,1") {
    rotationMode=BMM_Additive
    rotationSpace=BCS_BoneSpace
    translationMode=BMM_Additive
    translationSpace=BCS_BoneSpace
    alphaBoolBlend {blendOption=Linear}
    boneToModify {boneName=dyn_pigtail_l_03}
    rotation {pitch=-5}
    translation {z=-1}
   }
   n22 = "Transform (Modify) Bone - Bone: dyn_pigtail_r_03"(ComponentPose=@n21, Alpha=1, Translation="0,0,-1", Rotation="-5,0,0", Scale="1,1,1") {
    rotationMode=BMM_Additive
    rotationSpace=BCS_BoneSpace
    translationMode=BMM_Additive
    translationSpace=BCS_BoneSpace
    alphaBoolBlend {blendOption=Linear}
    boneToModify {boneName=dyn_pigtail_r_03}
    rotation {pitch=-5}
    translation {z=-1}
   }
   n23 = "Transform (Modify) Bone - Bone: dyn_pigtail_l_04"(ComponentPose=@n22, Alpha=1, Translation="0,0,-1.5", Rotation="-5,0,0", Scale="1,1,1") {
    rotationMode=BMM_Additive
    rotationSpace=BCS_BoneSpace
    translationMode=BMM_Additive
    translationSpace=BCS_BoneSpace
    alphaBoolBlend {blendOption=Linear}
    boneToModify {boneName=dyn_pigtail_l_04}
    rotation {pitch=-5}
    translation {z=-1.5}
   }
   n24 = "Transform (Modify) Bone - Bone: dyn_pigtail_l_04"(ComponentPose=@n23, Alpha=1, Translation="0,0,0", Rotation="-30,0,0", Scale="1,1,1") {
    rotationMode=BMM_Additive
    rotationSpace=BCS_BoneSpace
    translationMode=BMM_Additive
    translationSpace=BCS_BoneSpace
    alphaBoolBlend {blendOption=Linear}
    boneToModify {boneName=dyn_pigtail_l_04}
    rotation {pitch=-30}
   }
   n25 = "Transform (Modify) Bone - Bone: dyn_pigtail_l_05"(ComponentPose=@n24, Alpha=1, Translation="-1,0,-1.5", Rotation="-5,0,0", Scale="1,1,1") {
    rotationMode=BMM_Additive
    rotationSpace=BCS_BoneSpace
    translationMode=BMM_Additive
    translationSpace=BCS_BoneSpace
    alphaBoolBlend {blendOption=Linear}
    boneToModify {boneName=dyn_pigtail_l_05}
    rotation {pitch=-5}
    translation {x=-1, z=-1.5}
   }
   n26 = "Transform (Modify) Bone - Bone: dyn_pigtail_l_06"(ComponentPose=@n25, Alpha=1, Translation="-2,0,-1.5", Rotation="-5,0,0", Scale="1,1,1") {
    rotationMode=BMM_Additive
    rotationSpace=BCS_BoneSpace
    translationMode=BMM_Additive
    translationSpace=BCS_BoneSpace
    alphaBoolBlend {blendOption=Linear}
    boneToModify {boneName=dyn_pigtail_l_06}
    rotation {pitch=-5}
    translation {x=-2, z=-1.5}
   }
   n27 = "Transform (Modify) Bone - Bone: dyn_pigtail_r_04"(ComponentPose=@n26, Alpha=1, Translation="0,0,-1.5", Rotation="-5,0,0", Scale="1,1,1") {
    rotationMode=BMM_Additive
    rotationSpace=BCS_BoneSpace
    translationMode=BMM_Additive
    translationSpace=BCS_BoneSpace
    alphaBoolBlend {blendOption=Linear}
    boneToModify {boneName=dyn_pigtail_r_04}
    rotation {pitch=-5}
    translation {z=-1.5}
   }
   n28 = "Transform (Modify) Bone - Bone: dyn_pigtail_r_04"(ComponentPose=@n27, Alpha=1, Translation="0,0,0", Rotation="-30,0,0", Scale="1,1,1") {
    rotationMode=BMM_Additive
    rotationSpace=BCS_BoneSpace
    translationMode=BMM_Additive
    translationSpace=BCS_BoneSpace
    alphaBoolBlend {blendOption=Linear}
    boneToModify {boneName=dyn_pigtail_r_04}
    rotation {pitch=-30}
   }
   n29 = "Transform (Modify) Bone - Bone: dyn_pigtail_r_05"(ComponentPose=@n28, Alpha=1, Translation="-1,0,0", Rotation="-10,0,0", Scale="1,1,1") {
    rotationMode=BMM_Additive
    rotationSpace=BCS_BoneSpace
    translationMode=BMM_Additive
    translationSpace=BCS_BoneSpace
    alphaBoolBlend {blendOption=Linear}
    boneToModify {boneName=dyn_pigtail_r_05}
    rotation {pitch=-10}
    translation {x=-1}
   }
   n30 = "Transform (Modify) Bone - Bone: dyn_pigtail_r_06"(ComponentPose=@n29, Alpha=1, Translation="-2,0,-1.5", Rotation="-5,0,0", Scale="1,1,1") {
    rotationMode=BMM_Additive
    rotationSpace=BCS_BoneSpace
    translationMode=BMM_Additive
    translationSpace=BCS_BoneSpace
    alphaBoolBlend {blendOption=Linear}
    boneToModify {boneName=dyn_pigtail_r_06}
    rotation {pitch=-5}
    translation {x=-2, z=-1.5}
   }
   n31 = Component To Local(ComponentPose=@n30)
   n32 = Get backbling_equipped() -> (backbling_equipped:bool)
   n33 = "Transform (Modify) Bone - Bone: dyn_pigtail_l_01"(
    ComponentPose=Local To Component(LocalPose=Blend Poses by bool(BlendPose_0=@n31, BlendPose_1=@n10, bActiveValue=@n32) {blendType=Linear})
    Alpha=1
    Translation="0,0,0"
    Rotation="10,0,0"
    Scale="1,1,1"
   ) {
    rotationMode=BMM_Additive
    rotationSpace=BCS_BoneSpace
    translationMode=BMM_Additive
    translationSpace=BCS_BoneSpace
    alphaBoolBlend {blendOption=Linear}
    boneToModify {boneName=dyn_pigtail_l_01}
    rotation {pitch=10}
   }
   n34 = "Transform (Modify) Bone - Bone: dyn_pigtail_r_01"(ComponentPose=@n33, Alpha=1, Translation="0,0,0", Rotation="10,0,0", Scale="1,1,1") {
    rotationMode=BMM_Additive
    rotationSpace=BCS_BoneSpace
    translationMode=BMM_Additive
    translationSpace=BCS_BoneSpace
    alphaBoolBlend {blendOption=Linear}
    boneToModify {boneName=dyn_pigtail_r_01}
    rotation {pitch=10}
   }
   n35 = "Transform (Modify) Bone - Bone: dyn_pigtail_l_03"(ComponentPose=@n34, Alpha=1, Translation="0,0,0", Rotation="10,0,0", Scale="1,1,1") {
    rotationMode=BMM_Additive
    rotationSpace=BCS_BoneSpace
    translationMode=BMM_Additive
    translationSpace=BCS_BoneSpace
    alphaBoolBlend {blendOption=Linear}
    boneToModify {boneName=dyn_pigtail_l_03}
    rotation {pitch=10}
   }
   n36 = "Transform (Modify) Bone - Bone: dyn_pigtail_r_03"(ComponentPose=@n35, Alpha=1, Translation="0,0,0", Rotation="10,0,0", Scale="1,1,1") {
    rotationMode=BMM_Additive
    rotationSpace=BCS_BoneSpace
    translationMode=BMM_Additive
    translationSpace=BCS_BoneSpace
    alphaBoolBlend {blendOption=Linear}
    boneToModify {boneName=dyn_pigtail_r_03}
    rotation {pitch=10}
   }
   Output Pose(Result=Component To Local(ComponentPose=@n36)) {layerGroup=None}
   n37 = Layered blend per bone(
    BasePose=Copy Pose From Mesh() {bCopyCurves=true, bCopyCustomAttributes=true, bUseAttachedParent=true, bUseMeshPose=true}
    BlendPoses_0=Copy Pose From Mesh(SourceMeshComponent=Get Body_Mesh()) {bCopyCurves=true, bCopyCustomAttributes=true, bUseAttachedParent=true, bUseMeshPose=true}
   ) {layerSetup [{branchFilters [{boneName=upperarm_l}, {boneName=upperarm_r}, {boneName=clavicle_r}, {boneName=clavicle_l}]}]}
   n38 = Anim Dynamics - Bone: dyn_hair_l_long_01(ComponentPose=Local To Component(LocalPose=@n37), Alpha=1, ExternalForce="0,0,0") {
    angularDampingOverride=1
    angularSpringConstant=5000
    bAngularSpring=true
    bGravityOverrideInSimSpace=true
    bOverrideAngularDamping=true
    bUseGravityOverride=true
    simulationSpace=BoneRelative
    alphaBoolBlend {blendOption=Linear}
    boundBone {boneName=dyn_hair_l_long_01}
    componentAppliedLinearAccClamp=<1500 1500 1500>
    componentLinearAccScale=<0.4 0.4 1>
    componentLinearVelScale=<0.2 0.2 0.4>
    gravityOverride {z=-980}
    physicsBodyDefinitions [{
      sphereCollisionRadius=0
      boundBone {boneName=dyn_hair_l_long_01}
      boxExtents=<20 20 20>
      constraintSetup {angularLimitsMax=<3 5 5>, angularLimitsMin=<-3 -5 -5>, angularTarget=<1 0 0>}
      localJointOffset=<3 0 0>
     }]
    relativeSpaceBone {boneName=root}
   }
   n39 = Anim Dynamics - Bone: dyn_hair_l_long_02(ComponentPose=@n38, Alpha=1, ExternalForce="0,0,0") {
    angularDampingOverride=1
    angularSpringConstant=5000
    bAngularSpring=true
    bGravityOverrideInSimSpace=true
    bOverrideAngularDamping=true
    bUseGravityOverride=true
    simulationSpace=BoneRelative
    alphaBoolBlend {blendOption=Linear}
    boundBone {boneName=dyn_hair_l_long_02}
    componentAppliedLinearAccClamp=<1500 1500 1500>
    componentLinearAccScale=<0.4 0.4 1>
    componentLinearVelScale=<0.2 0.2 0.4>
    gravityOverride {z=-980}
    physicsBodyDefinitions [{
      sphereCollisionRadius=0
      boundBone {boneName=dyn_hair_l_long_02}
      boxExtents=<20 20 20>
      constraintSetup {angularLimitsMax=<3 5 10>, angularLimitsMin=<-3 -10 -10>, angularTarget=<1 0 0>}
      localJointOffset=<3 0 0>
     }]
    relativeSpaceBone {boneName=root}
   }
   n40 = Anim Dynamics - Bone: dyn_hair_bk_l_01(ComponentPose=@n39, Alpha=1, ExternalForce="0,0,0") {
    angularDampingOverride=1
    angularSpringConstant=5000
    bAngularSpring=true
    bGravityOverrideInSimSpace=true
    bOverrideAngularDamping=true
    bUseGravityOverride=true
    simulationSpace=BoneRelative
    alphaBoolBlend {blendOption=Linear}
    boundBone {boneName=dyn_hair_bk_l_01}
    componentAppliedLinearAccClamp=<1500 1500 1500>
    componentLinearAccScale=<0.4 0.4 1>
    componentLinearVelScale=<0.2 0.2 0.4>
    gravityOverride {z=-980}
    physicsBodyDefinitions [{
      sphereCollisionRadius=0
      boundBone {boneName=dyn_hair_bk_l_01}
      boxExtents=<20 20 20>
      constraintSetup {angularLimitsMax=<3 5 5>, angularLimitsMin=<-3 -5 -5>, angularTarget=<1 0 0>}
      localJointOffset=<3 0 0>
     }]
    relativeSpaceBone {boneName=root}
   }
   n41 = Anim Dynamics - Bone: dyn_hair_bk_l_02(ComponentPose=@n40, Alpha=1, ExternalForce="0,0,0") {
    angularDampingOverride=1
    angularSpringConstant=5000
    bAngularSpring=true
    bGravityOverrideInSimSpace=true
    bOverrideAngularDamping=true
    bUseGravityOverride=true
    simulationSpace=BoneRelative
    alphaBoolBlend {blendOption=Linear}
    boundBone {boneName=dyn_hair_bk_l_02}
    componentAppliedLinearAccClamp=<1500 1500 1500>
    componentLinearAccScale=<0.4 0.4 1>
    componentLinearVelScale=<0.2 0.2 0.4>
    gravityOverride {z=-980}
    physicsBodyDefinitions [{
      sphereCollisionRadius=0
      boundBone {boneName=dyn_hair_bk_l_02}
      boxExtents=<20 20 20>
      constraintSetup {angularLimitsMax=<3 5 10>, angularLimitsMin=<-3 -10 -10>, angularTarget=<1 0 0>}
      localJointOffset=<3 0 0>
     }]
    relativeSpaceBone {boneName=root}
   }
   n42 = Anim Dynamics - Bone: dyn_hair_l_mid_01(ComponentPose=@n41, Alpha=1, ExternalForce="0,0,0") {
    angularDampingOverride=1
    angularSpringConstant=5000
    bAngularSpring=true
    bGravityOverrideInSimSpace=true
    bOverrideAngularDamping=true
    bUseGravityOverride=true
    simulationSpace=BoneRelative
    alphaBoolBlend {blendOption=Linear}
    boundBone {boneName=dyn_hair_l_mid_01}
    componentAppliedLinearAccClamp=<1500 1500 1500>
    componentLinearAccScale=<0.4 0.4 1>
    componentLinearVelScale=<0.2 0.2 0.4>
    gravityOverride {z=-980}
    physicsBodyDefinitions [{
      sphereCollisionRadius=0
      boundBone {boneName=dyn_hair_l_mid_01}
      boxExtents=<20 20 20>
      constraintSetup {angularLimitsMax=<3 5 5>, angularLimitsMin=<-3 -5 -5>, angularTarget=<1 0 0>}
      localJointOffset=<3 0 0>
     }]
    relativeSpaceBone {boneName=root}
   }
   n43 = Anim Dynamics - Bone: dyn_hair_l_mid_02(ComponentPose=@n42, Alpha=1, ExternalForce="0,0,0") {
    angularDampingOverride=1
    angularSpringConstant=5000
    bAngularSpring=true
    bGravityOverrideInSimSpace=true
    bOverrideAngularDamping=true
    bUseGravityOverride=true
    simulationSpace=BoneRelative
    alphaBoolBlend {blendOption=Linear}
    boundBone {boneName=dyn_hair_l_mid_02}
    componentAppliedLinearAccClamp=<1500 1500 1500>
    componentLinearAccScale=<0.4 0.4 1>
    componentLinearVelScale=<0.2 0.2 0.4>
    gravityOverride {z=-980}
    physicsBodyDefinitions [{
      sphereCollisionRadius=0
      boundBone {boneName=dyn_hair_l_mid_02}
      boxExtents=<20 20 20>
      constraintSetup {angularLimitsMax=<3 5 10>, angularLimitsMin=<-3 -10 -10>, angularTarget=<1 0 0>}
      localJointOffset=<3 0 0>
     }]
    relativeSpaceBone {boneName=root}
   }
   n44 = Anim Dynamics - Bone: dyn_hair_top_l_01(ComponentPose=@n43, Alpha=1, ExternalForce="0,0,0") {
    angularDampingOverride=1
    angularSpringConstant=5000
    bAngularSpring=true
    bGravityOverrideInSimSpace=true
    bOverrideAngularDamping=true
    bUseGravityOverride=true
    simulationSpace=BoneRelative
    alphaBoolBlend {blendOption=Linear}
    boundBone {boneName=dyn_hair_top_l_01}
    componentAppliedLinearAccClamp=<1500 1500 1500>
    componentLinearAccScale=<0.4 0.4 1>
    componentLinearVelScale=<0.2 0.2 0.4>
    gravityOverride {z=-980}
    physicsBodyDefinitions [{
      sphereCollisionRadius=0
      boundBone {boneName=dyn_hair_top_l_01}
      boxExtents=<20 20 20>
      constraintSetup {angularLimitsMax=<3 5 5>, angularLimitsMin=<-3 -5 -5>, angularTarget=<1 0 0>}
      localJointOffset=<3 0 0>
     }]
    relativeSpaceBone {boneName=root}
   }
   n45 = Anim Dynamics - Bone: dyn_hair_top_l_02(ComponentPose=@n44, Alpha=1, ExternalForce="0,0,0") {
    angularDampingOverride=1
    angularSpringConstant=5000
    bAngularSpring=true
    bGravityOverrideInSimSpace=true
    bOverrideAngularDamping=true
    bUseGravityOverride=true
    simulationSpace=BoneRelative
    alphaBoolBlend {blendOption=Linear}
    boundBone {boneName=dyn_hair_top_l_02}
    componentAppliedLinearAccClamp=<1500 1500 1500>
    componentLinearAccScale=<0.4 0.4 1>
    componentLinearVelScale=<0.2 0.2 0.4>
    gravityOverride {z=-980}
    physicsBodyDefinitions [{
      sphereCollisionRadius=0
      boundBone {boneName=dyn_hair_top_l_02}
      boxExtents=<20 20 20>
      constraintSetup {angularLimitsMax=<3 5 7>, angularLimitsMin=<-3 -7 -7>, angularTarget=<1 0 0>}
      localJointOffset=<3 0 0>
     }]
    relativeSpaceBone {boneName=root}
   }
   n46 = Anim Dynamics - Bone: dyn_hair_bk_r_01(ComponentPose=@n45, Alpha=1, ExternalForce="0,0,0") {
    angularDampingOverride=1
    angularSpringConstant=5000
    bAngularSpring=true
    bGravityOverrideInSimSpace=true
    bOverrideAngularDamping=true
    bUseGravityOverride=true
    simulationSpace=BoneRelative
    alphaBoolBlend {blendOption=Linear}
    boundBone {boneName=dyn_hair_bk_r_01}
    componentAppliedLinearAccClamp=<1500 1500 1500>
    componentLinearAccScale=<0.4 0.4 1>
    componentLinearVelScale=<0.2 0.2 0.4>
    gravityOverride {z=-980}
    physicsBodyDefinitions [{
      sphereCollisionRadius=0
      boundBone {boneName=dyn_hair_bk_r_01}
      boxExtents=<20 20 20>
      constraintSetup {angularLimitsMax=<3 5 10>, angularLimitsMin=<-3 -10 -10>, angularTarget=<1 0 0>}
      localJointOffset=<3 0 0>
     }]
    relativeSpaceBone {boneName=root}
   }
   n47 = Anim Dynamics - Bone: dyn_hair_top_r_01(ComponentPose=@n46, Alpha=1, ExternalForce="0,0,0") {
    angularDampingOverride=1
    angularSpringConstant=5000
    bAngularSpring=true
    bGravityOverrideInSimSpace=true
    bOverrideAngularDamping=true
    bUseGravityOverride=true
    simulationSpace=BoneRelative
    alphaBoolBlend {blendOption=Linear}
    boundBone {boneName=dyn_hair_top_r_01}
    componentAppliedLinearAccClamp=<1500 1500 1500>
    componentLinearAccScale=<0.4 0.4 1>
    componentLinearVelScale=<0.2 0.2 0.4>
    gravityOverride {z=-980}
    physicsBodyDefinitions [{
      sphereCollisionRadius=0
      boundBone {boneName=dyn_hair_top_r_01}
      boxExtents=<20 20 20>
      constraintSetup {angularLimitsMax=<3 5 10>, angularLimitsMin=<-3 -10 -10>, angularTarget=<1 0 0>}
      localJointOffset=<3 0 0>
     }]
    relativeSpaceBone {boneName=root}
   }
   n48 = Anim Dynamics - Bone: dyn_hair_top_r_02(ComponentPose=@n47, Alpha=1, ExternalForce="0,0,0") {
    angularDampingOverride=1
    angularSpringConstant=5000
    bAngularSpring=true
    bGravityOverrideInSimSpace=true
    bOverrideAngularDamping=true
    bUseGravityOverride=true
    simulationSpace=BoneRelative
    alphaBoolBlend {blendOption=Linear}
    boundBone {boneName=dyn_hair_top_r_02}
    componentAppliedLinearAccClamp=<1500 1500 1500>
    componentLinearAccScale=<0.4 0.4 1>
    componentLinearVelScale=<0.2 0.2 0.4>
    gravityOverride {z=-980}
    physicsBodyDefinitions [{
      sphereCollisionRadius=0
      boundBone {boneName=dyn_hair_top_r_02}
      boxExtents=<20 20 20>
      constraintSetup {angularLimitsMax=<3 5 7>, angularLimitsMin=<-3 -7 -7>, angularTarget=<1 0 0>}
      localJointOffset=<3 0 0>
     }]
    relativeSpaceBone {boneName=root}
   }
   n49 = Anim Dynamics - Bone: dyn_hair_r_mid_03_a(ComponentPose=@n48, Alpha=1, ExternalForce="0,0,0") {
    angularDampingOverride=1
    angularSpringConstant=5000
    bAngularSpring=true
    bGravityOverrideInSimSpace=true
    bOverrideAngularDamping=true
    bUseGravityOverride=true
    simulationSpace=BoneRelative
    alphaBoolBlend {blendOption=Linear}
    boundBone {boneName=dyn_hair_r_mid_03_a}
    componentAppliedLinearAccClamp=<1500 1500 1500>
    componentLinearAccScale=<0.4 0.4 1>
    componentLinearVelScale=<0.2 0.2 0.4>
    gravityOverride {z=-980}
    physicsBodyDefinitions [{
      sphereCollisionRadius=0
      boundBone {boneName=dyn_hair_r_mid_03_a}
      boxExtents=<20 20 20>
      constraintSetup {angularLimitsMax=<3 5 5>, angularLimitsMin=<-3 -5 -5>, angularTarget=<1 0 0>}
      localJointOffset=<3 0 0>
     }]
    relativeSpaceBone {boneName=root}
   }
   n50 = Anim Dynamics - Bone: dyn_hair_r_mid_03_b(ComponentPose=@n49, Alpha=1, ExternalForce="0,0,0") {
    angularDampingOverride=1
    angularSpringConstant=5000
    bAngularSpring=true
    bGravityOverrideInSimSpace=true
    bOverrideAngularDamping=true
    bUseGravityOverride=true
    simulationSpace=BoneRelative
    alphaBoolBlend {blendOption=Linear}
    boundBone {boneName=dyn_hair_r_mid_03_b}
    componentAppliedLinearAccClamp=<1500 1500 1500>
    componentLinearAccScale=<0.4 0.4 1>
    componentLinearVelScale=<0.2 0.2 0.4>
    gravityOverride {z=-980}
    physicsBodyDefinitions [{
      sphereCollisionRadius=0
      boundBone {boneName=dyn_hair_r_mid_03_b}
      boxExtents=<20 20 20>
      constraintSetup {angularLimitsMax=<3 5 10>, angularLimitsMin=<-3 -10 -10>, angularTarget=<1 0 0>}
      localJointOffset=<3 0 0>
     }]
    relativeSpaceBone {boneName=root}
   }
   n51 = Anim Dynamics - Bone: dyn_hair_r_mid_02__base(ComponentPose=@n50, Alpha=1, ExternalForce="0,0,0") {
    angularDampingOverride=1
    angularSpringConstant=5000
    bAngularSpring=true
    bGravityOverrideInSimSpace=true
    bOverrideAngularDamping=true
    bUseGravityOverride=true
    simulationSpace=BoneRelative
    alphaBoolBlend {blendOption=Linear}
    boundBone {boneName=dyn_hair_r_mid_02__base}
    componentAppliedLinearAccClamp=<1500 1500 1500>
    componentLinearAccScale=<0.4 0.4 1>
    componentLinearVelScale=<0.2 0.2 0.4>
    gravityOverride {z=-980}
    physicsBodyDefinitions [{
      sphereCollisionRadius=0
      boundBone {boneName=dyn_hair_r_mid_02__base}
      boxExtents=<20 20 20>
      constraintSetup {angularLimitsMax=<15 5 60>, angularLimitsMin=<-8 -15 -60>, angularTarget=<1 0 0>}
      localJointOffset=<3 0 0>
     }]
    relativeSpaceBone {boneName=root}
   }
   n52 = Anim Dynamics - Bone: dyn_hair_r_mid_02_a(ComponentPose=@n51, Alpha=1, ExternalForce="0,0,0") {
    angularDampingOverride=1
    angularSpringConstant=5000
    bAngularSpring=true
    bGravityOverrideInSimSpace=true
    bOverrideAngularDamping=true
    bUseGravityOverride=true
    simulationSpace=BoneRelative
    alphaBoolBlend {blendOption=Linear}
    boundBone {boneName=dyn_hair_r_mid_02_a}
    componentAppliedLinearAccClamp=<1500 1500 1500>
    componentLinearAccScale=<0.4 0.4 1>
    componentLinearVelScale=<0.2 0.2 0.4>
    gravityOverride {z=-980}
    physicsBodyDefinitions [{
      sphereCollisionRadius=0
      boundBone {boneName=dyn_hair_r_mid_02_a}
      boxExtents=<20 20 20>
      constraintSetup {angularLimitsMax=<3 5 5>, angularLimitsMin=<-3 -5 -5>, angularTarget=<1 0 0>}
      localJointOffset=<3 0 0>
     }]
    relativeSpaceBone {boneName=root}
   }
   n53 = Anim Dynamics - Bone: dyn_hair_r_mid_02_b(ComponentPose=@n52, Alpha=1, ExternalForce="0,0,0") {
    angularDampingOverride=1
    angularSpringConstant=5000
    bAngularSpring=true
    bGravityOverrideInSimSpace=true
    bOverrideAngularDamping=true
    bUseGravityOverride=true
    simulationSpace=BoneRelative
    alphaBoolBlend {blendOption=Linear}
    boundBone {boneName=dyn_hair_r_mid_02_b}
    componentAppliedLinearAccClamp=<1500 1500 1500>
    componentLinearAccScale=<0.4 0.4 1>
    componentLinearVelScale=<0.2 0.2 0.4>
    gravityOverride {z=-980}
    physicsBodyDefinitions [{
      sphereCollisionRadius=0
      boundBone {boneName=dyn_hair_r_mid_02_b}
      boxExtents=<20 20 20>
      constraintSetup {angularLimitsMax=<8 5 10>, angularLimitsMin=<-8 -10 0>, angularTarget=<1 0 0>}
      localJointOffset=<3 0 0>
     }]
    relativeSpaceBone {boneName=root}
   }
   n54 = Anim Dynamics - Bone: dyn_hair_r_mid_01_a(ComponentPose=@n53, Alpha=1, ExternalForce="0,0,0") {
    angularDampingOverride=1
    angularSpringConstant=5000
    bAngularSpring=true
    bGravityOverrideInSimSpace=true
    bOverrideAngularDamping=true
    bUseGravityOverride=true
    simulationSpace=BoneRelative
    alphaBoolBlend {blendOption=Linear}
    boundBone {boneName=dyn_hair_r_mid_01_a}
    componentAppliedLinearAccClamp=<1500 1500 1500>
    componentLinearAccScale=<0.4 0.4 1>
    componentLinearVelScale=<0.2 0.2 0.4>
    gravityOverride {z=-980}
    physicsBodyDefinitions [{
      sphereCollisionRadius=0
      boundBone {boneName=dyn_hair_r_mid_01_a}
      boxExtents=<20 20 20>
      constraintSetup {angularLimitsMax=<3 5 5>, angularLimitsMin=<-3 -5 -5>, angularTarget=<1 0 0>}
      localJointOffset=<3 0 0>
     }]
    relativeSpaceBone {boneName=root}
   }
   n55 = Anim Dynamics - Bone: dyn_hair_r_mid_01_b(ComponentPose=@n54, Alpha=1, ExternalForce="0,0,0") {
    angularDampingOverride=1
    angularSpringConstant=5000
    bAngularSpring=true
    bGravityOverrideInSimSpace=true
    bOverrideAngularDamping=true
    bUseGravityOverride=true
    simulationSpace=BoneRelative
    alphaBoolBlend {blendOption=Linear}
    boundBone {boneName=dyn_hair_r_mid_01_b}
    componentAppliedLinearAccClamp=<1500 1500 1500>
    componentLinearAccScale=<0.4 0.4 1>
    componentLinearVelScale=<0.2 0.2 0.4>
    gravityOverride {z=-980}
    physicsBodyDefinitions [{
      sphereCollisionRadius=0
      boundBone {boneName=dyn_hair_r_mid_01_b}
      boxExtents=<20 20 20>
      constraintSetup {angularLimitsMax=<3 5 10>, angularLimitsMin=<-3 -10 -10>, angularTarget=<1 0 0>}
      localJointOffset=<3 0 0>
     }]
    relativeSpaceBone {boneName=root}
   }
   n56 = Anim Dynamics - Bone: dyn_hair_top_mid_01(ComponentPose=@n55, Alpha=1, ExternalForce="0,0,0") {
    angularDampingOverride=1
    angularSpringConstant=5000
    bAngularSpring=true
    bGravityOverrideInSimSpace=true
    bOverrideAngularDamping=true
    bUseGravityOverride=true
    simulationSpace=BoneRelative
    alphaBoolBlend {blendOption=Linear}
    boundBone {boneName=dyn_hair_top_mid_01}
    componentAppliedLinearAccClamp=<1500 1500 1500>
    componentLinearAccScale=<0.4 0.4 1>
    componentLinearVelScale=<0.2 0.2 0.4>
    gravityOverride {z=-980}
    physicsBodyDefinitions [{
      sphereCollisionRadius=0
      boundBone {boneName=dyn_hair_top_mid_01}
      boxExtents=<20 20 20>
      constraintSetup {angularLimitsMax=<3 5 5>, angularLimitsMin=<-3 -5 -5>, angularTarget=<1 0 0>}
      localJointOffset=<3 0 0>
     }]
    relativeSpaceBone {boneName=root}
   }
   n57 = Anim Dynamics - Bone: dyn_hair_top_mid_02(ComponentPose=@n56, Alpha=1, ExternalForce="0,0,0") {
    angularDampingOverride=1
    angularSpringConstant=5000
    bAngularSpring=true
    bGravityOverrideInSimSpace=true
    bOverrideAngularDamping=true
    bUseGravityOverride=true
    simulationSpace=BoneRelative
    alphaBoolBlend {blendOption=Linear}
    boundBone {boneName=dyn_hair_top_mid_02}
    componentAppliedLinearAccClamp=<1500 1500 1500>
    componentLinearAccScale=<0.4 0.4 1>
    componentLinearVelScale=<0.2 0.2 0.4>
    gravityOverride {z=-980}
    physicsBodyDefinitions [{
      sphereCollisionRadius=0
      boundBone {boneName=dyn_hair_top_mid_02}
      boxExtents=<20 20 20>
      constraintSetup {angularLimitsMax=<3 7 7>, angularLimitsMin=<-3 -7 -7>, angularTarget=<1 0 0>}
      localJointOffset=<3 0 0>
     }]
    relativeSpaceBone {boneName=root}
   }
   n58 = Anim Dynamics - Bone: dyn_r_fr_bang_01(ComponentPose=@n57, Alpha=1, ExternalForce="0,0,0") {
    angularDampingOverride=1
    angularSpringConstant=5000
    bAngularSpring=true
    bGravityOverrideInSimSpace=true
    bOverrideAngularDamping=true
    bUseGravityOverride=true
    simulationSpace=BoneRelative
    alphaBoolBlend {blendOption=Linear}
    boundBone {boneName=dyn_r_fr_bang_01}
    componentAppliedLinearAccClamp=<1500 1500 1500>
    componentLinearAccScale=<0.4 0.4 1>
    componentLinearVelScale=<0.2 0.2 0.4>
    gravityOverride {z=-980}
    physicsBodyDefinitions [{
      sphereCollisionRadius=0
      boundBone {boneName=dyn_r_fr_bang_01}
      boxExtents=<20 20 20>
      constraintSetup {angularLimitsMax=<2 2 2>, angularLimitsMin=<-2 -2 -2>, angularTarget=<1 0 0>}
      localJointOffset=<3 0 0>
     }]
    relativeSpaceBone {boneName=root}
   }
   n59 = Anim Dynamics - Bone: dyn_r_fr_bang_02(ComponentPose=@n58, Alpha=1, ExternalForce="0,0,0") {
    angularDampingOverride=1
    angularSpringConstant=5000
    bAngularSpring=true
    bGravityOverrideInSimSpace=true
    bOverrideAngularDamping=true
    bUseGravityOverride=true
    simulationSpace=BoneRelative
    alphaBoolBlend {blendOption=Linear}
    boundBone {boneName=dyn_r_fr_bang_02}
    componentAppliedLinearAccClamp=<1500 1500 1500>
    componentLinearAccScale=<0.4 0.4 1>
    componentLinearVelScale=<0.2 0.2 0.4>
    gravityOverride {z=-980}
    physicsBodyDefinitions [{
      sphereCollisionRadius=0
      boundBone {boneName=dyn_r_fr_bang_02}
      boxExtents=<20 20 20>
      constraintSetup {angularLimitsMax=<3 -5 10>, angularLimitsMin=<-3 -15 -10>, angularTarget=<1 0 0>}
      localJointOffset=<3 0 0>
     }]
    relativeSpaceBone {boneName=root}
   }
   n60 = Anim Dynamics - Bone: dyn_r_fr_bang_03(ComponentPose=@n59, Alpha=1, ExternalForce="0,0,0") {
    angularDampingOverride=1
    angularSpringConstant=5000
    bAngularSpring=true
    bGravityOverrideInSimSpace=true
    bOverrideAngularDamping=true
    bUseGravityOverride=true
    simulationSpace=BoneRelative
    alphaBoolBlend {blendOption=Linear}
    boundBone {boneName=dyn_r_fr_bang_03}
    componentAppliedLinearAccClamp=<1500 1500 1500>
    componentLinearAccScale=<0.4 0.4 1>
    componentLinearVelScale=<0.2 0.2 0.4>
    gravityOverride {z=-980}
    physicsBodyDefinitions [{
      sphereCollisionRadius=0
      boundBone {boneName=dyn_r_fr_bang_03}
      boxExtents=<20 20 20>
      constraintSetup {angularLimitsMax=<3 -5 10>, angularLimitsMin=<-3 -15 -10>, angularTarget=<1 0 0>}
      localJointOffset=<3 0 0>
     }]
    relativeSpaceBone {boneName=root}
   }
   n61 = Anim Dynamics - Bone: dyn_l_fr_bang_01(ComponentPose=@n60, Alpha=1, ExternalForce="0,0,0") {
    angularDampingOverride=1
    angularSpringConstant=5000
    bAngularSpring=true
    bGravityOverrideInSimSpace=true
    bOverrideAngularDamping=true
    bUseGravityOverride=true
    simulationSpace=BoneRelative
    alphaBoolBlend {blendOption=Linear}
    boundBone {boneName=dyn_l_fr_bang_01}
    componentAppliedLinearAccClamp=<1500 1500 1500>
    componentLinearAccScale=<0.4 0.4 1>
    componentLinearVelScale=<0.2 0.2 0.4>
    gravityOverride {z=-980}
    physicsBodyDefinitions [{
      sphereCollisionRadius=0
      boundBone {boneName=dyn_l_fr_bang_01}
      boxExtents=<20 20 20>
      constraintSetup {angularLimitsMax=<2 2 2>, angularLimitsMin=<-2 -2 -2>, angularTarget=<1 0 0>}
      localJointOffset=<3 0 0>
     }]
    relativeSpaceBone {boneName=root}
   }
   n62 = Anim Dynamics - Bone: dyn_l_fr_bang_02(ComponentPose=@n61, Alpha=1, ExternalForce="0,0,0") {
    angularDampingOverride=1
    angularSpringConstant=5000
    bAngularSpring=true
    bGravityOverrideInSimSpace=true
    bOverrideAngularDamping=true
    bUseGravityOverride=true
    simulationSpace=BoneRelative
    alphaBoolBlend {blendOption=Linear}
    boundBone {boneName=dyn_l_fr_bang_02}
    componentAppliedLinearAccClamp=<1500 1500 1500>
    componentLinearAccScale=<0.4 0.4 1>
    componentLinearVelScale=<0.2 0.2 0.4>
    gravityOverride {z=-980}
    physicsBodyDefinitions [{
      sphereCollisionRadius=0
      boundBone {boneName=dyn_l_fr_bang_02}
      boxExtents=<20 20 20>
      constraintSetup {angularLimitsMax=<3 -5 10>, angularLimitsMin=<-3 -15 -10>, angularTarget=<1 0 0>}
      localJointOffset=<3 0 0>
     }]
    relativeSpaceBone {boneName=root}
   }
   n63 = Anim Dynamics - Bone: dyn_l_fr_bang_03(ComponentPose=@n62, Alpha=1, ExternalForce="0,0,0") {
    angularDampingOverride=1
    angularSpringConstant=5000
    bAngularSpring=true
    bGravityOverrideInSimSpace=true
    bOverrideAngularDamping=true
    bUseGravityOverride=true
    simulationSpace=BoneRelative
    alphaBoolBlend {blendOption=Linear}
    boundBone {boneName=dyn_l_fr_bang_03}
    componentAppliedLinearAccClamp=<1500 1500 1500>
    componentLinearAccScale=<0.4 0.4 1>
    componentLinearVelScale=<0.2 0.2 0.4>
    gravityOverride {z=-980}
    physicsBodyDefinitions [{
      sphereCollisionRadius=0
      boundBone {boneName=dyn_l_fr_bang_03}
      boxExtents=<20 20 20>
      constraintSetup {angularLimitsMax=<3 -5 10>, angularLimitsMin=<-3 -15 -10>, angularTarget=<1 0 0>}
      localJointOffset=<3 0 0>
     }]
    relativeSpaceBone {boneName=root}
   }
   n64 = Anim Dynamics - Bone: dyn_hair_r_bang_bk_01(ComponentPose=@n63, Alpha=1, ExternalForce="0,0,0") {
    angularDampingOverride=1
    angularSpringConstant=5000
    bAngularSpring=true
    bGravityOverrideInSimSpace=true
    bOverrideAngularDamping=true
    bUseGravityOverride=true
    simulationSpace=BoneRelative
    alphaBoolBlend {blendOption=Linear}
    boundBone {boneName=dyn_hair_r_bang_bk_01}
    componentAppliedLinearAccClamp=<1500 1500 1500>
    componentLinearAccScale=<0.4 0.4 1>
    componentLinearVelScale=<0.2 0.2 0.4>
    gravityOverride {z=-980}
    physicsBodyDefinitions [{
      sphereCollisionRadius=0
      boundBone {boneName=dyn_hair_r_bang_bk_01}
      boxExtents=<20 20 20>
      constraintSetup {angularLimitsMax=<15 5 5>, angularLimitsMin=<-8 -15 -25>, angularTarget=<1 0 0>}
      localJointOffset=<3 0 0>
     }]
    relativeSpaceBone {boneName=root}
   }
   n65 = Anim Dynamics - Bone: dyn_hair_r_bang_bk_02(ComponentPose=@n64, Alpha=1, ExternalForce="0,0,0") {
    angularDampingOverride=1
    angularSpringConstant=5000
    bAngularSpring=true
    bGravityOverrideInSimSpace=true
    bOverrideAngularDamping=true
    bUseGravityOverride=true
    simulationSpace=BoneRelative
    alphaBoolBlend {blendOption=Linear}
    boundBone {boneName=dyn_hair_r_bang_bk_02}
    componentAppliedLinearAccClamp=<1500 1500 1500>
    componentLinearAccScale=<0.4 0.4 1>
    componentLinearVelScale=<0.2 0.2 0.4>
    gravityOverride {z=-980}
    physicsBodyDefinitions [{
      sphereCollisionRadius=0
      boundBone {boneName=dyn_hair_r_bang_bk_02}
      boxExtents=<20 20 20>
      constraintSetup {angularLimitsMax=<15 5 8>, angularLimitsMin=<-8 -15 -12>, angularTarget=<1 0 0>}
      localJointOffset=<3 0 0>
     }]
    relativeSpaceBone {boneName=root}
   }
   n66 = Anim Dynamics - Bone: dyn_hair_r_bang_bk_03(ComponentPose=@n65, Alpha=1, ExternalForce="0,0,0") {
    angularDampingOverride=1
    angularSpringConstant=5000
    bAngularSpring=true
    bGravityOverrideInSimSpace=true
    bOverrideAngularDamping=true
    bUseGravityOverride=true
    simulationSpace=BoneRelative
    alphaBoolBlend {blendOption=Linear}
    boundBone {boneName=dyn_hair_r_bang_bk_03}
    componentAppliedLinearAccClamp=<1500 1500 1500>
    componentLinearAccScale=<0.4 0.4 1>
    componentLinearVelScale=<0.2 0.2 0.4>
    gravityOverride {z=-980}
    physicsBodyDefinitions [{
      sphereCollisionRadius=0
      boundBone {boneName=dyn_hair_r_bang_bk_03}
      boxExtents=<20 20 20>
      constraintSetup {angularLimitsMax=<8 5 0>, angularLimitsMin=<-8 -15 0>, angularTarget=<1 0 0>}
      localJointOffset=<3 0 0>
     }]
    relativeSpaceBone {boneName=root}
   }
   n67 = "Constraint - dyn_r_earring_mstr (P)"(ComponentPose=@n66, Alpha=1, ConstraintWeights_0=1) {
    alphaBoolBlend {blendOption=Linear}
    boneToModify {boneName=dyn_r_earring_mstr}
    constraintSetup [{targetBone {boneName=head}}]
    constraintWeights [1]
   }
   n68 = "Constraint - dyn_l_earring_mstr (P)"(ComponentPose=@n67, Alpha=1, ConstraintWeights_0=1) {
    alphaBoolBlend {blendOption=Linear}
    boneToModify {boneName=dyn_l_earring_mstr}
    constraintSetup [{targetBone {boneName=head}}]
    constraintWeights [1]
   }
   n69 = "Transform (Modify) Bone - Bone: dyn_r_earring_mstr"(ComponentPose=@n68, Alpha=1, Translation="-0.8,0,0", Rotation="0,-7,0", Scale="1,1,1") {
    rotationMode=BMM_Additive
    rotationSpace=BCS_BoneSpace
    translationMode=BMM_Additive
    translationSpace=BCS_BoneSpace
    alphaBoolBlend {blendOption=Linear}
    boneToModify {boneName=dyn_r_earring_mstr}
    rotation {yaw=-7}
    translation {x=-0.8}
   }
   n70 = "Transform (Modify) Bone - Bone: dyn_l_earring_mstr"(ComponentPose=@n69, Alpha=1, Translation="-0.8,0,0", Rotation="0,7,0", Scale="1,1,1") {
    rotationMode=BMM_Additive
    rotationSpace=BCS_BoneSpace
    translationMode=BMM_Additive
    translationSpace=BCS_BoneSpace
    alphaBoolBlend {blendOption=Linear}
    boneToModify {boneName=dyn_l_earring_mstr}
    rotation {yaw=7}
    translation {x=-0.8}
   }
   n71 = Get TargetControl() -> (TargetControl:double)
   n72 = "Transform (Modify) Bone - Bone: dyn_r_earring_mstr"(ComponentPose=@n70, Alpha=@n71, Translation="-2,0,0", Rotation="0,0,0", Scale="1,1,1") {
    rotationMode=BMM_Additive
    rotationSpace=BCS_BoneSpace
    translationMode=BMM_Additive
    translationSpace=BCS_BoneSpace
    alphaBoolBlend {blendOption=Linear}
    boneToModify {boneName=dyn_r_earring_mstr}
    translation {x=-2}
   }
   n73 = "Transform (Modify) Bone - Bone: dyn_l_earring_mstr"(ComponentPose=@n72, Alpha=@n71, Translation="-2,0,0", Rotation="0,0,0", Scale="1,1,1") {
    rotationMode=BMM_Additive
    rotationSpace=BCS_BoneSpace
    translationMode=BMM_Additive
    translationSpace=BCS_BoneSpace
    alphaBoolBlend {blendOption=Linear}
    boneToModify {boneName=dyn_l_earring_mstr}
    translation {x=-2}
   }
   n74 = Get IsTargeting()
   n75 = Anim Dynamics - Bone: dyn_earring_base_r(ComponentPose=@n73, Alpha=@n74, ExternalForce="0,0,0") {
    angularDampingOverride=1
    angularSpringConstant=200
    bGravityOverrideInSimSpace=true
    bOverrideAngularDamping=true
    bOverrideLinearDamping=true
    bUseGravityOverride=true
    gravityScale=0.35
    linearDampingOverride=1
    linearSpringConstant=135
    simulationSpace=RootRelative
    alphaBoolBlend {blendOption=Linear}
    boundBone {boneName=dyn_earring_base_r}
    componentAppliedLinearAccClamp=<1500 1500 1500>
    componentLinearAccScale=<0.3 0.3 0.4>
    componentLinearVelScale=<0.3 0.3 0.4>
    gravityOverride {z=-980}
    physicsBodyDefinitions [{
      sphereCollisionRadius=0
      boundBone {boneName=dyn_earring_base_r}
      boxExtents=<5 5 5>
      constraintSetup {angularLimitsMax=<5 40 35>, angularLimitsMin=<-5 -40 -20>, angularTarget=<1 0 0>}
      localJointOffset=<-1 0 0>
     }]
   }
   n76 = Anim Dynamics - Bone: dyn_earring_base_l(ComponentPose=@n75, Alpha=@n74, ExternalForce="0,0,0") {
    angularDampingOverride=1
    angularSpringConstant=200
    bGravityOverrideInSimSpace=true
    bOverrideAngularDamping=true
    bOverrideLinearDamping=true
    bUseGravityOverride=true
    gravityScale=0.35
    linearDampingOverride=1
    linearSpringConstant=135
    simulationSpace=RootRelative
    alphaBoolBlend {blendOption=Linear}
    boundBone {boneName=dyn_earring_base_l}
    componentAppliedLinearAccClamp=<1500 1500 1500>
    componentLinearAccScale=<0.3 0.3 0.4>
    componentLinearVelScale=<0.3 0.3 0.4>
    gravityOverride {z=-980}
    physicsBodyDefinitions [{
      sphereCollisionRadius=0
      boundBone {boneName=dyn_earring_base_l}
      boxExtents=<5 5 5>
      constraintSetup {angularLimitsMax=<5 40 25>, angularLimitsMin=<-5 -40 -35>, angularTarget=<1 0 0>}
      localJointOffset=<1 0 0>
     }]
   }
   n77 = Anim Dynamics - Bone: dyn_r_earring(ComponentPose=@n76, Alpha=@n74, ExternalForce="0,0,0") {
    angularDampingOverride=1
    angularSpringConstant=200
    bGravityOverrideInSimSpace=true
    bOverrideAngularDamping=true
    bUseGravityOverride=true
    gravityScale=0.35
    linearDampingOverride=1
    linearSpringConstant=135
    simulationSpace=BoneRelative
    alphaBoolBlend {blendOption=Linear}
    boundBone {boneName=dyn_r_earring}
    componentAppliedLinearAccClamp=<1500 1500 1500>
    componentLinearAccScale=<0.3 0.3 0.4>
    componentLinearVelScale=<0.3 0.3 0.4>
    gravityOverride {z=-980}
    physicsBodyDefinitions [{
      sphereCollisionRadius=0
      boundBone {boneName=dyn_r_earring}
      boxExtents=<5 5 5>
      constraintSetup {angularLimitsMax=<45 35 35>, angularLimitsMin=<-45 -35 -25>, angularTarget=<1 0 0>}
      localJointOffset=<1 0 0>
     }]
    relativeSpaceBone {boneName=root}
   }
   n78 = Anim Dynamics - Bone: dyn_l_earring(ComponentPose=@n77, Alpha=@n74, ExternalForce="0,0,0") {
    angularDampingOverride=1
    angularSpringConstant=200
    bGravityOverrideInSimSpace=true
    bOverrideAngularDamping=true
    bUseGravityOverride=true
    gravityScale=0.35
    linearDampingOverride=1
    linearSpringConstant=135
    simulationSpace=BoneRelative
    alphaBoolBlend {blendOption=Linear}
    boundBone {boneName=dyn_l_earring}
    componentAppliedLinearAccClamp=<1500 1500 1500>
    componentLinearAccScale=<0.3 0.3 0.4>
    componentLinearVelScale=<0.3 0.3 0.4>
    gravityOverride {z=-980}
    physicsBodyDefinitions [{
      sphereCollisionRadius=0
      boundBone {boneName=dyn_l_earring}
      boxExtents=<5 5 5>
      constraintSetup {angularLimitsMax=<45 35 35>, angularLimitsMin=<-45 -35 -25>, angularTarget=<1 0 0>}
      localJointOffset=<1 0 0>
     }]
    relativeSpaceBone {boneName=root}
   }
   n79 = Control Rig(Source=Component To Local(ComponentPose=@n78)) {
    alphaBoolBlend {blendOption=Linear}
    controlRigAssetReference {blueprintRigClass=&/F_MED_RoyalDuskBlaze_FaceAcc_CtrlRig.F_MED_RoyalDuskBlaze_FaceAcc_CtrlRig_C}
   }
   n80 = RigidBody!(
    ComponentPose=Local To Component(LocalPose=@n79)
    Alpha=Get backbling_alpha() -> (backbling_alpha:double)
    OverrideWorldGravity:Vector=Get gravity_override() -> (gravity_override:Vector)
    ExternalForce
   ) -> (Pose) {
    bClampLinearTranslationLimitToRefPose=true
    bOverrideWorldGravity=true
    overridePhysicsAsset=&/F_MED_RoyalDuskBlaze_FaceAcc_Physics
    simulationSpace=BaseBoneSpace
    alphaBoolBlend {blendOption=Linear}
    baseBoneRef {boneName=spine_05}
    componentAppliedLinearAccClamp=<5000 5000 5000>
    componentLinearAccScale=<0.6 0.2 0.2>
    componentLinearVelScale=<0.3 0.1 0.1>
    overrideWorldGravity {z=-980}
    simSpaceSettings {
     maxAngularAcceleration=1000
     maxAngularVelocity=3
     maxLinearAcceleration=700
     maxLinearVelocity=500
     velocityScaleZ=0
     worldAlpha=1
     externalLinearDragV=<0.2 0.2>
    }
   }
   Save cached pose 'Base_Pose'(Pose=Component To Local(ComponentPose=@n80)) {cacheName=Base_Pose}
  ]}
 EventGraph {nodes [
   Event BlueprintUpdateAnimation()
   > n1 = Sequence!(execute) -> (then_0, then_1, then_2:Exec)
   > n2 = Gravity Override State Function DEPRECATED(CharacterPartAnimInstance=Self-Reference(), GameTimeInSeconds=GetGameTimeInSeconds(), GravityOverrideParams=Get Gravity Override Params())
   > Set gravity_override(gravity_override=@n2)
   > n3 = ComplexSimSpaceFunction DEPRECATED!(
    execute
    CharacterPartAnimInstance:Custom Character Part Anim Instance Object Reference=Self-Reference()
    InterpSpeedMax:double=9
    PawnSpeedMax:double=20
    RInterpErrorTolerance:double=1
    DeltaTime:double=0
    DebugDraw:bool=false
    sim_space_rotation_in:Rotator=Get sim_space_rotation() -> (sim_space_rotation:Rotator)
    inComponentSpaceIn:bool=Get InComponentSpace() -> (InComponentSpace:bool)
    inComponentSpaceLastFrameIn:bool=Get InComponentSpaceLastFrame() -> (InComponentSpaceLastFrame:bool)
    lockedOnSimSpaceIn:bool=Get LockedOnSimSpace() -> (LockedOnSimSpace:bool)
    InterpSpeedIn:double=Get InterpSpeed() -> (InterpSpeed:double)
   ) -> (
    then
    sim_space_rotation:Rotator
    inComponentSpaceOut:bool
    inComponentSpaceLastFrameOut:bool
    lockedOnSimSpaceOut:bool
    InterpSpeedOut:double
   )
   > Set sim_space_rotation(sim_space_rotation:Rotator=@n3) -> (Output_Get:Rotator)
   > Set InComponentSpace(InComponentSpace:bool=@n3.inComponentSpaceOut) -> (Output_Get:bool)
   > Set InComponentSpaceLastFrame(InComponentSpaceLastFrame:bool=@n3.inComponentSpaceLastFrameOut) -> (Output_Get:bool)
   > Set LockedOnSimSpace(LockedOnSimSpace:bool=@n3.lockedOnSimSpaceOut) -> (Output_Get:bool)
   > Set InterpSpeed(InterpSpeed:double=@n3.InterpSpeedOut) -> (Output_Get:double)
   > Branch(Condition=Get bIsDBNOCarried())
   > Set gravity_override(gravity_override="0, 0,-980")
   n4 = Get bIsTargeting()
   Set IsTargeting(
    execute=@n1.then_1
    IsTargeting:float=FInterpTo(Current=Get IsTargeting(), Target=SelectFloat(A=0.3, B=1, bPickA=@n4), DeltaTime=0.5, InterpSpeed=0.5)
   ) -> (Output_Get:float)
   > Set TargetControl(TargetControl:double=FInterpTo(Current=Get TargetControl(), Target=SelectFloat(A=1, B=0, bPickA=@n4), DeltaTime=0.5, InterpSpeed=0.5)) -> (Output_Get:double)
   n5 = CheckForBackbling()
   > Set backbling_alpha(backbling_alpha:double=SelectFloat(A=0.5, B=1, bPickA=@n5)) -> (Output_Get:double)
   > Set backbling_equipped(backbling_equipped:bool=@n5) -> (Output_Get:bool)
   n6 = Get bIsSkydiveDiveMode()
   > Set skydiving_alpha(skydiving_alpha:double=FInterpTo(Current=Get skydiving_alpha() -> (skydiving_alpha:double), Target=SelectFloat(A=1, B=0, bPickA=@n6), DeltaTime=0.5, InterpSpeed=0.5)) -> (Output_Get:double)
   n7 = Cast To PlayerPawn_Athena(execute=@n1.then_2, Object=TryGetPawnOwner())
   > Set Body_Mesh(Body_Mesh=GetSkeletalMeshForPartType(self=@n7))
   n8 = Event OnEndSkydiving()
   Delay(execute=[@n8.then, @n9.then], Duration=0.1)
   > ResetDynamics(InTeleportType=TeleportPhysics)
   > ResetDynamics(InTeleportType=ResetPhysics)
   n9 = Event OnEndPlayingEmote() {is=title K2Node_Event}
   Get bIsJumping()
  ]}
]
variables [
 backbling_alpha {type=number}
 backbling_equipped {type=boolean}
 Body_Mesh {type=SkeletalMeshComponent}
 Gravity Override Params {type=GravityOverrideParamsStruct}
 gravity_override {type=Vector}
 InComponentSpace {type=boolean}
 InComponentSpaceLastFrame {type=boolean}
 InterpSpeed {type=number}
 IsTargeting {type=number}
 LockedOnSimSpace {type=boolean}
 sim_space_rotation {type=Rotator}
 skydiving_alpha {type=number}
 TargetControl {type=number}
]
